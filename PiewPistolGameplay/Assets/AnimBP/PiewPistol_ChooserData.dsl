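; Animation chooser data for the Piew Pistol that feeds the weapon's overheat state into pose selection.
; On each chooser data update, it gets the weapon, checks validity and casts it to BP_Pistol_Piew_Athena.
; It copies that pistol's Overheated value into OverheatedRelaxPose.
; The item layer update event has no logic.
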(event EventBlueprintUpdateChooserData
  (params (DeltaTime :type Float%20%28single-precision%29))
  (bind _returnvalue (CallFunction|TryGetWeapon))
  (if (Utilities|IsValid _returnvalue :!class /Script/Engine.KismetSystemLibrary)
    (bind _asbp_pistol_piew_athena (Utilities|Casting|CastToBP_Pistol_Piew_Athena _returnvalue))
    (Variables|Default|SetOverheatedRelaxPose (Variables|/PiewPistolGameplay/Gameplay/BP_Pistol_Piew_Athena.BP_Pistol_Piew_Athena_C|GetOverheated _asbp_pistol_piew_athena))))

(event EventBlueprintOnUpdateItemLayer
  (params (ItemLayerAnimInstance :type Anim%20Instance%20Object%20Reference) (DeltaTime :type Float%20%28single-precision%29)))
