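; Pickup audio Blueprint for the Piew Pistol that plays a music sting the first time a pawn picks up the weapon.
; On pickup, where cosmetic events can run, it checks that the pawn is valid, locally viewed and lacks the HasEquipped tag.
; It then attaches the music sound to the mesh, registers it with the music event subsystem and adds that loose tag to prevent repeats.
; The actor destroys itself afterwards.
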(event EventOnPickedUp
  (params (PickupTarget :type Fort%20Pawn%20Object%20Reference))
  (|CallParentFunction PickupTarget)
  (Networking|CanExecuteCosmeticEvents
    (:True
      (if (not (GameplayTags|HasMatchingGameplayTag PickupTarget "(TagName=\"PiewPistol.HasEquipped\")"))
        (Utilities|IsValid PickupTarget
          (:Is%20Valid
            (if (Pawn|IsLocallyViewed PickupTarget)
              (bind _returnvalue (Audio|SpawnSoundAttached "/PiewPistolGameplay/Assets/Audio/IP_MSS_PiewPistol_Pickup_Music.IP_MSS_PiewPistol_Pickup_Music" (Variables|/Script/Engine.Character|GetMesh PickupTarget)))
              (Audio|AddEventbyComponent (GameInstanceSubsystems|GetMusicEventSubsystem) _returnvalue "(TagName=\"MusicEventSubsystem.Event.Gameplay.Ability.1P\")" "(GameplayTags=)" 0 "")
              (Ability|GameplayEffect|AddLooseGameplayTags PickupTarget (GameplayTags|MakeGameplayTagContainerfromTag "(TagName=\"PiewPistol.HasEquipped\")"))
              (Actor|DestroyActor))))))
    (:False
      _))
  (Actor|DestroyActor))

(fn ConstructionScript)
