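; Pickup component for the Piew Pistol that plays a music sting the first time the locally viewed pawn picks it up.
; On BeginPlay it casts its owner to FortPickup and binds a custom handler to OnPickup.
; Where cosmetics can run, the handler attaches the sound to the pawn's mesh and registers it with MusicEventSubsystem as a first-person ability event.
; A loose PiewPistol.HasEquipped gameplay tag on the pawn blocks replays.
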(event EventBeginPlay
  (bind _asfort_pickup (Utilities|Casting|CastToFortPickup (Components|GetOwner)))
  (Interaction|BindEventtoOnPickup _asfort_pickup (EventDispatchers|CreateEvent :!event Custom|PickUpCustomAudio)))

(event Custom|PickUpCustomAudio
  (params (SelfActor :type Fort%20Pickup%20Object%20Reference) (InteractingPawn :type Fort%20Pawn%20Object%20Reference))
  (Networking|CanExecuteCosmeticEvents
    (:True
      (Utilities|IsValid InteractingPawn
        (:Is%20Valid
          (if (Pawn|IsLocallyViewed InteractingPawn)
            (if (not (GameplayTags|HasMatchingGameplayTag InteractingPawn "(TagName=\"PiewPistol.HasEquipped\")"))
              (bind _returnvalue (Audio|SpawnSoundAttached "/PiewPistolGameplay/Assets/Audio/IP_MSS_PiewPistol_Pickup_Music.IP_MSS_PiewPistol_Pickup_Music" (Variables|/Script/Engine.Character|GetMesh InteractingPawn)))
              (Audio|AddEventbyComponent (GameInstanceSubsystems|GetMusicEventSubsystem) _returnvalue "(TagName=\"MusicEventSubsystem.Event.Gameplay.Ability.1P\")" "(GameplayTags=)" 0 "")
              (Ability|GameplayEffect|AddLooseGameplayTags InteractingPawn (GameplayTags|MakeGameplayTagContainerfromTag "(TagName=\"PiewPistol.HasEquipped\")")))))))))
