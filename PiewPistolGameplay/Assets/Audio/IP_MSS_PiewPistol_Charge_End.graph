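# One-shot mono MetaSound for the end of the PiewPistol charge, using the ranged weapon sound class and the charge attenuation.
# On play it picks a random wave from the Sounds array and pitch-shifts it randomly around a base value within a spread.
# The player's left channel is scaled by a fixed decibel gain, then by an affiliation mix that makes enemies louder than non-enemies.
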MetaSoundSource /PiewPistolGameplay/Assets/Audio/IP_MSS_PiewPistol_Charge_End
inputs [
 Character.Affiliation {type=Enum:FortAudioAffiliation}
 Character.Health {type=Float}
 Character.UniqueID {type=Int32}
 Random.Pitch.Base {type=Float}
 Random.Pitch.Spread {type=Float}
 Random.Volume.Volume Base {type=Float}
 Sounds {type=WaveAsset:Array}
 UE.Source.OnPlay {type=Trigger}
]
interfaces [Fort.Character:1.0, UE.OutputFormat.Mono:1.0, UE.Source.OneShot:1.0, UE.Source:1.1]
outputs [
 "UE.OutputFormat.Mono.Audio:0" {type=Audio}
 UE.Source.OneShot.OnFinished {type=Trigger}
]
pages [Default {
  input_defaults {
   Character.Affiliation=0
   Character.Health=0
   Character.UniqueID=0
   Random.Pitch.Base=0
   Random.Pitch.Spread=0.5
   Random.Volume.Volume Base=-8
   UE.Source.OnPlay=null
   Sounds [
    &/IP_PiewPistol_Charge_End_01
    &/IP_PiewPistol_Charge_End_02
    &/IP_PiewPistol_Charge_End_03
   ]
  }
  nodes [
   DecibelstoLinearGain_0 = UE.Decibels to Linear Gain:Float(Decibels=^Random.Volume.Volume Base)
   ArrayRandomGet_0 = Array.Random Get:WaveAsset:Array(In Array=^Sounds, Next=^UE.Source.OnPlay)
   WavePlayer_0 = UE.Wave Player:Stereo(
    Play=@ArrayRandomGet_0.On Next
    Wave Asset=@ArrayRandomGet_0.Value
    Pitch Shift=UE.RandomFloat(
     Next=^UE.Source.OnPlay
     Max=UE.Add:Float(PrimaryOperand=^Random.Pitch.Base, AdditionalOperands=^Random.Pitch.Spread).Out
     Min=UE.Subtract:Float(PrimaryOperand=^Random.Pitch.Base, AdditionalOperands=^Random.Pitch.Spread).Out
    ).Value
   ) {also {version [1,1]}}
   Multiply_1 = UE.Multiply:Audio by Float(
    AdditionalOperands=&MSP_AffiliationMixFloat(Affiliation=^Character.Affiliation) {also {defaults {Enemy=0.5, Non-Enemy=0.3}}}.Out
    PrimaryOperand=UE.Multiply:Audio by Float(AdditionalOperands=@DecibelstoLinearGain_0.Linear Gain, PrimaryOperand=@WavePlayer_0.Out Left).Out
   )
  ]
  links [{dst="@UE.OutputFormat.Mono.Audio:0", src=Multiply_1.Out}, {dst="@UE.Source.OneShot.OnFinished", src=WavePlayer_0.On Finished}]
 }]
properties {
 assetImportData=&self:AssetImportData
 attenuationSettings=&/IP_ATT_PiewPistol_Charge
 maxDistance=22400
 sampleRate=48000
 soundClassObject=&Fort_WeaponRanged
}
