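# One-shot stereo MetaSound Source for the Piew Pistol pickup music sting, classed as loot collect with item attenuation and concurrency.
# On play, a stereo Wave Player plays a single music wave and signals finished when it ends.
# Both channels pass through multipliers sharing one decibels-to-linear gain node, left at its default, before reaching the outputs.
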MetaSoundSource /PiewPistolGameplay/Assets/Audio/IP_MSS_PiewPistol_Pickup_Music
inputs [UE.Source.OnPlay {type=Trigger}]
interfaces [UE.OutputFormat.Stereo:1.0, UE.Source.OneShot:1.0, UE.Source:1.1]
outputs [
 "UE.OutputFormat.Stereo.Audio:0" {type=Audio}
 "UE.OutputFormat.Stereo.Audio:1" {type=Audio}
 UE.Source.OneShot.OnFinished {type=Trigger}
]
pages [Default {
  input_defaults {UE.Source.OnPlay=null}
  nodes [
   DecibelstoLinearGain_0 = UE.Decibels to Linear Gain:Float()
   WavePlayer_0 = UE.Wave Player:Stereo(Play=^UE.Source.OnPlay) {also {defaults {Wave Asset=&/IP_PiewPistol_Pickup_Music_01}, version [1,1]}}
   Multiply_0 = UE.Multiply:Audio by Float(AdditionalOperands=@DecibelstoLinearGain_0.Linear Gain, PrimaryOperand=@WavePlayer_0.Out Left)
   Multiply_1 = UE.Multiply:Audio by Float(AdditionalOperands=@DecibelstoLinearGain_0.Linear Gain, PrimaryOperand=@WavePlayer_0.Out Right)
  ]
  links [
   {dst="@UE.OutputFormat.Stereo.Audio:0", src=Multiply_0.Out}
   {dst="@UE.OutputFormat.Stereo.Audio:1", src=Multiply_1.Out}
   {dst="@UE.Source.OneShot.OnFinished", src=WavePlayer_0.On Finished}
  ]
 }]
properties {
 assetImportData=&self:AssetImportData
 attenuationSettings=/FortMetasound/Items/AS_Fort_Item
 maxDistance=3500
 numChannels=2
 outputFormat=Stereo
 sampleRate=48000
 soundClassObject=&Fort_Loot_Collect
 assetUserData [&self:AudioAssetUserData_0]
 concurrencySet [&CS_Fort_Item]
}
