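# Unlit additive Niagara sprite material for the idle energy ball effect of the PiewPistol.
# Radial-corrected UVs drive panning caustic, noise and star textures, masked by distance from center and tinted by particle color around a hot core.
# A static switch adds energy wisps and a normal-map-distorted outer "sucky" layer.
# A soft sphere mask forms opacity, with depth fade on all shading paths except mobile.
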Material /PiewPistolGameplay/Assets/FX/M_EnergyBall_PiewPistol_Idle
graph {
 properties {
  bAutomaticallySetUsageInEditor=false
  bUsedWithNiagaraSprites=true
  blendMode=BLEND_Additive
  shadingModel=MSM_Unlit
  usedShadingModels=Unlit
 }
 nodes [
  n1 = TextureCoordinate()
  n2 = ComponentMask(None=Add(A=@n1, B=Constant() {r=-0.5})) {g=true, r=true}
  n3 = DotProduct(A=@n2, B=@n2)
  n4 = Power(Base=@n3) {constExponent=1.5}
  MaterialExpressionConstantClamp_0 = Saturate(None=@n3)
  n5 = Power(Base=@MaterialExpressionConstantClamp_0)
  MaterialExpressionVertexColor_0 = ParticleColor()
  n6 = OneMinus(None=SquareRoot(None=@n3))
  n7 = Power(Base=@n6, Exp=ScalarParameter() {defaultValue=8, parameterName=EnergyAmount})
  n8 = MaterialFunctionCall!(UVs=Subtract(A=TextureCoordinate() {uTiling=2, vTiling=2}, B=Constant() {r=1})) {materialFunction=/Game/Effects/Fort_Effects/Materials/_Functions/VectortoRadialCorrection}
  n9 = Constant2Vector() {g=1, r=3}
  n10 = Time()
  n11 = ScalarParameter() {defaultValue=3, parameterName=Speed}
  n12 = Add(
   A=Panner(Coordinate=Multiply(A=@n8.Vector to Radial, B=@n9), Time=Multiply(A=@n10, B=@n11)) {speedY=-0.15}
   B=Multiply(
    A=TextureSample(UVs=@n8.Vector to Radial) {texture=/Game/Weapons/FORT_Melee/Pickaxe_CoatCheck/FX/T_Pickaxe_CoatCheck_Noise}.G
    B=Constant() {r=0.5}
   )
  )
  n13 = Panner(Coordinate=@n8.Vector to Radial) {speedY=0.02}
  n14 = Panner(Coordinate=Multiply(A=@n8.Vector to Radial, B=Constant2Vector() {g=0.2, r=0.8})) {speedY=0.1}
  n15 = Multiply(
   A=Add(
    A=TextureSample(UVs=@n12) {texture=&T_ShieldCaustics}.RGB
    B=TextureSample(UVs=@n13) {texture=&T_Pickaxe_CoatCheck_Noise}.RGB
   )
   B=Multiply(A=TextureSample(UVs=@n14) {texture=/Game/Effects/Fort_Effects/Gameplay/Butterfly/T_noise_stars}.B, B=Constant() {r=50})
  )
  n16 = Multiply(A=@n8.Vector to Radial, B=Constant2Vector() {g=0.05, r=1})
  MaterialExpressionConstantClamp_3 = Saturate(None=TextureSample(UVs=Panner(Coordinate=@n16) {speedX=0.025, speedY=0.5}) {texture=/Game/Effects/Fort_Effects/Textures/Energy/T_NoisePack_RGB}.B)
  n17 = ScalarParameter() {defaultValue=2, parameterName=EnergyWisps_Amount}
  MaterialExpressionConstantClamp_2 = Saturate(None=Multiply(A=Power(Base=@MaterialExpressionConstantClamp_3, Exp=@n17), B=ScalarParameter() {defaultValue=64, parameterName=EnergyWisps_Intensity}))
  n18 = Panner(Coordinate=@n16) {speedY=-0.3}
  n19 = Multiply(
   A=ComponentMask(None=TextureSample(UVs=@n18) {samplerType=SAMPLERTYPE_Normal, texture=/Game/Effects/SnowSandInteraction/Textures/SSI_T_SandNormal}.RGB) {g=true, r=true}
   B=Constant() {r=1}
  )
  n20 = ComponentMask(None=Add(A=Add(A=@n1, B=@n19), B=Constant() {r=-0.7})) {g=true, r=true}
  n21 = DotProduct(A=@n20, B=@n20)
  n22 = ScalarParameter() {defaultValue=5, parameterName=OuterSucky_Amount}
  MaterialExpressionConstantClamp_4 = Saturate(None=Multiply(A=Power(Base=@n21, Exp=@n22), B=ScalarParameter() {defaultValue=1000, parameterName=OuterSucky_Hardness}))
  n23 = Add(
   A=Multiply(A=@MaterialExpressionConstantClamp_2) {constB=3}
   B=Multiply(A=@MaterialExpressionConstantClamp_4, B=ScalarParameter() {defaultValue=0.25, parameterName=OuterSucky_Intensity})
  )
  n24 = Multiply(
   A=StaticSwitchParameter(True=Add(A=@n15, B=@n23), False=@n15) {defaultValue=true, parameterName=EnergyWisps}
   B=Multiply(A=@MaterialExpressionVertexColor_0.RGB, B=@MaterialExpressionVertexColor_0.RGB)
  )
  n25 = Constant() {r=50}
  n26 = Constant() {r=64}
  MaterialExpressionConstantClamp_1 = Saturate(None=Multiply(A=Power(Base=@n6, Exp=@n26), B=Constant() {r=2}))
  n27 = Add(
   A=Add(A=Multiply(B=@n4) {constA=0.1}, B=Multiply(A=@n5, B=@MaterialExpressionVertexColor_0.RGB))
   B=LinearInterpolate(A=Multiply(A=@n7, B=@n24), B=Multiply(A=@MaterialExpressionVertexColor_0.RGB, B=@n25), Alpha=@MaterialExpressionConstantClamp_1)
  )
  MaterialFunctionCall(Emissive=LinearInterpolate(A=@n27, B=Saturate(None=@n27))) {materialFunction=&MF_FX_EmissiveAdjustments}
  n28 = SphereMask(A=TextureCoordinate(), B=Constant2Vector() {g=0.5, r=0.5}) {attenuationRadius=0.5, hardnessPercent=0}
  MaterialExpressionConstantClamp_5 = Saturate(None=Power(Base=Multiply(A=@n28) {constB=0.25}))
  n29 = DepthFade(Opacity=@MaterialExpressionConstantClamp_5) {fadeDistanceDefault=10}
  ShadingPathSwitch(Default=@n29, Deferred=@n29, Forward=@n29, Mobile=@MaterialExpressionConstantClamp_5)
  SmoothStep() {constMin=0.5}
  StaticBool() {value=true}
 ]
}
