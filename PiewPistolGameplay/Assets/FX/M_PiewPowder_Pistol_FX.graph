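# M_PiewPowder_Pistol_FX is a Default Lit material for the PiewPowder pistol FX mesh, flagged for use with Niagara mesh particles.
# It holds three texture samples from the pistol's texture set: D, N sampled as a normal map, and S sampled as masks.
# The graph has no math or parameters and shows no links from the samples to material outputs.
# Automatic usage flagging in the editor is off.
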Material /PiewPistolGameplay/Assets/FX/M_PiewPowder_Pistol_FX
graph {
 properties {
  bAutomaticallySetUsageInEditor=false
  bUsedWithNiagaraMeshParticles=true
  previewMesh=&/SM_PiewPowder_Pistol_FX
  usedShadingModels=Default Lit
 }
 nodes [
  TextureSample() {samplerType=SAMPLERTYPE_Masks, texture=/PiewPistolGameplay/Assets/Textures/T_PiewPowder_Pistol_S}
  TextureSample() {samplerType=SAMPLERTYPE_Normal, texture=/PiewPistolGameplay/Assets/Textures/T_PiewPowder_Pistol_N}
  TextureSample() {texture=/PiewPistolGameplay/Assets/Textures/T_PiewPowder_Pistol_D}
 ]
}
