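# Niagara system for the ADS explosion of the Piew Pistol, tinted by hue-shift parameters and mostly sized by User.FXScale.
# Combines an impact-normal decal, mesh electricity, sprite streaks, flash with distortion, flare and light, and radial smoke.
# PackExplosionFX and PackRadialFX each drive several renderers from one particle via per-layer attributes for shockwave, ring, plasma and other layers.
# Camera distance scales decal and streaks; quality level sets decal lifetime.
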NiagaraSystem /PiewPistolGameplay/Assets/FX/NS_Explosion_PiewPistol_ADS
dependencies {
 usedDataInterfaces [
  &NiagaraDataInterfaceCurve
  &NiagaraDataInterfaceVector2DCurve
  &NiagaraDataInterfaceColorCurve
  &NiagaraDataInterfaceVectorCurve
  &NiagaraDataInterfaceArrayFloat
 ]
 usedDynamicInputs []
 usedModules [&SystemState]
}
emitters {
 Decal {
  emitterUpdateScript [
   &EmitterState()
   &SpawnBurst_Instantaneous(Use Spawn Probability=false, UseLoopCountLimit=false)
  ]
  particleSpawnScript [
   &InitializeParticle(
    Color=&HueShiftLinearColor(Hue=$System.HueShift, Saturation=1, Value=1, Linear Color=<rgba 2 0 0.1667 1>)
    Lifetime=&SelectFloatFromArray(Direct Array Index=$Engine.QualityLevel, Float Selection Array=DataInterface() {propertyValues {FloatData [2,3,3.5,4]}})
    Position=$Particles.Position
    Position Mode=::NewEnumerator1
    UsePositionOffset=false
    Write Material Random=true
   )
   /Niagara/Modules/Spawn/Decal/InitialDecalOrientation(
    Orientation Coordinate Space:ENiagaraCoordinateSpace=::World
    Orientation Vector:Vector3f=$User.NormalImpact
    Rotation:Vector3f=&RandomRangeVector(Minimum=<0 0 0>, Maximum=<0 0 1>)
    Use Rotation Vector=true
   )
  ]
  particleUpdateScript [
   &ParticleState()
   /SetVariables(Particles.DecalSize:Vector3f=&Multiply_VectorByFloat(
     Vector=<200 320 320>
     Float=&FloatFromCurve(
      FloatCurve=DataInterface() {propertyValues {Curve {keys [
          {interpMode=RCIM_Cubic, tangentMode=RCTM_Break, time=0, value=0.5}
          {interpMode=RCIM_Cubic, tangentMode=RCTM_Break, time=0.7, value=1}
          <at 1 1>
         ]}}}
      CurveIndex=&FN_LOD_Distance_Remap(Start Distance=1000, End Distance=12000, Minimum Output=0, Maximum Output=1, Clamp Output=true)
      Scale Curve=1.5
     )
    )) {isSetParameters=true}
   /SetVariables(Particles.DecalFade:NiagaraFloat=&Lerp_Float(
     A=0
     B=1
     Alpha=&FloatFromCurve(FloatCurve=DataInterface() {propertyValues {Curve {keys [
          {arriveTangent=41.722, time=0, value=0.25}
          {interpMode=RCIM_Cubic, tangentMode=RCTM_User, time=0.02, value=1}
          {arriveTangent=-1.1796, interpMode=RCIM_Cubic, leaveTangent=-1.1796, tangentMode=RCTM_User, time=0.7322, value=0.6336}
          {arriveTangent=-3.611, arriveTangentWeight=0.112, interpMode=RCIM_Cubic, leaveTangent=-3.611, tangentMode=RCTM_User, time=1, value=0}
         ]}}})
    )) {isSetParameters=true}
  ]
  rendererClasses [&NiagaraDecalRendererProperties]
  renderers [{rendererClass=&NiagaraDecalRendererProperties, rendererIndex=0}]
 }
 Electricity {
  emitterUpdateScript [
   &EmitterState(Life Cycle Mode=::NewEnumerator1, Loop Behavior=::NewEnumerator1, Loop Duration=0.2, UseLoopDelay=true)
   &SpawnBurst_Instantaneous(Spawn Count=6, Use Spawn Probability=false)
   &SpawnRate(SpawnRate=30, Use Spawn Probability=false)
  ]
  particleSpawnScript [
   /Niagara/Modules/Spawn/Initialization/InitializeParticle(
    Lifetime=&RandomRangeFloat(Minimum=0.225, Maximum=0.275)
    Mesh Scale=&RandomRangeVector(Minimum=<34 72 72>, Maximum=<45 90 90>)
    Module.Write Color=true
    Module.Write Lifetime=true
    Module.Write Mass=true
    Module.Write Material Random=true
    Module.Write Position=true
    Module.Write Scale=true
    Module.Write SpriteRotation=false
    Module.Write SpriteSize=false
    Position=&SimulationPosition()
    Sprite Size=<25 25>
   )
   &ShapeLocation(
    MaskSpawnGroup=false
    Override Seed=false
    Rotation Mode=::NewEnumerator1
    Sphere Radius=&Divide_Float(A=$User.FXScale, B=5)
    UseOwnerScale=true
   )
   &InitialMeshOrientation(
    Mesh Orientation Mode=::NewEnumerator2
    Orientation Vector=$Engine.Owner.SystemXAxis
    Rotation=&RandomRangeVector(Minimum=<-1 -1 -1>, Maximum=<1 1 1>)
    Use Rotation Vector=true
   )
   &DynamicMaterialParameters(
    Index 0 Param 1=&RandomRangeFloat(Minimum=6.5, Maximum=8.5)
    Index 0 Param 3=2.2
    Index 0 Param 4=&RandomRangeFloat(Minimum=-60, Maximum=60)
    Module.Param0WriteEnabled=true
    Module.Param1WriteEnabled=false
    Module.Param2WriteEnabled=true
    Module.Param3WriteEnabled=false
   )
  ]
  particleUpdateScript [
   &ParticleState()
   &ScaleMeshSize(Scale Factor=&VectorFromFloat(Value=&FloatFromCurve(
      FloatCurve=DataInterface() {propertyValues {Curve {keys [
          {arriveTangent=4.469, interpMode=RCIM_Cubic, leaveTangent=4.469, tangentMode=RCTM_User, time=0, value=0}
          {arriveTangent=1.1375, interpMode=RCIM_Cubic, leaveTangent=1.1375, tangentMode=RCTM_User, time=1, value=2}
         ]}}}
      Scale Curve=&Divide_Float(A=$User.FXScale, B=1750)
     )))
   &DynamicMaterialParameters(
    Index 0 Param 2=&FloatFromCurve(FloatCurve=DataInterface() {propertyValues {Curve {keys [<at 0 2>,<at 1 -0.6>]}}})
    Index 0 Param 3=2.2
    Index 0 Param 4=&RandomRangeFloat(Minimum=-60, Maximum=60)
    Module.Param0WriteEnabled=false
    Module.Param1WriteEnabled=true
    Module.Param2WriteEnabled=false
    Module.Param3WriteEnabled=true
   ) {module=DynamicMaterialParameters001}
   &Color(Color=&HueShiftLinearColor(
     Hue=$User.HueShift
     Saturation=1
     Value=1
     Linear Color=&ColorFromCurve(ColorCurve=DataInterface() {propertyValues {
        AlphaCurve {keys [<at 0 1 RCIM_Cubic>,<at 1 0 RCIM_Cubic>]}
        BlueCurve {keys [<at 0 2.6158 RCIM_Cubic>,<at 0.6888 0.1667 RCIM_Cubic>]}
        GreenCurve {keys [<at 0 2.4906 RCIM_Cubic>,<at 0.6888 0 RCIM_Cubic>]}
        RedCurve {keys [<at 0 10 RCIM_Cubic>,<at 0.6888 1 RCIM_Cubic>]}
       }})
    ))
  ]
  rendererClasses [&NiagaraMeshRendererProperties]
  renderers [{rendererClass=&NiagaraMeshRendererProperties, rendererIndex=0}]
 }
 Impact_Streaks {
  emitterUpdateScript [
   &EmitterState()
   &SpawnBurst_Instantaneous(
    Spawn Count=&Multiply_Int(
     A=&RandomRangeInt(Minimum=3, Maximum=6)
     B=&MakeIntFromFloat(Float=&FN_LOD_Distance_Remap(Start Distance=0, End Distance=750, Minimum Output=1, Maximum Output=6, Clamp Output=true))
    )
    Use Spawn Probability=false
    UseLoopCountLimit=false
   )
  ]
  particleSpawnScript [
   /Niagara/Modules/Spawn/Initialization/V2/InitializeParticle(
    Color=<rgba 10 10 10 1>
    Lifetime Max=0.5
    Lifetime Min=0.35
    Lifetime Mode=::NewEnumerator1
    Sprite Size Max=<100 225>
    Sprite Size Min=<25 80>
    Sprite Size Mode=::NewEnumerator2
    UsePositionOffset=false
    Write Material Random=true
   )
   &SubUVAnimation(End Frame=3, SubUV Animation Mode:ENiagara_SubUVLookupMode=::NewEnumerator2)
   &ShapeLocation(
    Cylinder Distribution=::NewEnumerator2
    Cylinder Height=0
    Cylinder Radius=&Multiply_Float(
     A=10
     B=&FN_LOD_Distance_Remap(Start Distance=0, End Distance=1500, Minimum Output=0, Maximum Output=1, Clamp Output=true)
    )
    MaskSpawnGroup=false
    Shape Primitive=::NewEnumerator1
    U Position=&ReturnNormalizedExecIndex()
    UseOwnerScale=true
   )
   &AddVelocity(
    Velocity=&Add_Vector(A=$Particles.ShapeLocation.ShapeNormal, B=&RandomRangeVector(Minimum=<0 0 0>, Maximum=<1 0 0>))
    Velocity Speed Scale=&FN_LOD_Distance_Remap(Start Distance=0, End Distance=1500, Minimum Output=1, Maximum Output=50, Clamp Output=true)
   )
  ]
  particleUpdateScript [
   &SolveForcesAndVelocity(Clamp Velocity=false, Limit Acceleration=false)
   &ParticleState()
   &ScaleColor(
    Scale Alpha=&FloatFromCurve(FloatCurve=DataInterface() {propertyValues {Curve {keys [{arriveTangent=-3.3102, interpMode=RCIM_Cubic, leaveTangent=-3.3102, tangentMode=RCTM_User, time=0, value=1}, <at 1 0 RCIM_Cubic>]}}})
    Scale RGB=&MakeVectorFromLinearColorRGB(Color=&HueShiftLinearColor(
      Hue=$User.HueShift
      Saturation=1
      Value=2
      Linear Color=&ColorFromCurve(ColorCurve=DataInterface() {propertyValues {
         AlphaCurve {keys []}
         BlueCurve {keys [<at 0 0.3013 RCIM_Cubic>,<at 0.5629 0.0463>,<at 1 0.0313 RCIM_Cubic>]}
         GreenCurve {keys [<at 0 1.1095 RCIM_Cubic>,<at 0.5629 0.0301>,<at 1 0 RCIM_Cubic>]}
         RedCurve {keys [<at 0 10 RCIM_Cubic>,<at 0.5629 1>,<at 1 0.1878 RCIM_Cubic>]}
        }})
     ))
    ScaleA=true
    ScaleRGB=true
   )
   &ScaleSpriteSize(
    Non-Uniform Curve Scale=&Multiply_Vector2D_ByFloat(
     Float=&Divide_Float(A=$User.FXScale, B=550)
     Vector2D=&MakeVector2D(
      X=&FN_LOD_Distance_Remap(Start Distance=0, End Distance=750, Minimum Output=0.2, Maximum Output=1, Clamp Output=true)
      Y=&FN_LOD_Distance_Remap(Start Distance=0, End Distance=750, Minimum Output=0.25, Maximum Output=1, Clamp Output=true)
     )
    )
    Non-Uniform Curve Sprite Scale=DataInterface() {propertyValues {
      XCurve {keys [
        {arriveTangent=0.9067, interpMode=RCIM_Cubic, leaveTangent=0.9067, tangentMode=RCTM_User, time=0, value=0.8}
        {arriveTangentWeight=0.2034, interpMode=RCIM_Cubic, leaveTangentWeight=0.2034, tangentMode=RCTM_User, time=0.5, value=1}
        {arriveTangent=-3.1509, interpMode=RCIM_Cubic, leaveTangent=-3.1509, tangentMode=RCTM_User, time=1, value=0.1}
       ]}
      YCurve {keys [
        {arriveTangent=3.7985, interpMode=RCIM_Cubic, leaveTangent=3.7985, tangentMode=RCTM_User, time=0, value=0}
        {arriveTangent=2, interpMode=RCIM_Cubic, leaveTangent=2, time=0.5, value=1.6}
        <at 1 2 RCIM_Cubic>
       ]}
     }}
    Scale Sprite Size Mode=::NewEnumerator3
   )
  ]
  rendererClasses [&NiagaraSpriteRendererProperties]
  renderers [{rendererClass=&NiagaraSpriteRendererProperties}]
 }
 PackExplosionFX {
  emitterUpdateScript [
   &EmitterState()
   &SpawnBurst_Instantaneous(Spawn Count=1, Use Spawn Probability=false) {module=SpawnBurst_Instantaneous001}
  ]
  particleSpawnScript [
   &InitializeParticle(
    Color=&HueShiftLinearColor(Hue=$User.HueShift, Saturation=1, Value=1, Linear Color=<rgba 2 0.4022 0.5354 2>)
    Mesh Scale Mode=::NewEnumerator3
    Mesh Uniform Scale=14
    Sprite Size Mode=::NewEnumerator1
    Uniform Sprite Size Max=1500
    Uniform Sprite Size Min=1250
    UsePositionOffset=false
    Write Material Random=true
   )
   &InitialMeshOrientation(
    Mesh Orientation Mode=::NewEnumerator2
    Rotation=&RandomRangeVector(Minimum=<0 0 -0.25>, Maximum=<0 0 0.25>)
    Use Rotation Vector=true
   )
  ]
  particleUpdateScript [
   &ParticleState()
   &ScaleSpriteSize(Scale Sprite Size Mode=::NewEnumerator0, Uniform Scale Factor=&Divide_Float(A=$User.FXScale, B=350))
   &ScaleColor(
    Scale Alpha=&FloatFromCurve(
     FloatCurve=DataInterface() {propertyValues {Curve {keys [
         {arriveTangent=45, interpMode=RCIM_Cubic, time=0, value=1}
         {arriveTangent=-1, interpMode=RCIM_Cubic, leaveTangent=-1, time=0.3, value=2}
         <at 1 0 RCIM_Cubic>
        ]}}}
     CurveIndex=&NormalizeFloat(Value=&Divide_Float(A=$Particles.NormalizedAge, B=0.5), Min=0, Max=1)
     Scale Curve=0.25
    )
    Scale RGB=<3 3 3>
    ScaleA=true
    ScaleRGB=true
   )
   /SetVariables(
    Particles.Color_Distortion:LinearColor=&MultiplyLinearColor(
     Linear Color=&HueShiftLinearColor(Hue=$User.HueShift, Saturation=1, Value=1, Linear Color=<rgba 1 0.2491 0.3742 1>)
     Scale Factor=&MakeVector4FromVectorAndFloat(
      "Vector (RGB)"=&VectorFromCurve(
       VectorCurve=DataInterface() {propertyValues {XCurve {keys [<at 0 1>,<at 0.5 0.8>]}, YCurve {keys [<at 0 1>,<at 0.5 0.8>]}, ZCurve {keys [<at 0 1>,<at 0.5 0.8>]}}}
       CurveIndex=&NormalizeFloat(
        Value=&Divide_Float(A=$Particles.NormalizedAge, B=&RandomRangeFloat(Minimum=0.35, Maximum=0.4))
        Min=0
        Max=1
       )
      )
      "Float (Alpha)"=&FloatFromCurve(
       FloatCurve=DataInterface() {propertyValues {Curve {keys [
           {interpMode=RCIM_Cubic, tangentMode=RCTM_User, time=0.0045, value=0.9397}
           {arriveTangent=-1.0036, interpMode=RCIM_Cubic, leaveTangent=-1.0036, tangentMode=RCTM_User, time=0.4647, value=0.5233}
           {arriveTangent=-2.4196, interpMode=RCIM_Cubic, leaveTangent=-2.4196, tangentMode=RCTM_User, time=1.0011, value=0.002}
          ]}}}
       CurveIndex=&NormalizeFloat(
        Value=&Divide_Float(A=$Particles.NormalizedAge, B=&RandomRangeFloat(Minimum=0.35, Maximum=0.4))
        Min=0
        Max=1
       )
       Scale Curve=0.2
      )
     )
    )
    Particles.DP_01_Distortion:Vector4f=&MakeVector4(
     X=&FloatFromCurve(
      FloatCurve=DataInterface() {propertyValues {Curve {keys [
          {interpMode=RCIM_Cubic, tangentMode=RCTM_User, time=0, value=1}
          {interpMode=RCIM_Cubic, tangentMode=RCTM_User, time=0.3, value=0.98}
          {arriveTangent=-0.0005, interpMode=RCIM_Cubic, leaveTangent=-0.0005, tangentMode=RCTM_User, time=0.6, value=0.98}
          {interpMode=RCIM_Cubic, tangentMode=RCTM_User, time=1, value=1}
         ]}}}
      CurveIndex=&NormalizeFloat(
       Value=&Divide_Float(A=$Particles.NormalizedAge, B=&RandomRangeFloat(Minimum=0.35, Maximum=0.4))
       Min=0
       Max=1
      )
     )
     Y=&FloatFromCurve(
      FloatCurve=DataInterface() {propertyValues {Curve {keys [<at 0 15 RCIM_Cubic>, {arriveTangent=-40.0295, interpMode=RCIM_Cubic, leaveTangent=-40.0295, tangentMode=RCTM_User, time=1, value=0}]}}}
      CurveIndex=&NormalizeFloat(
       Value=&Divide_Float(A=$Particles.NormalizedAge, B=&RandomRangeFloat(Minimum=0.35, Maximum=0.4))
       Min=0
       Max=1
      )
     )
     Z=0
     W=0
    )
    Particles.Scale_Distortion:Vector3f=&Multiply_VectorByFloat(
     Vector=<7 7 7>
     Float=&FloatFromCurve(
      FloatCurve=DataInterface() {propertyValues {Curve {keys [
          {arriveTangent=4.1785, interpMode=RCIM_Cubic, leaveTangent=4.1785, tangentMode=RCTM_User, time=0, value=0}
          {arriveTangent=1.9532, interpMode=RCIM_Cubic, leaveTangent=1.9532, tangentMode=RCTM_User, time=0.2223, value=0.6}
          {arriveTangent=-0.0405, interpMode=RCIM_Cubic, leaveTangent=-0.0405, tangentMode=RCTM_User, time=1, value=1.25}
         ]}}}
      CurveIndex=&NormalizeFloat(
       Value=&Divide_Float(A=$Particles.NormalizedAge, B=&RandomRangeFloat(Minimum=0.35, Maximum=0.4))
       Min=0
       Max=1
      )
      Scale Curve=&Divide_Float(A=$User.FXScale, B=375)
     )
    )
   ) {isSetParameters=true}
   /SetVariables(
    Particles.CameraOffset_Flare=50
    Particles.Color_Flare:LinearColor=&MakeLinearColorFromVectorAndFloat(
     "Vector (RGB)"=&MakeVectorFromLinearColorRGB(Color=&HueShiftLinearColor(
       Hue=$User.HueShift
       Saturation=1
       Value=1
       Linear Color=&ColorFromCurve(
        ColorCurve=DataInterface() {propertyValues {
          AlphaCurve {keys []}
          BlueCurve {keys [<at 0 0.6026 RCIM_Cubic>,<at 0.125 0.0926>,<at 1 0.0313 RCIM_Cubic>]}
          GreenCurve {keys [<at 0 0.6026 RCIM_Cubic>,<at 0.125 0.0603>,<at 1 0 RCIM_Cubic>]}
          RedCurve {keys [<at 0 20 RCIM_Cubic>,<at 0.125 2>,<at 1 0.376 RCIM_Cubic>]}
         }}
        CurveIndex=&NormalizeFloat(Value=&Divide_Float(A=$Particles.NormalizedAge, B=0.15), Min=0, Max=1)
       )
      ))
     "Float (Alpha)"=&FloatFromCurve(
      FloatCurve=DataInterface() {propertyValues {Curve {keys [{arriveTangent=-3.3102, interpMode=RCIM_Cubic, time=0, value=0}, <at 0.5 1 RCIM_Cubic>, <at 1 0 RCIM_Cubic>]}}}
      CurveIndex=&NormalizeFloat(Value=&Divide_Float(A=$Particles.NormalizedAge, B=0.15), Min=0, Max=1)
     )
    )
    Particles.DP_01_Flare:Vector4f=<1 1 1 1>
    Particles.Position_Flare:NiagaraPosition=&AddVectorToPosition(Position=/Niagara/DynamicInputs/Helpers/SimulationPosition(), Vector=<0 0 25>)
   ) {isSetParameters=true}
   &DynamicMaterialParameters(
    Index 0 Param 1=&FloatFromCurve(
     FloatCurve=DataInterface() {propertyValues {Curve {keys [<at 0 0.07>,<at 1 0.25>]}}}
     CurveIndex=&NormalizeFloat(Value=&Divide_Float(A=$Particles.NormalizedAge, B=0.5), Min=0, Max=1)
    )
    Index 0 Param 2=&RandomRangeFloat(Minimum=0, Maximum=1)
    Index 0 Param 3=&RandomRangeFloat(Minimum=0, Maximum=1)
    Index 0 Param 4=&FloatFromCurve(
     FloatCurve=DataInterface() {propertyValues {Curve {keys [<at 0 1>,<at 1 3>]}}}
     CurveIndex=&NormalizeFloat(Value=&Divide_Float(A=$Particles.NormalizedAge, B=0.5), Min=0, Max=1)
    )
    Module.Param0WriteEnabled=true
    Module.Param1WriteEnabled=true
    Module.Param2WriteEnabled=true
    Module.Param3WriteEnabled=true
   )
   &FN_Light_Attributes(
    Color=&HueShiftLinearColor(Hue=$User.HueShift, Saturation=1, Value=1, Linear Color=<rgba 3 0.0652 0.2119 1>)
    Create Color Attribute=true
    Create Exponent Attribute=true
    Create Position Attribute=true
    Create Radius Attribute=true
    Exponent=1
    Position=&Add_Vector(A=&SimulationPosition(), B=<0 0 50>)
    Radius=&Divide_Float(A=$User.FXScale, B=5)
    Scale Color=&FloatFromCurve(
     FloatCurve=DataInterface() {propertyValues {Curve {keys [
         <at 0 1 RCIM_Cubic>
         {arriveTangent=-1, interpMode=RCIM_Cubic, leaveTangent=-1, leaveTangentWeight=0.9776, time=0.7, value=1}
         <at 1 0 RCIM_Cubic>
        ]}}}
     CurveIndex=&NormalizeFloat(Value=&Divide_Float(A=$Particles.NormalizedAge, B=0.5), Min=0, Max=1)
     Scale Curve=10
    )
   )
  ]
  rendererClasses [
   &NiagaraMeshRendererProperties
   &NiagaraSpriteRendererProperties
   &NiagaraLightRendererProperties
  ]
  renderers [
   {rendererClass=&NiagaraMeshRendererProperties, rendererIndex=0}
   {rendererClass=&NiagaraSpriteRendererProperties, rendererIndex=1}
   {rendererClass=&NiagaraLightRendererProperties, rendererIndex=2}
  ]
 }
 PackRadialFX {
  emitterUpdateScript [
   &EmitterState()
   &SpawnBurst_Instantaneous(Spawn Count=1, Use Spawn Probability=false)
  ]
  particleSpawnScript [&InitializeParticle(
    Color=&HueShiftLinearColor(Hue=$User.HueShift, Saturation=1, Value=1, Linear Color=<rgba 1 0.2491 0.4994 0.7>)
    Mesh Scale Mode=::NewEnumerator3
    Mesh Uniform Scale=&Multiply_Float(A=13, B=&Divide_Float(A=$User.FXScale, B=450))
    UsePositionOffset=false
    Write Material Random=true
   )]
  particleUpdateScript [
   &ParticleState()
   /SetVariables(
    Particles.CameraOffset_NoiseSphere=250
    Particles.Color_NoiseSphere:LinearColor=&MultiplyLinearColor(
     Linear Color=&HueShiftLinearColor(Hue=$User.HueShift, Saturation=1, Value=1, Linear Color=<rgba 4 0.5319 1.3989 3>)
     Scale Factor=&MakeVector4FromVectorAndFloat(
      "Vector (RGB)"=<1 1 1>
      "Float (Alpha)"=&FloatFromCurve(
       FloatCurve=DataInterface() {propertyValues {Curve {keys [<at 0 1>, {interpMode=RCIM_Cubic, tangentMode=RCTM_User, time=0.1, value=1}, {interpMode=RCIM_Cubic, tangentMode=RCTM_User, time=1, value=0}]}}}
       CurveIndex=&NormalizeFloat(Value=&Divide_Float(A=$Particles.NormalizedAge, B=0.3), Min=0, Max=1)
      )
     )
    )
    Particles.DP_01_NoiseSphere:Vector4f=<1 0 1 1>
    Particles.Scale_NoiseSphere:Vector3f=&Multiply_VectorByFloat(
     Vector=<7 7 7>
     Float=&FloatFromCurve(
      FloatCurve=DataInterface() {propertyValues {Curve {keys [
          {arriveTangent=4.1785, interpMode=RCIM_Cubic, leaveTangent=4.1785, tangentMode=RCTM_User, time=0, value=0.05}
          {arriveTangent=0.9995, interpMode=RCIM_Cubic, leaveTangent=0.9995, tangentMode=RCTM_User, time=0.3, value=0.6}
          {arriveTangent=-0.0405, interpMode=RCIM_Cubic, leaveTangent=-0.0405, tangentMode=RCTM_User, time=1, value=1}
         ]}}}
      CurveIndex=&NormalizeFloat(Value=&Divide_Float(A=$Particles.NormalizedAge, B=0.3), Min=0, Max=1)
      Scale Curve=&Divide_Float(A=$User.FXScale, B=550)
     )
    )
   ) {isSetParameters=true}
   /SetVariables(
    Particles.Color_ShockwaveSphere:LinearColor=&MultiplyLinearColor(
     Linear Color=&HueShiftLinearColor(Hue=$User.HueShift, Saturation=1, Value=1, Linear Color=<rgba 12 2.4132 5.6088 0.7>)
     Scale Factor=&MakeVector4FromVectorAndFloat(
      "Vector (RGB)"=&VectorFromCurve(
       VectorCurve=DataInterface() {propertyValues {XCurve {keys [<at 0 1>,<at 0.5 0.8>]}, YCurve {keys [<at 0 1>,<at 0.5 0.8>]}, ZCurve {keys [<at 0 1>,<at 0.5 0.8>]}}}
       CurveIndex=&NormalizeFloat(Value=&Divide_Float(A=$Particles.NormalizedAge, B=0.5), Min=0, Max=1)
      )
      "Float (Alpha)"=&FloatFromCurve(
       FloatCurve=DataInterface() {propertyValues {Curve {keys [
           {arriveTangent=45, interpMode=RCIM_Cubic, time=0, value=1}
           {arriveTangent=-1, interpMode=RCIM_Cubic, leaveTangent=-1, time=0.3, value=2}
           <at 1 0 RCIM_Cubic>
          ]}}}
       CurveIndex=&NormalizeFloat(Value=&Divide_Float(A=$Particles.NormalizedAge, B=0.5), Min=0, Max=1)
       Scale Curve=0.4
      )
     )
    )
    Particles.DP_01_ShockwaveSphere:Vector4f=&MakeVector4(
     X=1
     Y=&FloatFromCurve(
      FloatCurve=DataInterface() {propertyValues {Curve {keys [{interpMode=RCIM_Cubic, tangentMode=RCTM_User, time=0, value=0.25}, {interpMode=RCIM_Cubic, tangentMode=RCTM_User, time=1, value=-1}]}}}
      CurveIndex=&NormalizeFloat(Value=&Divide_Float(A=$Particles.NormalizedAge, B=0.5), Min=0, Max=1)
     )
     Z=0.5
     W=50
    )
    Particles.MeshOrientation_ShockwaveSphere:Quat4f=&MakeQuaternion(
     Quaternion From=5
     Axis=?
     Angle=?
     Yaw=?
     Pitch=?
     Roll=?
     X=?
     Y=?
     Z=?
     From=?
     To=?
     Angle Type=::NewEnumerator0
     Coordinate Space=?
     Normalize Vectors=?
    )
    Particles.Scale_ShockwaveSphere:Vector3f=&Multiply_VectorByFloat(
     Vector=<14 14 14>
     Float=&FloatFromCurve(
      FloatCurve=DataInterface() {propertyValues {Curve {keys [
          {arriveTangent=4.1785, interpMode=RCIM_Cubic, leaveTangent=4.1785, tangentMode=RCTM_User, time=0, value=0.05}
          {arriveTangent=0.9995, interpMode=RCIM_Cubic, leaveTangent=0.9995, tangentMode=RCTM_User, time=0.15, value=0.6825}
          {arriveTangent=-0.0405, interpMode=RCIM_Cubic, leaveTangent=-0.0405, tangentMode=RCTM_User, time=1, value=1}
         ]}}}
      CurveIndex=&NormalizeFloat(Value=&Divide_Float(A=$Particles.NormalizedAge, B=0.5), Min=0, Max=1)
      Scale Curve=&Divide_Float(A=$User.FXScale, B=450)
     )
    )
   ) {isSetParameters=true}
   /SetVariables(
    Particles.Color_SnapMeshImpact:LinearColor=&MultiplyLinearColor(
     Linear Color=&HueShiftLinearColor(Hue=$User.HueShift, Saturation=1, Value=1, Linear Color=<rgba 1 0 0.0333 1>)
     Scale Factor=&MakeVector4FromVectorAndFloat(
      "Vector (RGB)"=&VectorFromFloat(Value=&FloatFromCurve(
        FloatCurve=DataInterface() {propertyValues {Curve {keys [{interpMode=RCIM_Cubic, tangentMode=RCTM_User, time=0, value=1}, {interpMode=RCIM_Cubic, tangentMode=RCTM_User, time=1, value=0}]}}}
        CurveIndex=&NormalizeFloat(
         Value=&Divide_Float(A=$Particles.NormalizedAge, B=&RandomRangeFloat(Minimum=0.4, Maximum=0.6))
         Min=0
         Max=1
        )
        Scale Curve=5
       ))
      "Float (Alpha)"=&FloatFromCurve(
       FloatCurve=DataInterface() {propertyValues {Curve {keys [
           {tangentMode=RCTM_User, time=0, value=1}
           {arriveTangent=-0.0405, interpMode=RCIM_Cubic, leaveTangent=-0.0405, tangentMode=RCTM_User, time=0.7009, value=1}
           {arriveTangent=-8.0089, interpMode=RCIM_Cubic, leaveTangent=-8.0089, tangentMode=RCTM_User, time=1, value=0}
          ]}}}
       CurveIndex=&NormalizeFloat(
        Value=&Divide_Float(A=$Particles.NormalizedAge, B=&RandomRangeFloat(Minimum=0.4, Maximum=0.6))
        Min=0
        Max=1
       )
      )
     )
    )
    Particles.DP_01_SnapMeshImpact:Vector4f=&MakeVector4(
     X=&FloatFromCurve(
      FloatCurve=DataInterface() {propertyValues {Curve {keys [{arriveTangent=0.0592, interpMode=RCIM_Cubic, time=0, value=0}, <at 1 1 RCIM_Cubic>]}}}
      CurveIndex=&NormalizeFloat(
       Value=&Divide_Float(A=$Particles.NormalizedAge, B=&RandomRangeFloat(Minimum=0.4, Maximum=0.6))
       Min=0
       Max=1
      )
     )
     Y=50
     Z=1
     W=1
    )
    Particles.MeshOrientation_SnapMeshImpact:Quat4f=&MakeQuaternion(
     Quaternion From=5
     Axis=?
     Angle=?
     Yaw=?
     Pitch=?
     Roll=?
     X=?
     Y=?
     Z=?
     From=?
     To=?
     Angle Type=::NewEnumerator0
     Coordinate Space=?
     Normalize Vectors=?
    )
    Particles.Scale_SnapMeshImpact:Vector3f=&Multiply_VectorByFloat(
     Vector=&VectorFromFloat(Value=&RandomRangeFloat(Minimum=7.25, Maximum=8))
     Float=&FloatFromCurve(
      FloatCurve=DataInterface() {propertyValues {Curve {keys [{arriveTangent=2.0369, interpMode=RCIM_Cubic, leaveTangent=2.0369, tangentMode=RCTM_User, time=0, value=0.25}, <at 1 1>]}}}
      CurveIndex=&NormalizeFloat(
       Value=&Divide_Float(A=$Particles.NormalizedAge, B=&RandomRangeFloat(Minimum=0.4, Maximum=0.6))
       Min=0
       Max=1
      )
      Scale Curve=&Multiply_Float(
       A=&FN_LOD_Distance_Remap(Start Distance=0, End Distance=1500, Minimum Output=0, Maximum Output=1, Clamp Output=true)
       B=&Divide_Float(A=$User.FXScale, B=250)
      )
     )
    )
   ) {isSetParameters=true}
   /SetVariables(
    Particles.Color_RingAOE:LinearColor=&MultiplyLinearColor(
     Linear Color=&HueShiftLinearColor(Hue=$User.HueShift, Saturation=1, Value=1, Linear Color=<rgba 1 0.052 0.0836 1>)
     Scale Factor=&MakeVector4FromVectorAndFloat(
      "Vector (RGB)"=<1 1 1>
      "Float (Alpha)"=&FloatFromCurve(
       FloatCurve=DataInterface() {propertyValues {Curve {keys [{interpMode=RCIM_Cubic, tangentMode=RCTM_User, time=0.5, value=1}, {interpMode=RCIM_Cubic, tangentMode=RCTM_User, time=1, value=0}]}}}
       CurveIndex=&NormalizeFloat(Value=&Divide_Float(A=$Particles.NormalizedAge, B=0.25), Min=0, Max=1)
       Scale Curve=0.75
      )
     )
    )
    Particles.Scale_RingAOE:Vector3f=&Multiply_Vector(
     A=&Multiply_VectorByFloat(Vector=<10 10 2.5>, Float=&Divide_Float(A=$User.FXScale, B=490))
     B=&MakeVector(
      X=&FloatFromCurve(
       FloatCurve=DataInterface() {propertyValues {Curve {keys [
           {arriveTangent=4.1785, interpMode=RCIM_Cubic, leaveTangent=4.1785, tangentMode=RCTM_User, time=0, value=0}
           {arriveTangent=1.1179, interpMode=RCIM_Cubic, leaveTangent=1.1179, tangentMode=RCTM_User, time=0.2042, value=0.6228}
           {arriveTangent=0.0224, interpMode=RCIM_Cubic, leaveTangent=0.0224, tangentMode=RCTM_User, time=1, value=1}
          ]}}}
       CurveIndex=&NormalizeFloat(Value=&Divide_Float(A=$Particles.NormalizedAge, B=0.25), Min=0, Max=1)
      )
      Y=&FloatFromCurve(
       FloatCurve=DataInterface() {propertyValues {Curve {keys [
           {arriveTangent=4.1785, interpMode=RCIM_Cubic, leaveTangent=4.1785, tangentMode=RCTM_User, time=0, value=0}
           {arriveTangent=1.1179, interpMode=RCIM_Cubic, leaveTangent=1.1179, tangentMode=RCTM_User, time=0.2042, value=0.6228}
           {arriveTangent=0.0224, interpMode=RCIM_Cubic, leaveTangent=0.0224, tangentMode=RCTM_User, time=1, value=1}
          ]}}}
       CurveIndex=&NormalizeFloat(Value=&Divide_Float(A=$Particles.NormalizedAge, B=0.25), Min=0, Max=1)
      )
      Z=1
     )
    )
   ) {isSetParameters=true}
   /SetVariables(
    Particles.CameraOffset_EnergyWave=30
    Particles.Color_EnergyWave:LinearColor=&MultiplyLinearColor(
     Linear Color=&HueShiftLinearColor(Hue=$User.HueShift, Saturation=1, Value=1, Linear Color=<rgba 3 0.0652 0.3097 1>)
     Scale Factor=&MakeVector4FromVectorAndFloat(
      "Vector (RGB)"=&VectorFromFloat(Value=&FloatFromCurve(
        FloatCurve=DataInterface() {propertyValues {Curve {keys [
            {interpMode=RCIM_Cubic, tangentMode=RCTM_Break, time=0, value=10}
            {arriveTangent=-6.6072, interpMode=RCIM_Cubic, leaveTangent=-6.6072, tangentMode=RCTM_User, time=0.15, value=3}
            {interpMode=RCIM_Cubic, tangentMode=RCTM_Break, time=1, value=1}
           ]}}}
        CurveIndex=&NormalizeFloat(Value=&Divide_Float(A=$Particles.NormalizedAge, B=0.5), Min=0, Max=1)
       ))
      "Float (Alpha)"=&Multiply_Float(
       A=&FloatFromCurve(
        FloatCurve=DataInterface() {propertyValues {Curve {defaultValue=1, keys [<at 0 1>,<at 0.2388 1>,<at 1 0>]}}}
        CurveIndex=&NormalizeFloat(Value=&Divide_Float(A=$Particles.NormalizedAge, B=0.5), Min=0, Max=1)
       )
       B=&FloatFromCurve(
        FloatCurve=DataInterface() {propertyValues {Curve {keys [{interpMode=RCIM_Cubic, tangentMode=RCTM_User, time=0.5, value=1}, {interpMode=RCIM_Cubic, tangentMode=RCTM_User, time=1, value=0}]}}}
        CurveIndex=&NormalizeFloat(Value=&Divide_Float(A=$Particles.NormalizedAge, B=0.5), Min=0, Max=1)
       )
      )
     )
    )
    Particles.DP_01_EnergyWave:Vector4f=&MakeVector4(
     X=&FloatFromCurve(
      FloatCurve=DataInterface() {propertyValues {Curve {keys [{arriveTangent=-2.1015, interpMode=RCIM_Cubic, leaveTangent=-2.1015, tangentMode=RCTM_User, time=0, value=0.5}, <at 1 -0.15>]}}}
      CurveIndex=&NormalizeFloat(Value=&Divide_Float(A=$Particles.NormalizedAge, B=0.5), Min=0, Max=1)
     )
     Y=&FloatFromCurve(
      FloatCurve=DataInterface() {propertyValues {Curve {keys [{arriveTangent=-0.3464, interpMode=RCIM_Cubic, leaveTangent=-0.3464, tangentMode=RCTM_User, time=0, value=1}, <at 1 0.9>]}}}
      CurveIndex=&NormalizeFloat(Value=&Divide_Float(A=$Particles.NormalizedAge, B=0.5), Min=0, Max=1)
     )
     Z=&FloatFromCurve(
      FloatCurve=DataInterface() {propertyValues {Curve {keys [
          {interpMode=RCIM_Cubic, tangentMode=RCTM_Break, time=0, value=0}
          {arriveTangent=0.4478, interpMode=RCIM_Cubic, leaveTangent=0.4478, tangentMode=RCTM_User, time=0.15, value=1}
          {interpMode=RCIM_Cubic, tangentMode=RCTM_Break, time=1, value=1.2}
         ]}}}
      CurveIndex=&NormalizeFloat(Value=&Divide_Float(A=$Particles.NormalizedAge, B=0.5), Min=0, Max=1)
     )
     W=0
    )
    Particles.MeshOrientation_EnergyWave:Quat4f=&MakeQuaternion(
     Quaternion From=0
     Axis=<1 0 0>
     Angle=&RandomRangeFloat(Minimum=-360, Maximum=360)
     From=?
     To=?
     Angle Type=::NewEnumerator0
     Normalize Vectors=false
    )
    Particles.Scale_EnergyWave:Vector3f=&Multiply_VectorByFloat(
     Vector=<5 5 5>
     Float=&FloatFromCurve(
      FloatCurve=DataInterface() {propertyValues {Curve {keys [
          {interpMode=RCIM_Cubic, tangentMode=RCTM_Break, time=0, value=0.2}
          {arriveTangent=2.9365, interpMode=RCIM_Cubic, leaveTangent=2.9365, tangentMode=RCTM_User, time=0.1, value=1}
          {interpMode=RCIM_Cubic, tangentMode=RCTM_Break, time=1, value=2}
         ]}}}
      CurveIndex=&NormalizeFloat(Value=&Divide_Float(A=$Particles.NormalizedAge, B=0.5), Min=0, Max=1)
      Scale Curve=&Divide_Float(A=$User.FXScale, B=495)
     )
    )
   ) {isSetParameters=true}
   /SetVariables(
    Particles.CameraOffset_GroundRadialNoise=250
    Particles.Color_GroundRadialNoise:LinearColor=&MultiplyLinearColor(
     Linear Color=&HueShiftLinearColor(Hue=$User.HueShift, Saturation=1, Value=1, Linear Color=<rgba 4 0.5319 0.7053 10>)
     Scale Factor=&MakeVector4FromVectorAndFloat(
      "Vector (RGB)"=<1 1 1>
      "Float (Alpha)"=&FloatFromCurve(
       FloatCurve=DataInterface() {propertyValues {Curve {keys [<at 0 1>, {interpMode=RCIM_Cubic, tangentMode=RCTM_User, time=0.1, value=1}, {interpMode=RCIM_Cubic, tangentMode=RCTM_User, time=1, value=0}]}}}
       CurveIndex=&NormalizeFloat(Value=&Divide_Float(A=$Particles.NormalizedAge, B=0.45), Min=0, Max=1)
      )
     )
    )
    Particles.DP_01_GroundRadialNoise:Vector4f=<1 0 1 1>
    Particles.MeshOrientation_GroundRadialNoise:Quat4f=&MakeQuaternion(
     Quaternion From=0
     Axis=<0 0 1>
     Angle=&RandomRangeFloat(Minimum=-360, Maximum=360)
     From=?
     To=?
     Angle Type=::NewEnumerator0
     Normalize Vectors=false
    )
    Particles.Scale_GroundRadialNoise:Vector3f=&Multiply_VectorByFloat(
     Vector=<6 6 6>
     Float=&FloatFromCurve(
      FloatCurve=DataInterface() {propertyValues {Curve {keys [
          {arriveTangent=4.1785, interpMode=RCIM_Cubic, leaveTangent=4.1785, tangentMode=RCTM_User, time=0, value=0.05}
          {arriveTangent=0.9995, interpMode=RCIM_Cubic, leaveTangent=0.9995, tangentMode=RCTM_User, time=0.3, value=0.6}
          {arriveTangent=-0.0405, interpMode=RCIM_Cubic, leaveTangent=-0.0405, tangentMode=RCTM_User, time=1, value=1}
         ]}}}
      CurveIndex=&NormalizeFloat(Value=&Divide_Float(A=$Particles.NormalizedAge, B=0.45), Min=0, Max=1)
      Scale Curve=&Divide_Float(A=$User.FXScale, B=500)
     )
    )
   ) {isSetParameters=true}
   /SetVariables(
    Particles.Color_PlasmaSphere:LinearColor=&MultiplyLinearColor(
     Linear Color=&HueShiftLinearColor(Hue=$User.HueShift, Saturation=1, Value=1, Linear Color=<rgba 1 0.052 0.0836 1>)
     Scale Factor=&MakeVector4FromVectorAndFloat(
      "Vector (RGB)"=&VectorFromFloat(Value=&Lerp_Float(
        A=0
        B=30
        Alpha=&Multiply_Float(
         A=1
         B=&FloatFromCurve(
          FloatCurve=DataInterface() {propertyValues {Curve {keys [
              {arriveTangent=-2.6334, interpMode=RCIM_Cubic, leaveTangent=-2.6334, tangentMode=RCTM_User, time=0, value=1}
              {arriveTangent=-0.1206, interpMode=RCIM_Cubic, leaveTangent=-0.1206, tangentMode=RCTM_User, time=1, value=0}
             ]}}}
          CurveIndex=&NormalizeFloat(Value=&Divide_Float(A=$Particles.NormalizedAge, B=0.35), Min=0, Max=1)
         )
        )
       ))
      "Float (Alpha)"=&FloatFromCurve(
       FloatCurve=DataInterface() {propertyValues {Curve {keys [
           {arriveTangent=-1.3753, interpMode=RCIM_Cubic, leaveTangent=-1.3753, tangentMode=RCTM_User, time=0, value=1}
           {interpMode=RCIM_Cubic, tangentMode=RCTM_User, time=1, value=0}
          ]}}}
       CurveIndex=&NormalizeFloat(Value=&Divide_Float(A=$Particles.NormalizedAge, B=0.35), Min=0, Max=1)
      )
     )
    )
    Particles.DP_01_PlasmaSphere:Vector4f=&MakeVector4(
     X=&FloatFromCurve(
      FloatCurve=DataInterface() {propertyValues {Curve {keys [
          {interpMode=RCIM_Cubic, tangentMode=RCTM_User, time=0.25, value=0}
          {arriveTangent=2.7194, interpMode=RCIM_Cubic, leaveTangent=2.7194, tangentMode=RCTM_User, time=1, value=1}
         ]}}}
      CurveIndex=&NormalizeFloat(Value=&Divide_Float(A=$Particles.NormalizedAge, B=0.35), Min=0, Max=1)
     )
     Y=0
     Z=0
     W=0
    )
    Particles.Scale_PlasmaSphere:Vector3f=&Multiply_Vector(
     A=&Multiply_VectorByFloat(Vector=<0.3 0.3 0.05>, Float=&Divide_Float(A=$User.FXScale, B=400))
     B=&MakeVector(
      X=&FloatFromCurve(
       FloatCurve=DataInterface() {propertyValues {Curve {keys [
           {arriveTangent=1.0165, interpMode=RCIM_Cubic, leaveTangent=1.0165, tangentMode=RCTM_User, time=0, value=0}
           {arriveTangent=0.0224, interpMode=RCIM_Cubic, leaveTangent=0.0224, tangentMode=RCTM_User, time=1, value=0.75}
          ]}}}
       CurveIndex=&NormalizeFloat(Value=&Divide_Float(A=$Particles.NormalizedAge, B=0.35), Min=0, Max=1)
      )
      Y=&FloatFromCurve(
       FloatCurve=DataInterface() {propertyValues {Curve {keys [
           {arriveTangent=1.0165, interpMode=RCIM_Cubic, leaveTangent=1.0165, tangentMode=RCTM_User, time=0, value=0}
           {arriveTangent=0.0224, interpMode=RCIM_Cubic, leaveTangent=0.0224, tangentMode=RCTM_User, time=1, value=0.75}
          ]}}}
       CurveIndex=&NormalizeFloat(Value=&Divide_Float(A=$Particles.NormalizedAge, B=0.35), Min=0, Max=1)
      )
      Z=&FloatFromCurve(
       FloatCurve=DataInterface() {propertyValues {Curve {keys [
           {arriveTangent=3.1958, interpMode=RCIM_Cubic, leaveTangent=3.1958, tangentMode=RCTM_User, time=0, value=0}
           {arriveTangent=0.0224, interpMode=RCIM_Cubic, leaveTangent=0.0224, tangentMode=RCTM_User, time=1, value=1}
          ]}}}
       CurveIndex=&NormalizeFloat(Value=&Divide_Float(A=$Particles.NormalizedAge, B=0.35), Min=0, Max=1)
      )
     )
    )
   ) {isSetParameters=true}
   &ScaleColor(
    Scale Alpha=&FloatFromCurve(
     FloatCurve=DataInterface() {propertyValues {Curve {keys [
         {interpMode=RCIM_Cubic, tangentMode=RCTM_User, time=0, value=1}
         {interpMode=RCIM_Cubic, tangentMode=RCTM_User, time=0.1, value=1}
         {interpMode=RCIM_Cubic, tangentMode=RCTM_User, time=1, value=0}
        ]}}}
     CurveIndex=&NormalizeFloat(Value=&Divide_Float(A=$Particles.NormalizedAge, B=0.2), Min=0, Max=1)
     Scale Curve=&FN_Select_Float_from_Time_Of_Day(Morning=0.6, Day=0.6, Evening=0.6, Night=0.6)
    )
    ScaleA=true
    ScaleRGB=true
   )
   &ScaleMeshSize(Scale Factor=&VectorFromFloat(Value=&FloatFromCurve(
      FloatCurve=DataInterface() {propertyValues {Curve {keys [
          {arriveTangent=4.1785, interpMode=RCIM_Cubic, leaveTangent=4.1785, tangentMode=RCTM_User, time=0, value=0.2}
          {arriveTangent=0.9995, interpMode=RCIM_Cubic, leaveTangent=0.9995, tangentMode=RCTM_User, time=0.15, value=0.8}
          {arriveTangent=-0.0405, interpMode=RCIM_Cubic, leaveTangent=-0.0405, tangentMode=RCTM_User, time=1, value=1}
         ]}}}
      CurveIndex=&NormalizeFloat(Value=&Divide_Float(A=$Particles.NormalizedAge, B=0.2), Min=0, Max=1)
      Scale Curve=&Multiply_Float(A=1.5, B=&Divide_Float(A=$User.FXScale, B=1))
     )))
   &CameraOffset(Camera Offset Amount=-20)
  ]
  rendererClasses [&NiagaraMeshRendererProperties]
  renderers [
   {rendererClass=&NiagaraMeshRendererProperties, rendererIndex=0}
   {rendererClass=&NiagaraMeshRendererProperties, rendererIndex=1}
   {rendererClass=&NiagaraMeshRendererProperties, rendererIndex=2}
   {rendererClass=&NiagaraMeshRendererProperties, rendererIndex=3}
   {rendererClass=&NiagaraMeshRendererProperties, rendererIndex=4}
   {rendererClass=&NiagaraMeshRendererProperties, rendererIndex=5}
   {rendererClass=&NiagaraMeshRendererProperties, rendererIndex=6}
   {rendererClass=&NiagaraMeshRendererProperties, rendererIndex=7}
  ]
 }
 SmokeWorld {
  emitterUpdateScript [
   &EmitterState()
   &SpawnBurst_Instantaneous(Spawn Count=26, Use Spawn Probability=false)
  ]
  particleSpawnScript [
   &InitializeParticle(
    Color=&HueShiftLinearColor(Hue=$User.HueShift, Saturation=1, Value=1, Linear Color=<rgba 3 0 0.5 1>)
    Lifetime Max=0.85
    Lifetime Min=0.6
    Lifetime Mode=::NewEnumerator1
    Position Offset=<0 0 5>
    Sprite Rotation Mode=::NewEnumerator2
    Sprite Size Mode=::NewEnumerator1
    Uniform Sprite Size Max=300
    Uniform Sprite Size Min=180
    UsePositionOffset=true
    Write Material Random=true
   )
   &ShapeLocation(
    Cylinder Distribution=::NewEnumerator2
    Cylinder Height=0
    Cylinder Radius=2
    MaskSpawnGroup=false
    Rotation Coordinate Space=::World
    Shape Primitive=::NewEnumerator1
    U Position=&ReturnNormalizedExecIndex()
    UseOwnerScale=true
   )
   &AddVelocity(
    Velocity=$Particles.ShapeLocation.ShapeNormal
    Velocity Speed Scale=&Multiply_Float(
     A=&RandomRangeFloat(Minimum=1400, Maximum=1900)
     B=&Divide_Float(A=$User.FXScale, B=500)
    )
   )
   /SetVariables(Particles.RandomFloat=&RandomRangeFloat(Minimum=0, Maximum=1)) {isSetParameters=true}
  ]
  particleUpdateScript [
   &ParticleState()
   &SpriteRotationRate(Rotation Rate=&Multiply_Float(
     A=1
     B=&FloatFromCurve(
      FloatCurve=DataInterface() {propertyValues {Curve {keys [
          {arriveTangent=-4.0107, interpMode=RCIM_Cubic, leaveTangent=-4.0107, tangentMode=RCTM_User, time=-0.0046, value=1.0113}
          {arriveTangent=-0.2496, interpMode=RCIM_Cubic, leaveTangent=-0.2496, tangentMode=RCTM_User, time=0.5, value=0.16}
         ]}}}
      CurveIndex=$Emitter.Age
      Scale Curve=150
     )
    ))
   &Drag(
    Drag=&FloatFromCurve(
     FloatCurve=DataInterface() {propertyValues {Curve {keys [<at 0 2 RCIM_Cubic>,<at 0.5 1 RCIM_Cubic>]}}}
     Scale Curve=&RandomRangeFloat(Minimum=1, Maximum=1.5)
    )
    Use Linear Drag=true
    Use Rotational Drag=false
   )
   &SolveForcesAndVelocity(Clamp Velocity=false, Limit Acceleration=false)
   &ScaleSpriteSize(Scale Factor=&Vector2DFromFloat(Value=&FloatFromCurve(
      FloatCurve=DataInterface() {propertyValues {Curve {keys [{arriveTangent=1.7458, interpMode=RCIM_Cubic, leaveTangent=1.7458, tangentMode=RCTM_User, time=0, value=0.5}, <at 1 1.25 RCIM_Cubic>]}}}
      Scale Curve=&Divide_Float(A=$User.FXScale, B=500)
     )))
   &DynamicMaterialParameters(
    Index 0 Param 1=&Add_Float(
     A=$Particles.RandomFloat
     B=&FloatFromCurve(FloatCurve=DataInterface() {propertyValues {Curve {keys [
          {arriveTangent=0.9088, interpMode=RCIM_Cubic, leaveTangent=0.9088, tangentMode=RCTM_User, time=0, value=1}
          {arriveTangent=0.5473, interpMode=RCIM_Cubic, leaveTangent=0.5473, tangentMode=RCTM_User, time=1, value=2}
         ]}}})
    )
    Index 1 Param 1=&FloatFromCurve(
     FloatCurve=DataInterface() {propertyValues {Curve {keys [<at 0 0>,<at 0.25 1>]}}}
     Scale Curve=&RandomRangeFloat(Minimum=2, Maximum=1)
    )
    Index 2 Param 1=&FloatFromCurve(FloatCurve=DataInterface() {propertyValues {Curve {keys [{interpMode=RCIM_Cubic, leaveTangent=3.2309, tangentMode=RCTM_Break, time=0.1, value=0}, <at 1 1>]}}})
    Index 2 Param 2=&FloatFromCurve(
     FloatCurve=DataInterface() {propertyValues {Curve {keys [<at 0 0.2 RCIM_Cubic>,<at 1 0.8 RCIM_Cubic>]}}}
     Scale Curve=&RandomRangeFloat(Minimum=0.2, Maximum=1.2)
    )
    Index 2 Param 3=&FloatFromCurve(FloatCurve=DataInterface() {propertyValues {Curve {keys [
         {interpMode=RCIM_Cubic, tangentMode=RCTM_Break, time=0, value=1}
         {interpMode=RCIM_Cubic, leaveTangent=11.1233, tangentMode=RCTM_Break, time=1, value=2}
        ]}}})
    Index 2 Param 4=&FloatFromCurve(
     FloatCurve=DataInterface() {propertyValues {Curve {keys [<at 0 1 RCIM_Cubic>, {interpMode=RCIM_Cubic, leaveTangent=-1.2956, tangentMode=RCTM_Break, time=1, value=0.75}]}}}
     Scale Curve=0.8
    )
    Module.Param0WriteEnabled=true
    Module.Param0WriteEnabled1=false
    Module.Param0WriteEnabled2=true
    Module.Param1WriteEnabled=false
    Module.Param1WriteEnabled1=false
    Module.Param1WriteEnabled2=true
    Module.Param2WriteEnabled=false
    Module.Param2WriteEnabled1=false
    Module.Param2WriteEnabled2=true
    Module.Param3WriteEnabled=true
    Module.Param3WriteEnabled1=false
    Module.Param3WriteEnabled2=true
    Write Parameter Index 1=true
    Write Parameter Index 2=true
   )
   &CameraOffset(Camera Offset Amount=&RandomRangeFloat(Minimum=30, Maximum=50))
   &Color(Color=&HueShiftLinearColor(
     Hue=$User.HueShift
     Saturation=1
     Value=1
     Linear Color=&MultiplyLinearColor(
      Linear Color=&ColorFromCurve(ColorCurve=DataInterface() {propertyValues {
         AlphaCurve {keys [<at 0 0.5 RCIM_Cubic>,<at 1 0 RCIM_Cubic>]}
         BlueCurve {keys [<at 0 0.345 RCIM_Cubic>, {arriveTangent=-0.7802, leaveTangent=-0.7802, time=0.5389, value=0.4149}, <at 0.951 0.258>]}
         GreenCurve {keys [<at 0 0.3 RCIM_Cubic>, {arriveTangent=-0.36, leaveTangent=-0.36, time=0.5389, value=0.4149}, <at 0.951 0.2576>]}
         RedCurve {keys [<at 0 3 RCIM_Cubic>, {arriveTangent=-2.8837, leaveTangent=-2.8837, time=0.5389, value=0.4149}, <at 0.951 0.2576>]}
        }})
      Scale Factor=&MakeVector4FromVectorAndFloat(
       "Vector (RGB)"=<1 1 1>
       "Float (Alpha)"=&FN_LOD_Distance_Remap(Start Distance=0, End Distance=2500, Minimum Output=0, Maximum Output=1, Clamp Output=true)
      )
     )
    ))
  ]
  rendererClasses [&NiagaraSpriteRendererProperties]
  renderers [
   {rendererClass=&NiagaraSpriteRendererProperties}
   {rendererClass=&NiagaraSpriteRendererProperties, rendererIndex=1}
  ]
 }
 Sparks_Colored {
  emitterUpdateScript [
   &EmitterState()
   &SpawnBurst_Instantaneous(
    Spawn Count=&RandomRangeInt(Minimum=30, Maximum=45)
    Use Spawn Probability=false
    UseLoopCountLimit=false
   )
  ]
  particleSpawnScript [
   &InitializeParticle(
    Color=&HueShiftLinearColor(Hue=$User.HueShift, Saturation=1, Value=1, Linear Color=<rgba 5 0 0.4167 1>)
    Lifetime=&MakeCustomFloatFromBool(
     Bool=&SetBoolByIntComparison(A=$Particles.AmmoType, B=3, Comparison Type=::NewEnumerator4)
     True Float=&RandomRangeFloat(Minimum=0.8, Maximum=1.3)
     False Float=&RandomRangeFloat(Minimum=0.2, Maximum=1)
    )
    Position=$Particles.Position
    Position Mode=::NewEnumerator1
    Sprite Size Max=<2 3>
    Sprite Size Min=<0.5 2>
    Sprite Size Mode=::NewEnumerator2
    UsePositionOffset=false
    Write Material Random=true
   )
   &ShapeLocation(MaskSpawnGroup=false, Override Seed=false, Sphere Radius=10, UseOwnerScale=true)
   &AddVelocity(
    Cone Angle=75
    Cone Axis=<0 0 1>
    Module.UseConeFalloff=false
    Module.UseVelDistribution=false
    Override Seed=false
    Velocity Mode=::NewEnumerator2
    Velocity Speed=&Lerp_Float(
     A=750
     B=1500
     Alpha=&FloatFromCurve(
      FloatCurve=DataInterface() {propertyValues {Curve {keys [
          {interpMode=RCIM_Cubic, tangentMode=RCTM_User, time=0, value=0}
          {arriveTangent=2.6808, interpMode=RCIM_Cubic, leaveTangent=2.6808, tangentMode=RCTM_User, time=1, value=1}
         ]}}}
      CurveIndex=&RandomRangeFloat(Minimum=0, Maximum=1)
     )
    )
   )
  ]
  particleUpdateScript [
   &ParticleState()
   &GravityForce(Gravity=<0 0 -750>)
   &CurlNoiseForce(
    Bias Noise=false
    Noise Quality / Cost=::NewEnumerator0
    Noise Strength=&FloatFromCurve(
     FloatCurve=DataInterface() {propertyValues {Curve {keys [{interpMode=RCIM_Cubic, tangentMode=RCTM_Break, time=0, value=1}, {interpMode=RCIM_Cubic, tangentMode=RCTM_Break, time=1, value=0}]}}}
     Scale Curve=1000
    )
    Pan Noise=false
   )
   &Drag(
    Drag=&FloatFromCurve(
     FloatCurve=DataInterface() {propertyValues {Curve {keys [
         {interpMode=RCIM_Cubic, leaveTangent=0.4409, leaveTangentWeight=0.5018, tangentMode=RCTM_Break, time=0, value=1}
         {interpMode=RCIM_Cubic, tangentMode=RCTM_Break, time=0.2, value=0}
        ]}}}
     Scale Curve=&MakeCustomFloatFromBool(
      Bool=&SetBoolByIntComparison(A=$Particles.AmmoType, B=3, Comparison Type=::NewEnumerator4)
      True Float=12
      False Float=20
     )
    )
    Use Linear Drag=true
    Use Rotational Drag=false
   )
   &SolveForcesAndVelocity(Clamp Velocity=false, Limit Acceleration=false)
   &ScaleSpriteSize(
    Uniform Curve Scale=&FloatFromCurve(
     FloatCurve=DataInterface() {propertyValues {Curve {
        defaultValue=1
        keys [
         {arriveTangent=0.4785, interpMode=RCIM_Cubic, leaveTangent=0.4785, tangentMode=RCTM_User, time=0, value=0.8}
         {arriveTangent=0.0224, interpMode=RCIM_Cubic, leaveTangent=0.0224, tangentMode=RCTM_User, time=1, value=1}
        ]
       }}}
     CurveIndex=$Particles.DistanceScale
     Scale Curve=10
    )
    Uniform Curve Sprite Scale=DataInterface() {propertyValues {Curve {keys [{interpMode=RCIM_Cubic, tangentMode=RCTM_Break, time=0, value=1}, {interpMode=RCIM_Cubic, tangentMode=RCTM_Break, time=1, value=0.2}]}}}
   )
   &ScaleSpriteSize(
    Scale Sprite Size Mode=::NewEnumerator0
    Uniform Scale Factor=&MakeCustomFloatFromBool(
     Bool=&SetBoolByIntComparison(A=$Particles.AmmoType, B=3, Comparison Type=::NewEnumerator4)
     True Float=&MakeCustomFloatFromBool(
      Bool=&SetBoolByIntComparison(
       A=&ModuloInt(A=&ReturnParticleIDAttributes(), B=$Particles.NDC Spawn Count)
       B=16
       Comparison Type=::NewEnumerator2
      )
      True Float=1
      False Float=0
     )
     False Float=1
    )
   ) {module=ScaleSpriteSize001}
   &ScaleSpriteSizeBySpeed(Max Scale Factor=<1.25 3>)
   &ScaleColor(
    Scale Alpha=&FloatFromCurve(FloatCurve=DataInterface() {propertyValues {Curve {keys [
         {interpMode=RCIM_Cubic, tangentMode=RCTM_User, time=0.0048, value=0.438}
         {arriveTangent=0.3021, interpMode=RCIM_Cubic, leaveTangent=0.3021, tangentMode=RCTM_User, time=0.0983, value=1.5996}
         {arriveTangent=-1.0514, interpMode=RCIM_Cubic, leaveTangent=-1.0514, tangentMode=RCTM_User, time=0.3368, value=0.5383}
         {arriveTangent=-0.8117, interpMode=RCIM_Cubic, leaveTangent=-0.8117, tangentMode=RCTM_User, time=0.5811, value=1.092}
         <at 1 0>
        ]}}})
    Scale RGB=&VectorFromFloat(Value=&FloatFromCurve(FloatCurve=DataInterface() {propertyValues {Curve {keys [{arriveTangent=-52.0046, interpMode=RCIM_Cubic, leaveTangent=-52.0046, tangentMode=RCTM_User, time=0, value=16}, <at 0.8 0.125 RCIM_Cubic>]}}}))
    ScaleA=true
    ScaleRGB=true
   )
   &ScaleColor(
    Scale Alpha=&FloatFromCurve(
     FloatCurve=DataInterface() {propertyValues {Curve {
        defaultValue=1
        keys [
         {arriveTangent=0.5176, interpMode=RCIM_Cubic, leaveTangent=0.5176, tangentMode=RCTM_User, time=0, value=0.1}
         {interpMode=RCIM_Cubic, tangentMode=RCTM_User, time=0.5, value=1}
        ]
       }}}
     CurveIndex=$Particles.DistanceScale
    )
    Scale RGB=&VectorFromFloat(Value=&Lerp_Float(
      A=4
      B=1
      Alpha=&FloatFromCurve(
       FloatCurve=DataInterface() {propertyValues {Curve {keys [
           {interpMode=RCIM_Cubic, tangentMode=RCTM_Break, time=0.4, value=0}
           {arriveTangent=4.1312, interpMode=RCIM_Cubic, tangentMode=RCTM_Break, time=1, value=1}
          ]}}}
       CurveIndex=&RandomRangeFloat(Minimum=0, Maximum=1)
      )
     ))
    ScaleA=true
    ScaleRGB=true
   ) {module=ScaleColor001}
   &DynamicMaterialParameters(
    Index 0 Param 1=2
    Index 0 Param 2=20
    Index 0 Param 3=&SelectFloatFromArray(Direct Array Index=$Particles.VariantIndex, Float Selection Array=DataInterface() {propertyValues {FloatData [0.1,1,1,0.25,1,0.5,1,1]}})
    Module.Index0Float0Write=true
    Module.Index0Float1Write=true
    Module.Index0Float2Write=true
    Module.Index0Float3Write=true
   )
  ]
  rendererClasses [&NiagaraSpriteRendererProperties]
  renderers [{rendererClass=&NiagaraSpriteRendererProperties}]
 }
 Sparks_Heated {
  emitterUpdateScript [
   &EmitterState()
   &SpawnBurst_Instantaneous(
    Spawn Count=&RandomRangeInt(Minimum=30, Maximum=60)
    Use Spawn Probability=false
    UseLoopCountLimit=false
   )
  ]
  particleSpawnScript [
   &InitializeParticle(
    Color=&HueShiftLinearColor(Hue=$User.HueShift, Saturation=1, Value=1, Linear Color=<rgba 5 0.0833 0 1>)
    Lifetime=&MakeCustomFloatFromBool(
     Bool=&SetBoolByIntComparison(A=$Particles.AmmoType, B=3, Comparison Type=::NewEnumerator4)
     True Float=&RandomRangeFloat(Minimum=0.8, Maximum=1.3)
     False Float=&RandomRangeFloat(Minimum=0.2, Maximum=1)
    )
    Position=$Particles.Position
    Position Mode=::NewEnumerator1
    Sprite Size Max=<2 3>
    Sprite Size Min=<0.5 2>
    Sprite Size Mode=::NewEnumerator2
    UsePositionOffset=false
    Write Material Random=true
   )
   &ShapeLocation(MaskSpawnGroup=false, Override Seed=false, Sphere Radius=10, UseOwnerScale=true)
   &AddVelocity(
    Cone Angle=100
    Cone Axis=<0 0 1>
    Module.UseConeFalloff=false
    Module.UseVelDistribution=false
    Override Seed=false
    Velocity Mode=::NewEnumerator2
    Velocity Speed=&Lerp_Float(
     A=750
     B=2500
     Alpha=&FloatFromCurve(
      FloatCurve=DataInterface() {propertyValues {Curve {keys [
          {interpMode=RCIM_Cubic, tangentMode=RCTM_User, time=0, value=0}
          {arriveTangent=2.6808, interpMode=RCIM_Cubic, leaveTangent=2.6808, tangentMode=RCTM_User, time=1, value=1}
         ]}}}
      CurveIndex=&RandomRangeFloat(Minimum=0, Maximum=1)
     )
    )
   )
  ]
  particleUpdateScript [
   &ParticleState()
   &GravityForce()
   &CurlNoiseForce(
    Bias Noise=false
    Noise Quality / Cost=::NewEnumerator0
    Noise Strength=&FloatFromCurve(
     FloatCurve=DataInterface() {propertyValues {Curve {keys [{interpMode=RCIM_Cubic, tangentMode=RCTM_Break, time=0, value=1}, {interpMode=RCIM_Cubic, tangentMode=RCTM_Break, time=1, value=0}]}}}
     Scale Curve=1000
    )
    Pan Noise=false
   )
   &Drag(
    Drag=&FloatFromCurve(
     FloatCurve=DataInterface() {propertyValues {Curve {keys [{interpMode=RCIM_Cubic, leaveTangent=0.6329, leaveTangentWeight=0.5018, tangentMode=RCTM_Break, time=0, value=1}, <at 0.2 0 RCIM_Cubic>]}}}
     Scale Curve=&MakeCustomFloatFromBool(
      Bool=&SetBoolByIntComparison(A=$Particles.AmmoType, B=3, Comparison Type=::NewEnumerator4)
      True Float=12
      False Float=20
     )
    )
    Use Linear Drag=true
    Use Rotational Drag=false
   )
   &SolveForcesAndVelocity(Clamp Velocity=false, Limit Acceleration=false)
   &ScaleSpriteSize(
    Uniform Curve Scale=&FloatFromCurve(
     FloatCurve=DataInterface() {propertyValues {Curve {
        defaultValue=1
        keys [
         {arriveTangent=0.4785, interpMode=RCIM_Cubic, leaveTangent=0.4785, tangentMode=RCTM_User, time=0, value=0.8}
         {arriveTangent=0.0224, interpMode=RCIM_Cubic, leaveTangent=0.0224, tangentMode=RCTM_User, time=1, value=1}
        ]
       }}}
     CurveIndex=$Particles.DistanceScale
     Scale Curve=10
    )
    Uniform Curve Sprite Scale=DataInterface() {propertyValues {Curve {keys [{interpMode=RCIM_Cubic, tangentMode=RCTM_Break, time=0, value=1}, {interpMode=RCIM_Cubic, tangentMode=RCTM_Break, time=1, value=0.2}]}}}
   )
   &ScaleSpriteSize(
    Scale Sprite Size Mode=::NewEnumerator0
    Uniform Scale Factor=&MakeCustomFloatFromBool(
     Bool=&SetBoolByIntComparison(A=$Particles.AmmoType, B=3, Comparison Type=::NewEnumerator4)
     True Float=&MakeCustomFloatFromBool(
      Bool=&SetBoolByIntComparison(
       A=&ModuloInt(A=&ReturnParticleIDAttributes(), B=$Particles.NDC Spawn Count)
       B=16
       Comparison Type=::NewEnumerator2
      )
      True Float=1
      False Float=0
     )
     False Float=1
    )
   ) {module=ScaleSpriteSize001}
   &ScaleSpriteSizeBySpeed(Max Scale Factor=<1 5>)
   &ScaleColor(
    Scale Alpha=&FloatFromCurve(FloatCurve=DataInterface() {propertyValues {Curve {keys [
         {interpMode=RCIM_Cubic, tangentMode=RCTM_User, time=0.0048, value=0.438}
         {arriveTangent=0.3021, interpMode=RCIM_Cubic, leaveTangent=0.3021, tangentMode=RCTM_User, time=0.0983, value=1.5996}
         {arriveTangent=-1.0514, interpMode=RCIM_Cubic, leaveTangent=-1.0514, tangentMode=RCTM_User, time=0.3368, value=0.5383}
         {arriveTangent=-0.8117, interpMode=RCIM_Cubic, leaveTangent=-0.8117, tangentMode=RCTM_User, time=0.5811, value=1.092}
         <at 1 0>
        ]}}})
    Scale RGB=&MakeVectorFromLinearColorRGB(Color=&ColorFromCurve(
      ColorCurve=DataInterface() {propertyValues {
        AlphaCurve {keys [<at 0.0007 1>]}
        BlueCurve {keys [
          <at 0 0 RCIM_Cubic>
          {
           arriveTangent=0.0609
           arriveTangentWeight=0.1848
           interpMode=RCIM_Cubic
           leaveTangentWeight=0.1637
           tangentMode=RCTM_Break
           time=0.239
           value=0.0016
          }
          {interpMode=RCIM_Cubic, leaveTangent=1.4223, time=1, value=0}
         ]}
        GreenCurve {keys [
          <at 0 0 RCIM_Cubic>
          {arriveTangent=1.9854, interpMode=RCIM_Cubic, leaveTangent=2.1986, tangentMode=RCTM_Break, time=0.239, value=0.2212}
          {interpMode=RCIM_Cubic, leaveTangent=0.7462, time=1, value=0.6}
         ]}
        RedCurve {keys [
          <at 0 0.0417 RCIM_Cubic>
          {arriveTangent=5.5085, interpMode=RCIM_Cubic, leaveTangent=0.6674, tangentMode=RCTM_Break, time=0.239, value=0.9}
          <at 1 1 RCIM_Cubic>
         ]}
       }}
      CurveIndex=&OneMinusFloat(Float=$Particles.NormalizedAge)
     ))
    ScaleA=true
    ScaleRGB=true
   )
   &ScaleColor(
    Scale Alpha=&FloatFromCurve(
     FloatCurve=DataInterface() {propertyValues {Curve {
        defaultValue=1
        keys [
         {arriveTangent=0.0664, interpMode=RCIM_Cubic, leaveTangent=0.0664, tangentMode=RCTM_User, time=0, value=0.1}
         {interpMode=RCIM_Cubic, tangentMode=RCTM_User, time=0.5, value=1}
        ]
       }}}
     CurveIndex=$Particles.DistanceScale
    )
    Scale RGB=&VectorFromFloat(Value=10)
    ScaleA=true
    ScaleRGB=true
   ) {module=ScaleColor001}
   &DynamicMaterialParameters(
    Index 0 Param 1=2
    Index 0 Param 2=20
    Index 0 Param 3=20
    Module.Index0Float0Write=true
    Module.Index0Float1Write=true
    Module.Index0Float2Write=true
    Module.Index0Float3Write=true
   )
  ]
  rendererClasses [&NiagaraSpriteRendererProperties]
  renderers [{rendererClass=&NiagaraSpriteRendererProperties}]
 }
 Speedlines {
  emitterUpdateScript [
   &EmitterState()
   &SpawnBurst_Instantaneous(
    Spawn Count=&Multiply_Int(
     A=&RandomRangeInt(Minimum=2, Maximum=3)
     B=&MakeIntFromFloat(Float=&FN_LOD_Distance_Remap(Start Distance=0, End Distance=750, Minimum Output=1, Maximum Output=10, Clamp Output=true))
    )
    Use Spawn Probability=false
   )
  ]
  particleSpawnScript [
   /Niagara/Modules/Spawn/Initialization/InitializeParticle(
    Color=&HueShiftLinearColor(Hue=$User.HueShift, Saturation=1, Value=1, Linear Color=<rgba 1 0.1362 0.1506 0.25>)
    Lifetime=&RandomRangeFloat(Minimum=0.3, Maximum=0.4, Evaluation Type=::NewEnumerator1)
    Module.Write Color=true
    Module.Write Lifetime=true
    Module.Write Mass=false
    Module.Write Material Random=true
    Module.Write Position=false
    Module.Write Scale=false
    Module.Write SpriteRotation=false
    Module.Write SpriteSize=true
    Sprite Size=&RandomRangeVector2D(Minimum=<4 18>, Maximum=<8 24>, Evaluation Type=::NewEnumerator1)
   )
   /Niagara/Modules/Spawn/Location/SphereLocation(
    MaskSpawnGroup=false
    Module.Override Randomness=false
    Sphere Radius=&Multiply_Float(
     A=&FN_LOD_Distance_Remap(Start Distance=0, End Distance=1500, Minimum Output=0, Maximum Output=1, Clamp Output=true)
     B=&Divide_Float(A=$User.FXScale, B=10)
    )
    Surface Only=true
   )
   &AddVelocityFromPoint(
    Module.Use Falloff Distance=false
    Velocity Strength=&Multiply_Float(
     A=&RandomRangeFloat(Minimum=600, Maximum=1800)
     B=&FN_LOD_Distance_Remap(Start Distance=0, End Distance=1500, Minimum Output=0, Maximum Output=1, Clamp Output=true)
    )
   )
  ]
  particleUpdateScript [
   &ParticleState()
   &ScaleColor(
    Scale Alpha=&FloatFromCurve(FloatCurve=DataInterface() {propertyValues {Curve {keys [<at 0 1>,<at 1 0>]}}}, Scale Curve=0.5)
    Scale RGB=<3 3 3>
    ScaleA=true
    ScaleRGB=true
   )
   &Drag(Use Linear Drag=true, Use Rotational Drag=false)
   &ScaleSpriteSize(Scale Factor=&Vector2DFromCurve(
     Vector2Curve=DataInterface() {propertyValues {XCurve {keys [<at 0 1>,<at 1 0>]}, YCurve {keys [<at 0 8>,<at 1 4>]}}}
     ScaleCurve=&MakeVector2D(
      X=&FN_LOD_Distance_Remap(Start Distance=0, End Distance=750, Minimum Output=0.5, Maximum Output=1, Clamp Output=true)
      Y=&FN_LOD_Distance_Remap(Start Distance=0, End Distance=750, Minimum Output=0.1, Maximum Output=1, Clamp Output=true)
     )
    ))
   &SolveForcesAndVelocity(Clamp Velocity=false, Limit Acceleration=false)
  ]
  rendererClasses [&NiagaraSpriteRendererProperties]
  renderers [{rendererClass=&NiagaraSpriteRendererProperties}]
 }
}
summary {
 systemName=NS_Explosion_PiewPistol_ADS
 emitters [
  {emitterName=Speedlines, rendererClasses [&NiagaraSpriteRendererProperties]}
  {emitterName=SmokeWorld, rendererClasses [&NiagaraSpriteRendererProperties]}
  {emitterName=Electricity, rendererClasses [&NiagaraMeshRendererProperties]}
  {emitterName=Impact_Streaks, rendererClasses [&NiagaraSpriteRendererProperties]}
  {
   emitterName=PackExplosionFX
   rendererClasses [
    &NiagaraMeshRendererProperties
    &NiagaraSpriteRendererProperties
    &NiagaraLightRendererProperties
   ]
  }
  {emitterName=PackRadialFX, rendererClasses [&NiagaraMeshRendererProperties]}
  {emitterName=Sparks_Colored, rendererClasses [&NiagaraSpriteRendererProperties]}
  {emitterName=Sparks_Heated, rendererClasses [&NiagaraSpriteRendererProperties]}
  {emitterName=Decal, rendererClasses [&NiagaraDecalRendererProperties]}
 ]
 userVariables [
  User.FXScale {type=NiagaraFloat, defaultValue=200, description=""}
  User.HueShift {type=NiagaraFloat, defaultValue=0.025, description=""}
  User.NormalImpact {type=Vector3f, description="", defaultValue {struct=Vector3f, value=<0 0 0>}}
 ]
}
