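# Niagara muzzle flash for the Piew pistol, named as the aim-down-sights (ADS) variant.
# CombinedElements packs core flash, comets, energy radial, flare, glows and light into one burst emitter with multiple sprite, mesh and light renderers.
# Emitter variables animate those layers through per-element normalized ages, curves and visibility flags.
# MuzzleStreaks adds a brief cone burst of SubUV sprites, and User.HueShift tints both emitters.
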NiagaraSystem /PiewPistolGameplay/Assets/FX/NS_MF_PiewPistol_ADS
dependencies {
 usedDataInterfaces [
  &NiagaraDataInterfaceVectorCurve
  &NiagaraDataInterfaceCurve
  &NiagaraDataInterfaceColorCurve
  &NiagaraDataInterfaceVector2DCurve
  &NiagaraDataInterfaceSpriteRendererInfo
 ]
 usedDynamicInputs []
 usedModules [&SystemState]
}
emitters {
 CombinedElements {
  emitterSpawnScript [/SetVariables(
    Emitter.CoreMuzzleFlashPosition:NiagaraPosition=<1 0 0>
    Emitter.EnergyRadialDynamic0:Vector4f=<0.5 0 1 2>
    Emitter.EnergyRadialDynamic1:Vector4f=&MakeVector4(
     X=&RandomRangeFloat(Minimum=-1, Maximum=1)
     Y=&RandomRangeFloat(Minimum=-1, Maximum=1)
     Z=0
     W=0
    )
    Emitter.EnergyRadialDynamic2:Vector4f=<1 1 1 1>
    Emitter.EnergyRadialDynamic3:Vector4f=<10 0 0 0>
    Emitter.FlareCameraOffset=20
    Emitter.FlarePosition:NiagaraPosition=<5 0 0>
    Emitter.FlareRotation:NiagaraFloat=&RandomRangeFloat(Minimum=0, Maximum=360)
    Emitter.FlareSpriteSize:Vector2f=<120 120>
    Emitter.LightPosition:NiagaraPosition=<0 5 0>
    Emitter.LightRadius:NiagaraFloat=&RandomRangeFloat(Minimum=133, Maximum=250)
    Emitter.LongCometOrientation:Quat4f=<0 1 1 0>
    Emitter.SoftGlowCameraOffset=5
    Emitter.SoftGlowSpriteSize:Vector2f=&Vector2DFromFloat(Value=&RandomRangeFloat(Minimum=35, Maximum=65))
    Emitter.WideCometMeshScale:Vector3f=<0.1 0.23 0.23>
   ) {isSetParameters=true}]
  emitterUpdateScript [
   &EmitterState(Inactive Response=::NewEnumerator1, Life Cycle Mode=::NewEnumerator1, Loop Behavior=::NewEnumerator1, Loop Duration=0.25, UseLoopDelay=false)
   &SpawnBurst_Instantaneous(Use Spawn Probability=false, UseLoopCountLimit=false)
   /SetVariables(
    Emitter.CoreMuzzleFlashNormalizedAge:NiagaraFloat=&RemapRange(Input Value To Remap=$Emitter.Age, Input Min=0, Input Max=0.1, Output Min=0, Output Max=1)
    Emitter.EnergyRadialNormalizedAge:NiagaraFloat=&RemapRange(
     Input Value To Remap=$Emitter.Age
     Input Min=0
     Input Max=&RandomRangeFloat(Minimum=0.2, Maximum=0.25)
     Output Min=0
     Output Max=1
    )
    Emitter.FlareNormalizedAge:NiagaraFloat=&RemapRange(Input Value To Remap=$Emitter.Age, Input Min=0, Input Max=0.125, Output Min=0, Output Max=1)
    Emitter.LightNormalizedAge:NiagaraFloat=&RemapRange(Input Value To Remap=$Emitter.Age, Input Min=0, Input Max=0.18, Output Min=0, Output Max=1)
    Emitter.LongCometNormalizedAge:NiagaraFloat=&RemapRange(
     Input Value To Remap=$Emitter.Age
     Input Min=0
     Input Max=&RandomRangeFloat(Minimum=0.125, Maximum=0.15)
     Output Min=0
     Output Max=1
    )
    Emitter.SoftGlowNormalizedAge:NiagaraFloat=&RemapRange(Input Value To Remap=$Emitter.Age, Input Min=0, Input Max=0.25, Output Min=0, Output Max=1)
    Emitter.WideCometNormalizedAge:NiagaraFloat=&RemapRange(Input Value To Remap=$Emitter.Age, Input Min=0, Input Max=0.15, Output Min=0, Output Max=1)
   ) {isSetParameters=true}
   /SetVariables(
    Emitter.CoreMuzzleFlashColor:LinearColor=&MultiplyLinearColor(
     Linear Color=&HueShiftLinearColor(Hue=$User.HueShift, Saturation=1, Value=1, Linear Color=<rgba 1 0.0001 0.0001 1>)
     Scale Factor=&MakeFloat4FromVectorAndFloat(
      "Vector (RGB)"=&VectorFromCurve(
       VectorCurve=DataInterface() {propertyValues {XCurve {keys [<at 0.7 12>,<at 1 1>]}, YCurve {keys [<at 0.7 12>,<at 1 1>]}, ZCurve {keys [<at 0.7 12>,<at 1 1>]}}}
       CurveIndex=$Emitter.CoreMuzzleFlashNormalizedAge
      )
      "Float (Alpha)"=1
     )
    )
    Emitter.CoreMuzzleFlashMeshScale:Vector3f=&Multiply_VectorByFloat(
     Vector=&UniformRangedVector(Minimum=<0.6 0.2 0.2>, Maximum=<0.6 0.2 0.2>)
     Float=&FloatFromCurve(
      FloatCurve=DataInterface() {propertyValues {Curve {keys [{arriveTangent=2.2457, interpMode=RCIM_Cubic, leaveTangent=2.2457, tangentMode=RCTM_User, time=0, value=0}, <at 1 1>]}}}
      CurveIndex=$Emitter.CoreMuzzleFlashNormalizedAge
     )
    )
    Emitter.EnergyRadialColor:LinearColor=&MakeLinearColorFromVectorAndFloat(
     "Vector (RGB)"=&MakeVectorFromLinearColorRGB(Color=&HueShiftLinearColor(
       Hue=$User.HueShift
       Saturation=1
       Value=1
       Linear Color=&LerpLinearColors(
        StartColor=<rgba 1 0.2 0.2 1>
        EndColor=<rgba 5 0 0.0833 1>
        LerpFactor=&FloatFromCurve(
         FloatCurve=DataInterface() {propertyValues {Curve {keys [{interpMode=RCIM_Cubic, tangentMode=RCTM_User, time=0.2, value=1}, {interpMode=RCIM_Cubic, tangentMode=RCTM_User, time=0.4, value=0}]}}}
         CurveIndex=$Emitter.EnergyRadialNormalizedAge
        )
       )
      ))
     "Float (Alpha)"=&FloatFromCurve(
      FloatCurve=DataInterface() {propertyValues {Curve {keys [
          {arriveTangent=9.0725, interpMode=RCIM_Cubic, leaveTangent=9.0725, tangentMode=RCTM_User, time=0, value=0}
          {arriveTangent=0.4061, interpMode=RCIM_Cubic, leaveTangent=0.4061, time=0.1408, value=0.9847}
          {arriveTangent=-0.3443, interpMode=RCIM_Cubic, leaveTangent=-0.3443, tangentMode=RCTM_User, time=0.4365, value=0.1773}
          {arriveTangent=-0.3185, interpMode=RCIM_Cubic, leaveTangent=-0.3185, tangentMode=RCTM_User, time=1, value=0}
         ]}}}
      CurveIndex=$Emitter.EnergyRadialNormalizedAge
     )
    )
    Emitter.EnergyRadialSpriteScale:Vector2f=&Vector2DFromFloat(Value=&FloatFromCurve(
      FloatCurve=DataInterface() {propertyValues {Curve {keys [
          {arriveTangent=1.2221, interpMode=RCIM_Cubic, leaveTangent=1.2221, tangentMode=RCTM_User, time=0, value=0.25}
          {interpMode=RCIM_Cubic, leaveTangent=0.3375, time=1, value=1}
         ]}}}
      CurveIndex=$Emitter.EnergyRadialNormalizedAge
      Scale Curve=&RandomRangeFloat(Minimum=45, Maximum=60)
     ))
    Emitter.FlareColor:LinearColor=&MultiplyLinearColor(
     Linear Color=&HueShiftLinearColor(Hue=$User.HueShift, Saturation=1, Value=1, Linear Color=<rgba 8 0 0.1333 0.5>)
     Scale Factor=&MakeVector4_FromFloat(Float=&FloatFromCurve(
       FloatCurve=DataInterface() {propertyValues {Curve {keys [
           {arriveTangent=-2.0731, interpMode=RCIM_Cubic, leaveTangent=-2.0731, tangentMode=RCTM_User, time=0, value=1}
           {interpMode=RCIM_Cubic, tangentMode=RCTM_User, time=1, value=0}
          ]}}}
       CurveIndex=$Emitter.FlareNormalizedAge
      ))
    )
    Emitter.LightColor:LinearColor=&MultiplyLinearColor(
     Linear Color=&HueShiftLinearColor(Hue=$User.HueShift, Saturation=1, Value=1, Linear Color=<rgba 10 2 2.1333 0.05>)
     Scale Factor=&MakeFloat4FromVectorAndFloat(
      "Vector (RGB)"=<1 1 1>
      "Float (Alpha)"=&FloatFromCurve(
       FloatCurve=DataInterface() {propertyValues {Curve {keys [<at 0 0 RCIM_Cubic>,<at 0.5 1 RCIM_Cubic>,<at 1 0 RCIM_Cubic>]}}}
       CurveIndex=$Emitter.SoftGlowNormalizedAge
       Scale Curve=4
      )
     )
    )
    Emitter.LongCometColor:LinearColor=&MultiplyLinearColor(
     Linear Color=&HueShiftLinearColor(Hue=$User.HueShift, Saturation=1, Value=1, Linear Color=<rgba 5 0 0.0833 1>)
     Scale Factor=&MakeFloat4FromVectorAndFloat(
      "Vector (RGB)"=&VectorFromCurve(
       VectorCurve=DataInterface() {propertyValues {XCurve {keys [<at 0 2>,<at 1 0.985>]}, YCurve {keys [<at 0 2>,<at 1 0.6599>]}, ZCurve {keys [<at 0 2>,<at 1 1>]}}}
       CurveIndex=$Emitter.LongCometNormalizedAge
      )
      "Float (Alpha)"=1
     )
    )
    Emitter.LongCometMeshScale:Vector3f=&VectorFromCurve(
     VectorCurve=DataInterface() {propertyValues {XCurve {keys [<at 0 1>,<at 1 1.5>]}, YCurve {keys [<at 0 1.1>,<at 1 0.9>]}, ZCurve {keys [<at 0 1.1>,<at 1 0.9>]}}}
     CurveIndex=$Emitter.LongCometNormalizedAge
     Scale Curve=<0.2 0.1 0.1>
    )
    Emitter.SoftGlowColor:LinearColor=&MultiplyLinearColor(
     Linear Color=&HueShiftLinearColor(Hue=$User.HueShift, Saturation=1, Value=1, Linear Color=<rgba 1 0 0.0167 0.1>)
     Scale Factor=&MakeFloat4FromVectorAndFloat(
      "Vector (RGB)"=<1 1 1>
      "Float (Alpha)"=&FloatFromCurve(
       FloatCurve=DataInterface() {propertyValues {Curve {keys [<at 0 0 RCIM_Cubic>,<at 0.5 1 RCIM_Cubic>,<at 1 0 RCIM_Cubic>]}}}
       CurveIndex=$Emitter.SoftGlowNormalizedAge
      )
     )
    )
    Emitter.WideCometColor:LinearColor=&MultiplyLinearColor(
     Linear Color=&HueShiftLinearColor(Hue=$User.HueShift, Saturation=1, Value=1, Linear Color=<rgba 3 0 0.05 1>)
     Scale Factor=&MakeFloat4FromVectorAndFloat(
      "Vector (RGB)"=&VectorFromCurve(
       VectorCurve=DataInterface() {propertyValues {XCurve {keys [<at 0 2>,<at 1 0.985>]}, YCurve {keys [<at 0 2>,<at 1 0.6599>]}, ZCurve {keys [<at 0 2>,<at 1 1>]}}}
       CurveIndex=$Emitter.WideCometNormalizedAge
      )
      "Float (Alpha)"=&FloatFromCurve(
       FloatCurve=DataInterface() {propertyValues {Curve {keys [<at 0 0 RCIM_Cubic>,<at 0.2 1 RCIM_Cubic>,<at 1 0>]}}}
       CurveIndex=$Emitter.WideCometNormalizedAge
       Scale Curve=0.3
      )
     )
    )
   ) {isSetParameters=true}
   /SetVariables(
    Emitter.EnergyRadialVisibility:NiagaraBool=&SetBoolByFloatComparison(A=$Emitter.EnergyRadialNormalizedAge, B=1, Comparison Type=::NewEnumerator6)
    Emitter.LongCometVisibility:NiagaraBool=&SetBoolByFloatComparison(A=$Emitter.LongCometNormalizedAge, B=1, Comparison Type=::NewEnumerator6)
    Emitter.ShortParticlesVisibility:NiagaraBool=&SetBoolByFloatComparison(A=$Emitter.Age, B=0.1, Comparison Type=::NewEnumerator7)
    Emitter.WideCometVisibility:NiagaraBool=&SetBoolByFloatComparison(A=$Emitter.WideCometNormalizedAge, B=1, Comparison Type=::NewEnumerator6)
   ) {isSetParameters=true}
  ]
  particleSpawnScript [&InitializeParticle(Color Mode=::NewEnumerator0, UsePositionOffset=false, Write Material Random=true)]
  particleUpdateScript [
   &ParticleState()
   /SetVariables(
    Particles.CameraOffsetGlow=20
    Particles.ColorGlow:LinearColor=&HueShiftLinearColor(
     Hue=$User.HueShift
     Saturation=1
     Value=1
     Linear Color=&ColorFromCurve(ColorCurve=DataInterface() {propertyValues {
        AlphaCurve {keys [<at 0 0>,<at 0.1122 1>,<at 0.3117 0.4>,<at 0.9535 0 RCIM_Cubic>]}
        BlueCurve {keys [<at 0 0.2137 RCIM_Cubic>,<at 0.3521 0.035 RCIM_Cubic>]}
        GreenCurve {keys [<at 0 0.175 RCIM_Cubic>,<at 0.3521 0.035 RCIM_Cubic>]}
        RedCurve {keys [<at 0 2.5 RCIM_Cubic>,<at 0.3521 0.5 RCIM_Cubic>]}
       }})
    )
    Particles.DynParGlow:Vector4f=&MakeVector4(X=&Multiply_Float(A=$Engine.System.Age, B=0.5), Y=0, Z=0, W=0)
    Particles.SpriteSizeGlow:Vector2f=&Multiply_Vector2D_ByFloat(
     Float=&FloatFromCurve(FloatCurve=DataInterface() {propertyValues {Curve {keys [<at 0 0.8>,<at 1 1>]}}})
     Vector2D=&Vector2DFromFloat(Value=40)
    )
   ) {isSetParameters=true}
  ]
  rendererClasses [
   &NiagaraSpriteRendererProperties
   &NiagaraLightRendererProperties
   &NiagaraMeshRendererProperties
  ]
  renderers [
   {rendererClass=&NiagaraSpriteRendererProperties}
   {rendererClass=&NiagaraSpriteRendererProperties, rendererIndex=1}
   {rendererClass=&NiagaraLightRendererProperties, rendererIndex=2}
   {rendererClass=&NiagaraMeshRendererProperties, rendererIndex=3}
   {rendererClass=&NiagaraMeshRendererProperties, rendererIndex=4}
   {rendererClass=&NiagaraSpriteRendererProperties, rendererIndex=5}
   {rendererClass=&NiagaraMeshRendererProperties, rendererIndex=6}
   {rendererClass=&NiagaraSpriteRendererProperties, rendererIndex=7}
  ]
 }
 MuzzleStreaks {
  emitterUpdateScript [
   &EmitterState(Inactive Response=::NewEnumerator1, Life Cycle Mode=::NewEnumerator1, Loop Behavior=::NewEnumerator1, Loop Duration=0.1, UseLoopDelay=false)
   &SpawnBurst_Instantaneous(Spawn Count=5, Spawn Time=0.02, Use Spawn Probability=false, UseLoopCountLimit=false)
  ]
  particleSpawnScript [
   &InitializeParticle(
    Color=&HueShiftLinearColor(Hue=$User.HueShift, Saturation=1, Value=1, Linear Color=<rgba 1 0 0.0167 0.5>)
    Lifetime Max=0.2
    Lifetime Min=0.1
    Lifetime Mode=::NewEnumerator1
    Position Offset=<-3 0 0>
    Sprite Size=&RandomRangeVector2D(Minimum=<7 11>, Maximum=<10 15>)
    Sprite Size Mode=::NewEnumerator4
    UsePositionOffset=true
    Write Material Random=true
   )
   &AddVelocity(
    Cone Angle=60
    Module.UseConeFalloff=false
    Module.UseVelDistribution=false
    Override Seed=false
    Velocity Mode=::NewEnumerator2
    Velocity Speed=&RandomRangeFloat(Minimum=10, Maximum=30)
   )
   &ApplyInitialForces()
   &AddVelocity(Velocity=&RandomRangeVector(Minimum=<3 0 0>, Maximum=<20 0 0>)) {module=AddVelocity001}
  ]
  particleUpdateScript [
   &ParticleState()
   &ScaleSpriteSize(
    Scale Factor=&Vector2DFromCurve(
     Vector2Curve=DataInterface() {propertyValues {
       XCurve {keys [{arriveTangent=0.5317, interpMode=RCIM_Cubic, time=-0.0045, value=1}, <at 1 0.7>]}
       YCurve {keys [{arriveTangent=3.7386, interpMode=RCIM_Cubic, time=-0.0045, value=0.0371}, <at 1 2>]}
      }}
     ScaleCurve=<1 1>
    )
    Scale Sprite Size Mode=::NewEnumerator2
   )
   &SolveForcesAndVelocity(Clamp Velocity=false, Limit Acceleration=false)
   &SubUVAnimation(
    "Sprite Renderer "=DataInterface() {propertyValues {SpriteRenderer=&self:MuzzleStreaks_3.NiagaraSpriteRendererProperties_1}}
    UseEndFrame=false
    UseStartFrame=false
   )
   &ScaleColor(
    Scale Alpha=&FloatFromCurve(FloatCurve=DataInterface() {propertyValues {Curve {keys [<at 0 1 RCIM_Cubic>,<at 1 0>]}}})
    Scale RGB=&VectorFromCurve(VectorCurve=DataInterface() {propertyValues {XCurve {keys [<at 0 50 RCIM_Cubic>,<at 1 5>]}, YCurve {keys [<at 0 50 RCIM_Cubic>,<at 1 5>]}, ZCurve {keys [<at 0 50 RCIM_Cubic>,<at 1 5>]}}})
    ScaleA=true
    ScaleRGB=true
   )
   &KillParticles(Kill Particles=&SetBoolByFloatComparison(A=$Emitter.Age, B=0.11))
  ]
  rendererClasses [&NiagaraSpriteRendererProperties]
  renderers [{rendererClass=&NiagaraSpriteRendererProperties}]
 }
}
summary {
 systemName=NS_MF_PiewPistol_ADS
 emitters [
  {
   emitterName=CombinedElements
   rendererClasses [
    &NiagaraSpriteRendererProperties
    &NiagaraLightRendererProperties
    &NiagaraMeshRendererProperties
   ]
  }
  {emitterName=MuzzleStreaks, rendererClasses [&NiagaraSpriteRendererProperties]}
 ]
 userVariables [User.HueShift {type=NiagaraFloat, defaultValue=0.025, description=""}]
}
