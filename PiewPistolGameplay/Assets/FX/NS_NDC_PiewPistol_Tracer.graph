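# Niagara system that draws PiewPistol projectile tracers from Fortnite projectile data read through a Niagara Data Channel.
# Only the Tracer emitter is enabled, spawning red HDR sprites sized by scale arrays, velocity and LOD distance.
# It kills start points, limits mid-flight segments to momentary sprites, and animates end segments from previous to final position, with separate hitscan timing.
# Bullet, super-shot tracer, lightning-arc mesh and debug emitters are present but disabled.
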NiagaraSystem /PiewPistolGameplay/Assets/FX/NS_NDC_PiewPistol_Tracer
dependencies {
 usedDataInterfaces [
  &NiagaraDataInterfaceCurve
  &NiagaraDataInterfaceArrayFloat
  /Script/Niagara.NiagaraDataInterfaceDataChannelRead
 ]
 usedDynamicInputs [&SelectFloatFromArray]
 usedModules [
  &SystemState
  &self:SystemSpawnScript.SystemScriptSource.SystemScriptGraph.NiagaraNodeAssignment_0.SetVariables_#1
  /Niagara/Modules/System/CompleteIfUnused
 ]
}
emitters {
 Bullet {
  bEnabled=false
  emitterSpawnScript [/SetVariables(Emitter.UseTeardrop=true) {isSetParameters=true}]
  emitterUpdateScript [
   &EmitterState()
   null() {enabled=false, module=Read_NDC_PiewPistol_Tracer_01}
   /SetVariables(
    Emitter.bIsNormalShot:NiagaraBool=&InvertBool(Boolean:NiagaraBool=$Emitter.Read_NDC_PiewPistol_Tracer_01.bIsSuperShot)
    Emitter.bIsSuperShot:NiagaraBool=$Emitter.Read_NDC_PiewPistol_Tracer_01.bIsSuperShot
   ) {isSetParameters=true}
   /Game/Effects/Niagara/DataChannels/Projectiles/Spawn_NDCFortProjectiles(Max Spawn Count=1, Min Spawn Count=1, Reader:NiagaraDataInterfaceDataChannelRead=$System.NDCReader)
  ]
  particleSpawnScript [
   /Game/Effects/Niagara/DataChannels/Projectiles/InitializeParticleFromNDC_FortProjectiles(FallbackAmmoType=2)
   /SetVariables(Particles.PositionType:NiagaraInt32=$Output.InitializeParticleFromNDC_FortProjectiles.PositionType) {isSetParameters=true}
   /SetVariables(Particles.Random=&RandomRangeFloat(
     Minimum=0
     Maximum=1
     Randomness Mode=::NewEnumerator1
     Random Seed=&ReturnParticleIDAttributes(Particle ID Attribute=::NewEnumerator1, Niagara ID=$Particles.RibbonID)
     Fixed Random Seed=true
    )) {isSetParameters=true}
   /SetVariables(Particles.LOD Distance:NiagaraFloat=$Output.InitializeParticleFromNDC_FortProjectiles.LODDistance) {isSetParameters=true}
   /SetVariables(Particles.IsStartPosition:NiagaraBool=&SetBoolByIntComparison(A=$Particles.PositionType, B=0, Comparison Type=::NewEnumerator4)) {isSetParameters=true}
   /SetVariables(Particles.IsEndPosition:NiagaraBool=&SetBoolByIntComparison(A=$Particles.PositionType, B=2, Comparison Type=::NewEnumerator4)) {isSetParameters=true}
   /SetVariables(Particles.OriginTime:NiagaraFloat=$Output.InitializeParticleFromNDC_FortProjectiles.OriginTime) {isSetParameters=true}
   /SetVariables(Particles.ProjectileAge:NiagaraFloat=&Subtract_Float(A=$Engine.Time, B=$Particles.OriginTime)) {isSetParameters=true}
   /SetVariables(Particles.MuzzlePosition:NiagaraPosition=$Output.InitializeParticleFromNDC_FortProjectiles.OriginPosition) {isSetParameters=true}
   /SetVariables(Particles.PreviousPosition:NiagaraPosition=$Output.InitializeParticleFromNDC_FortProjectiles.PrevPosition) {isSetParameters=true}
   /SetVariables(Particles.TargetPosition:NiagaraPosition=$Output.InitializeParticleFromNDC_FortProjectiles.Position) {isSetParameters=true}
   /SetVariables(Particles.RibbonUVDistance:NiagaraFloat=&DistanceBetweenPositions(A=$Particles.TargetPosition, B=$Particles.MuzzlePosition)) {isSetParameters=true}
   /SetVariables(Particles.Distance:NiagaraFloat=&DistanceBetweenPositions(A=$Particles.PreviousPosition, B=$Particles.TargetPosition)) {isSetParameters=true}
   /SetVariables(Particles.Velocity=&Subtract_Position(A=$Particles.TargetPosition, B=$Particles.PreviousPosition)) {isSetParameters=true}
   /SetVariables(Particles.VelocityNormalized:Vector3f=&NormalizeVector(Vector To Normalize=$Particles.Velocity, Safe Normalize=true, Safe Fallback Threshold=0, Safe Fallback Vector=<0 0 0.001>)) {isSetParameters=true}
   /SetVariables(Particles.ProjectileVelocity:Vector3f=&NormalizeVector(
     Vector To Normalize=$Output.InitializeParticleFromNDC_FortProjectiles.Velocity
     Safe Normalize=true
     Safe Fallback Threshold=0
     Safe Fallback Vector=<0 0 0.001>
    )) {isSetParameters=true}
   /SetVariables(Particles.ProjectileSpeed:NiagaraFloat=&Multiply_Float(
     A=&VectorLength(Vector=$Output.InitializeParticleFromNDC_FortProjectiles.Velocity)
     B=$Engine.DeltaTime
    )) {isSetParameters=true}
   /SetVariables(Particles.RibbonID=$Output.InitializeParticleFromNDC_FortProjectiles.ID) {isSetParameters=true}
   /SetVariables(Particles.RemapByDistance:NiagaraFloat=&RemapRange(
     Input Value To Remap=&FloatFromCurve(
      FloatCurve=DataInterface() {propertyValues {Curve {keys [
          {arriveTangent=0.0002, interpMode=RCIM_Cubic, leaveTangent=0.0002, leaveTangentWeight=0.3333, tangentMode=RCTM_User, time=0, value=0}
          <at 1 0>
          {arriveTangent=0.0001, arriveTangentWeight=833, interpMode=RCIM_Cubic, leaveTangent=0.0001, tangentMode=RCTM_User, time=2500, value=0.35}
          <at 15000 0.8>
          {arriveTangentWeight=6666.667, interpMode=RCIM_Cubic, tangentMode=RCTM_User, time=35000, value=1}
         ]}}}
      CurveIndex=$Particles.LOD Distance
     )
     Input Min=0
     Input Max=1
     Output Min=0.5
     Output Max=3
    )) {isSetParameters=true}
   /SetVariables(Particles.TracerLength:NiagaraFloat=&Multiply_Float(A=&VectorLength(Vector=$Particles.Velocity), B=1)) {isSetParameters=true}
   /SetVariables(Particles.TracerLength:NiagaraFloat=&MakeCustomFloatFromBool(
     Bool=$Particles.IsEndPosition
     True Float=$Particles.TracerLength
     False Float=&Multiply_Float(
      A=$Particles.TracerLength
      B=&ClampFloat(Float=&Divide_Float(A=$Particles.ProjectileAge, B=0.01))
     )
    )) {isSetParameters=true}
   /SetVariables(Particles.TracerSoftness:NiagaraFloat=&MakeCustomFloatFromBool(Bool=$Output.InitializeParticleFromNDC_FortProjectiles.bIsSupressed, True Float=0.8, False Float=0.3)) {isSetParameters=true}
   /SetVariables(Particles.Hit Lifetime:NiagaraFloat=&Lerp_Float(
     A=0.1
     B=0.1667
     Alpha=&ClampFloat(Float=&Divide_Float(A=$Particles.Distance, B=$Particles.ProjectileSpeed))
    )) {isSetParameters=true}
   &InitializeParticle(
    Color=<rgba 1 0 0.025 1>
    Lifetime=&MakeCustomFloatFromBool(Bool=$Particles.IsEndPosition, True Float=$Particles.Hit Lifetime, False Float=0.0167)
    Mesh Scale Mode=::NewEnumerator3
    Mesh Uniform Scale=&RandomRangeFloat(Minimum=0.075, Maximum=0.125)
    Position=$Particles.TargetPosition
    Position Mode=::NewEnumerator1
    Sprite Size=&MakeVector2D(X=5, Y=$Particles.TracerLength)
    Sprite Size Mode=::NewEnumerator4
    UsePositionOffset=false
    Write Material Random=true
   )
   /SetVariables(Particles.DMP_Tracer:Vector4f=&MakeVector4(X=1, Y=1, Z=1, W=1)) {isSetParameters=true}
   /SetVariables(Particles.Muzzle Position Offset:Vector3f=&Multiply_VectorByFloat(
     Vector=$Particles.VelocityNormalized
     Float=&Multiply_Float(
      A=&MakeFloatFromVector2d(Vector 2D=$Particles.SpriteSize, Channel=::NewEnumerator1)
      B=&Multiply_Float(A=-0.5, B=1)
     )
    )) {isSetParameters=true}
  ]
  particleUpdateScript [
   &KillParticles(Kill Particles=$Particles.IsStartPosition)
   &ParticleState()
   &ScaleSpriteSize(
    Scale Factor=&MakeVector2D(
     X=$Particles.RemapByDistance
     Y=&MakeCustomFloatFromBool(
      Bool=$Particles.IsEndPosition
      True Float=&FloatFromCurve(FloatCurve=DataInterface() {propertyValues {Curve {keys [<at 0 0>,<at 0.25 0.66>,<at 0.6 0.66>,<at 1 0>]}}})
      False Float=1
     )
    )
    Scale Sprite Size Mode=::NewEnumerator2
   )
   &DynamicMaterialParameters(
    Index 0 Param 4=13
    Index 1 Param 1=&FloatFromCurve(FloatCurve=DataInterface() {propertyValues {Curve {keys [<at 0 0.5>,<at 1 1>]}}}, CurveIndex=$Particles.ProjectileAge)
    Index 1 Param 2=2
    Index 1 Param 3=&FloatFromCurve(
     FloatCurve=DataInterface() {propertyValues {Curve {keys [{interpMode=RCIM_Cubic, tangentMode=RCTM_Break, time=0.5, value=-0.5}, <at 1 0.1>]}}}
     CurveIndex=$Particles.ProjectileAge
    )
    Index 2 Param 1=&Multiply_Float(
     A=&FloatFromCurve(
      FloatCurve=DataInterface() {propertyValues {Curve {keys [{interpMode=RCIM_Cubic, leaveTangent=8.1085, tangentMode=RCTM_Break, time=0.8568, value=1}, {tangentMode=RCTM_User, time=1, value=0}]}}}
      CurveIndex=$Particles.ProjectileAge
     )
     B=&Divide_Float(
      A=&MakeFloatFromVector2d(Vector 2D=$Particles.SpriteSize)
      B=&RandomRangeFloat(Minimum=10, Maximum=15)
     )
    )
    Index 2 Param 2=&FloatFromCurve(
     FloatCurve=DataInterface() {propertyValues {Curve {keys [
         {interpMode=RCIM_Cubic, tangentMode=RCTM_Break, time=0.7, value=0}
         {arriveTangent=-0.0859, interpMode=RCIM_Cubic, tangentMode=RCTM_Break, time=1, value=-1}
        ]}}}
     CurveIndex=$Particles.ProjectileAge
     Scale Curve=50
    )
    Index 2 Param 3=&Waveform(
     Global Input Position=&RandomRangeFloat(Minimum=0, Maximum=15)
     Global Phase=$Particles.Age
     "Waveform [1]"=::NewEnumerator12
     "[1] Phase Scale"=2
     "[1] Frequency"=1
     "[1] Amplitude Min/Max"=<0 1>
    )
    Index 2 Param 4=25
    Module.Index0Param0Write=true
    Module.Index0Param1Write=true
    Module.Index0Param2Write=true
    Module.Index0Param3Write=true
    Module.Index1Param0Write=true
    Module.Index1Param1Write=true
    Module.Index1Param2Write=true
    Module.Index1Param3Write=true
    Module.Index2Param0Write=true
    Module.Index2Param1Write=true
    Module.Index2Param2Write=true
    Module.Index2Param3Write=true
    Write Parameter Index 1=true
    Write Parameter Index 2=true
   )
   /SetVariables(Particles.Muzzle Position Offset:Vector3f=&Multiply_VectorByFloat(
     Vector=$Particles.VelocityNormalized
     Float=&Multiply_Float(A=&MakeFloatFromVector2d(Vector 2D=$Particles.SpriteSize, Channel=::NewEnumerator1), B=-0.5)
    )) {isSetParameters=true, module=SetVariables_#2}
   /SetVariables(Particles.Position=&AddVectorToPosition(Position=$Particles.Initial.Position, Vector=$Particles.Muzzle Position Offset)) {isSetParameters=true, module=SetVariables_#3}
   /SetVariables(Particles.Position=&MakeCustomVectorFromBool(
     Bool=$Particles.IsEndPosition
     True Vector=&Lerp_Vector(
      A=Linked() {linkedVariable {name=Particles.PreviousPosition, type=NiagaraPosition}}
      B=Linked() {linkedVariable {name=Particles.Position, type=NiagaraPosition}}
      Alpha=&FloatFromCurve(FloatCurve=DataInterface() {propertyValues {Curve {keys [
           {arriveTangent=1.6275, interpMode=RCIM_Cubic, leaveTangent=1.6275, tangentMode=RCTM_User, time=0, value=0}
           {interpMode=RCIM_Cubic, tangentMode=RCTM_User, time=0.66, value=1}
          ]}}})
     )
     False Vector=Linked() {linkedVariable {name=Particles.Position, type=NiagaraPosition}}
    )) {isSetParameters=true, module=SetVariables_#4}
   /SetVariables(
    Particles.Color_SuperShot:LinearColor=<rgba 15 15 15 0.25>
    Particles.Color_SuperShot_Details:LinearColor=<rgba 5 5 5 1>
    Particles.SpriteSize_SuperShot:Vector2f=&Multiply_Vector2D(A=$Particles.SpriteSize, B=<1.75 1.1>)
    Particles.SpriteSize_SuperShot_Details:Vector2f=&Multiply_Vector2D(A=$Particles.SpriteSize, B=<6 1>)
   ) {isSetParameters=true}
   &ScaleColor(Scale RGB=&VectorFromFloat(Value=30), ScaleA=false, ScaleRGB=true)
  ]
  rendererClasses [&NiagaraSpriteRendererProperties]
  renderers [
   {rendererClass=&NiagaraSpriteRendererProperties}
   {rendererClass=&NiagaraSpriteRendererProperties, rendererIndex=1}
  ]
 }
 FN_Debug {
  bEnabled=false
  emitterUpdateScript [
   &EmitterState()
   &self:Read_NDC_PiewPistol_Trace(Data Channel:NiagaraDataInterfaceDataChannelRead=$System.NDCReader) {module=Read_NDC_PiewPistol_Trace}
   /SetVariables(Emitter.bIsSuperShot:NiagaraBool=$Emitter.bIsSuperShot) {isSetParameters=true}
   &SpawnBurst_Instantaneous(Spawn Count=1, Use Spawn Probability=false)
   &Spawn_NDCFortProjectiles(Reader:NiagaraDataInterfaceDataChannelRead=$System.NDCReader) {enabled=false}
  ]
  particleSpawnScript [
   /Niagara/Modules/Spawn/Initialization/InitializeParticle(
    Lifetime=5
    Module.Write Color=true
    Module.Write Lifetime=true
    Module.Write Mass=true
    Module.Write Material Random=true
    Module.Write Position=false
    Module.Write Scale=false
    Module.Write SpriteRotation=false
    Module.Write SpriteSize=true
    Position=&SimulationPosition()
    Sprite Size=<100 100>
   )
   &InitializeParticleFromNDC_FortProjectiles() {enabled=false}
  ]
  particleUpdateScript [
   &ParticleState()
   /SetVariables(Particles.Debug:NiagaraFloat=&MakeCustomFloatFromBool(Bool=$Emitter.bIsSuperShot, True Float=1, False Float=0)) {isSetParameters=true}
   &DynamicMaterialParameters(
    Index 0 Param 1=$Particles.Debug
    Module.Param0WriteEnabled=true
    Module.Param1WriteEnabled=true
    Module.Param2WriteEnabled=true
    Module.Param3WriteEnabled=true
   )
  ]
  rendererClasses [&NiagaraSpriteRendererProperties]
  renderers [{rendererClass=&NiagaraSpriteRendererProperties}]
 }
 SplitLightning_ArcsFromHands {
  bEnabled=false
  emitterSpawnScript [/SetVariables(Emitter.UseTeardrop=true) {isSetParameters=true}]
  emitterUpdateScript [
   &EmitterState()
   &self:Read_NDC_PiewPistol_Trace(Data Channel:NiagaraDataInterfaceDataChannelRead=$System.NDCReader, Read Index=1) {module=Read_NDC_PiewPistol_Trace}
   /SetVariables(Emitter.bIsSuperShot:NiagaraBool=$Emitter.Read_NDC_PiewPistol_Tracer_02.bIsSuperShot) {isSetParameters=true}
   &Spawn_NDCFortProjectiles(Enabled:NiagaraBool=$Emitter.bIsSuperShot, Reader:NiagaraDataInterfaceDataChannelRead=$System.NDCReader)
  ]
  particleSpawnScript [
   &InitializeParticleFromNDC_FortProjectiles(FallbackAmmoType=2)
   /SetVariables(Particles.PositionType:NiagaraInt32=$Output.InitializeParticleFromNDC_FortProjectiles.PositionType) {isSetParameters=true}
   /SetVariables(Particles.IsStartPosition:NiagaraBool=&SetBoolByIntComparison(A=$Particles.PositionType, B=0, Comparison Type=::NewEnumerator4)) {isSetParameters=true}
   /SetVariables(Particles.OriginTime:NiagaraFloat=$Output.InitializeParticleFromNDC_FortProjectiles.OriginTime) {isSetParameters=true}
   /SetVariables(Particles.ProjectileAge:NiagaraFloat=&Subtract_Float(A=$Engine.Time, B=$Particles.OriginTime)) {isSetParameters=true}
   /SetVariables(Particles.PreviousPosition:NiagaraPosition=$Output.InitializeParticleFromNDC_FortProjectiles.PrevPosition) {isSetParameters=true}
   /SetVariables(Particles.TargetPosition:NiagaraPosition=$Output.InitializeParticleFromNDC_FortProjectiles.Position) {isSetParameters=true}
   /SetVariables(Particles.Distance:NiagaraFloat=&DistanceBetweenPositions(A=$Particles.PreviousPosition, B=$Particles.TargetPosition)) {isSetParameters=true}
   /SetVariables(Particles.Velocity=&Subtract_Position(A=$Particles.TargetPosition, B=$Particles.PreviousPosition)) {isSetParameters=true}
   /SetVariables(Particles.VelocityNormalized:Vector3f=&NormalizeVector(Vector To Normalize=$Particles.Velocity, Safe Normalize=true, Safe Fallback Threshold=0, Safe Fallback Vector=<0 0 0.001>)) {isSetParameters=true}
   /SetVariables(Particles.ProjectileSpeed:NiagaraFloat=&Multiply_Float(
     A=&VectorLength(Vector=$Output.InitializeParticleFromNDC_FortProjectiles.Velocity)
     B=$Engine.DeltaTime
    )) {isSetParameters=true}
   /SetVariables(Particles.TracerLength:NiagaraFloat=&Multiply_Float(A=&VectorLength(Vector=$Particles.Velocity), B=1)) {isSetParameters=true}
   /SetVariables(Particles.TracerLength:NiagaraFloat=&MakeCustomFloatFromBool(
     Bool=$Particles.IsEndPosition
     True Float=$Particles.TracerLength
     False Float=&Multiply_Float(
      A=$Particles.TracerLength
      B=&ClampFloat(Float=&Divide_Float(A=$Particles.ProjectileAge, B=0.01))
     )
    )) {isSetParameters=true}
   /SetVariables(Particles.Hit Lifetime:NiagaraFloat=&Lerp_Float(
     A=0.2
     B=0.3
     Alpha=&ClampFloat(Float=&Divide_Float(A=$Particles.Distance, B=$Particles.ProjectileSpeed))
    )) {isSetParameters=true}
   /Niagara/Modules/Spawn/Initialization/V2/InitializeParticle(
    Color=&HueShiftLinearColor(
     Hue=$User.HueShift
     Saturation=1
     Value=1
     Linear Color=&MultiplyLinearColor(
      Linear Color=&RandomRangeLinearColor(Minimum=<rgba 1.8 0 0 0.3>, Maximum=<rgba 1.8 0 0 1>)
      Scale Factor=&MakeVector4_FromFloat(Float=0.8)
     )
    )
    Lifetime=&RandomRangeFloat(Minimum=0.2, Maximum=0.3)
    Mesh Scale Mode=::NewEnumerator1
    Mesh Uniform Scale Max=0.25
    Mesh Uniform Scale Min=0.125
    Position Offset=<0 0 60>
    UsePositionOffset=false
    Write Material Random=true
   )
   /SetVariables(Particles.Position=&LerpPosition(
     A=&AddVectorToPosition(Position=&MakeVectorFromLinearColorRGB(Color=$NPC.FortniteNPC.PlayerPosition), Vector=<0 0 60>)
     B=$Particles.TargetPosition
     Alpha=&Add_Float(A=0.35, B=$Particles.ProjectileAge)
    )) {isSetParameters=true}
   &InitialMeshOrientation(Rotation=&RandomRangeVector(Minimum=<0 0 -1>, Maximum=<0.1 0.1 1>), Use Rotation Vector=true)
  ]
  particleUpdateScript [
   &KillParticles(Kill Particles=$Particles.IsStartPosition)
   &ParticleState()
   &ScaleMeshSize(Scale Factor=&Multiply_VectorByFloat(Vector=&RandomRangeVector(Minimum=<1 1 2.5>, Maximum=<1.3 1.3 5>), Float=1)) {module=ScaleMeshSize001}
   &DynamicMaterialParameters(
    Index 0 Param 4=13
    Index 1 Param 1=&FloatFromCurve(FloatCurve=DataInterface() {propertyValues {Curve {keys [<at 0 0.5>,<at 1 1>]}}})
    Index 1 Param 2=2
    Index 1 Param 3=&FloatFromCurve(FloatCurve=DataInterface() {propertyValues {Curve {keys [{interpMode=RCIM_Cubic, tangentMode=RCTM_Break, time=0.5, value=-0.5}, <at 1 0.1>]}}})
    Index 2 Param 1=&Multiply_Float(
     A=&FloatFromCurve(FloatCurve=DataInterface() {propertyValues {Curve {keys [{interpMode=RCIM_Cubic, leaveTangent=0.5889, tangentMode=RCTM_Break, time=0.8568, value=1}, {tangentMode=RCTM_User, time=1, value=0}]}}})
     B=&Divide_Float(
      A=&MakeFloatFromVector(Vector=$Particles.Scale, Channel=::NewEnumerator0)
      B=&Multiply_Float(
       A=&Lerp_Float(A=1, B=0.25, Alpha=$Particles.ProjectileAge)
       B=&RandomRangeFloat(Minimum=10, Maximum=15)
      )
     )
    )
    Index 2 Param 2=&FloatFromCurve(
     FloatCurve=DataInterface() {propertyValues {Curve {keys [
         {interpMode=RCIM_Cubic, tangentMode=RCTM_Break, time=0.7, value=0}
         {arriveTangent=-0.0859, interpMode=RCIM_Cubic, tangentMode=RCTM_Break, time=1, value=-1}
        ]}}}
     Scale Curve=50
    )
    Index 2 Param 3=&Waveform(
     Global Input Position=&RandomRangeFloat(Minimum=0, Maximum=15)
     Global Phase=$Particles.Age
     "Waveform [1]"=::NewEnumerator12
     "[1] Phase Scale"=2
     "[1] Frequency"=1
     "[1] Amplitude Min/Max"=<0 1>
    )
    Index 2 Param 4=25
    Index 3 Param 1=&FloatFromCurve(
     FloatCurve=DataInterface() {propertyValues {Curve {keys [
         {interpMode=RCIM_Cubic, leaveTangent=-0.0025, tangentMode=RCTM_Break, time=0, value=0}
         {arriveTangent=1.8528, interpMode=RCIM_Cubic, tangentMode=RCTM_Break, time=1, value=1}
        ]}}}
     Scale Curve=&RandomRangeFloat(Minimum=0, Maximum=0)
    )
    Index 3 Param 2=&FloatFromCurve(
     FloatCurve=DataInterface() {propertyValues {Curve {keys [
         {interpMode=RCIM_Cubic, leaveTangent=-0.0025, tangentMode=RCTM_Break, time=0, value=0.5}
         {arriveTangent=-0.0316, interpMode=RCIM_Cubic, tangentMode=RCTM_Break, time=1, value=-0.2}
        ]}}}
     Scale Curve=0.1
    )
    Index 3 Param 3=&FloatFromCurve(FloatCurve=DataInterface() {propertyValues {Curve {keys [
         {interpMode=RCIM_Cubic, leaveTangent=1.1034, tangentMode=RCTM_Break, time=0, value=-1}
         {arriveTangent=2.2399, tangentMode=RCTM_User, time=0.7, value=0.125}
        ]}}})
    Index 3 Param 4=&FloatFromCurve(FloatCurve=DataInterface() {propertyValues {Curve {keys [
         {interpMode=RCIM_Cubic, leaveTangent=3.0601, tangentMode=RCTM_Break, time=0.5, value=-0.5}
         {arriveTangent=3.2053, tangentMode=RCTM_User, time=1, value=1}
        ]}}})
    Module.Index0Param0Write=true
    Module.Index0Param1Write=true
    Module.Index0Param2Write=true
    Module.Index0Param3Write=true
    Module.Index1Param0Write=true
    Module.Index1Param1Write=true
    Module.Index1Param2Write=true
    Module.Index1Param3Write=true
    Module.Index2Param0Write=true
    Module.Index2Param1Write=true
    Module.Index2Param2Write=true
    Module.Index2Param3Write=true
    Module.Index3Param0Write=true
    Module.Index3Param1Write=true
    Module.Index3Param2Write=true
    Module.Index3Param3Write=true
    Write Parameter Index 1=true
    Write Parameter Index 2=true
    Write Parameter Index 3=true
   )
   &ScaleColor(
    Scale Alpha=&FloatFromCurve(FloatCurve=DataInterface() {propertyValues {Curve {
        defaultValue=1
        keys [{interpMode=RCIM_Cubic, tangentMode=RCTM_User, time=0, value=1}, {interpMode=RCIM_Cubic, tangentMode=RCTM_User, time=1, value=0}]
       }}})
    Scale RGB=&Lerp_Vector(
     A=&VectorFromFloat(Value=&FloatFromCurve(FloatCurve=DataInterface() {propertyValues {Curve {keys [<at 0.2771 1>,<at 0.3786 1>]}}}, Scale Curve=25))
     B=&VectorFromFloat(Value=&RandomRangeFloat(Minimum=0.3, Maximum=2, Evaluation Type=::NewEnumerator1))
     Alpha=&FloatFromCurve(FloatCurve=DataInterface() {propertyValues {Curve {keys [
          {interpMode=RCIM_Cubic, leaveTangent=2.5294, tangentMode=RCTM_Break, time=0, value=0}
          {arriveTangent=-0.2166, interpMode=RCIM_Cubic, tangentMode=RCTM_Break, time=0.4, value=1}
         ]}}})
    )
    ScaleA=true
    ScaleRGB=true
   )
  ]
  rendererClasses [&NiagaraMeshRendererProperties]
  renderers [
   {rendererClass=&NiagaraMeshRendererProperties, rendererIndex=0}
   {rendererClass=&NiagaraMeshRendererProperties, rendererIndex=1}
  ]
 }
 Tracer {
  emitterUpdateScript [
   &EmitterState()
   &self:Read_NDC_PiewPistol_Trace(Data Channel:NiagaraDataInterfaceDataChannelRead=$System.NDCReader) {enabled=false, module=Read_NDC_PiewPistol_Trace}
   /SetVariables(Emitter.bIsSuperShot:NiagaraBool=$StackContext.bIsSuperShot) {enabled=false, isSetParameters=true}
   &Spawn_NDCFortProjectiles(Reader:NiagaraDataInterfaceDataChannelRead=$System.NDCReader)
  ]
  particleSpawnScript [
   &InitializeParticleFromNDC_FortProjectiles(
    End Lifetime=true
    End Lifetime Min=0.05
    FallbackAmmoType=2
    FallbackOrigin:NiagaraPosition=<0 50 0>
    Intro Time=true
    LOD Output Scale Min=0.5
    Length Scale Array=DataInterface() {propertyValues {GpuSyncMode=SyncCpuToGpu, MaxElements=4, FloatData [1.4,1.7,2,1]}}
    Length Scale by Array=true
    Length by Velocity=true
    Lifetime=0.0333
    Scale Array=DataInterface() {propertyValues {GpuSyncMode=SyncCpuToGpu, MaxElements=4, FloatData [18,20,26,14]}}
    Scale by Array=true
    Use LOD Distance Scaling=true
   )
   &InitializeParticle(
    Color=<rgba 30 0 1 1>
    Lifetime=$Particles.Lifetime
    Position=$Particles.CurrentPosition
    Position Mode=::NewEnumerator1
    Sprite Size=&MakeVector2D(
     X=&Multiply_Float(A=$Particles.ProjectileScale, B=0.75)
     Y=&Multiply_Float(A=$Particles.ProjectileLength, B=$Particles.NormalizedIntroDuration)
    )
    Sprite Size Mode=::NewEnumerator4
    UsePositionOffset=false
    Write Material Random=true
   )
   &KillParticles(Kill Particles=$Particles.IsStartPosition) {module=KillParticles001}
   /SetVariables(Particles.IsHitScan:NiagaraBool=&Boolean_AndOperation(
     Bool1=$Particles.IsEndPosition
     Bool2=&SetBoolByFloatComparison(
      A=&VectorLength(
       Vector=&SubtractVectorFromPosition(Position=$Particles.PreviousPosition, Vector=Linked() {linkedVariable {name=Particles.OriginPosition, type=NiagaraPosition}})
       Safe Length Calculation=false
      )
      B=0
      Comparison Type=::NewEnumerator4
     )
    )) {isSetParameters=true, module=SetVariables_#5}
   /SetVariables(Particles.DMP_Tracer:Vector4f=&MakeVector4(X=3.3, Y=1, Z=3, W=0.25)) {isSetParameters=true, module=SetVariables_#6}
  ]
  particleUpdateScript [
   &ParticleState()
   &KillParticles(Kill Particles=&Boolean_NotOperation(Bool=$Particles.IsEndPosition))
   /SetVariables(Particles.HitScanAge:NiagaraFloat=&MakeCustomFloatFromBool(
     Bool=$Particles.IsHitScan
     True Float=$Particles.NormalizedAge
     False Float=&ClampFloat(Float=&Add_Float(A=&Multiply_Float(A=$Particles.NormalizedAge, B=0.5), B=0.66))
    )) {isSetParameters=true, module=SetVariables_#7}
   /SetVariables(Particles.Position=$Particles.Initial.Position) {isSetParameters=true, module=SetVariables_#3}
   /SetVariables(Particles.Position=&MakeCustomVectorFromBool(
     Bool=$Particles.IsEndPosition
     True Vector=&Lerp_Vector(
      A=Linked() {linkedVariable {name=Particles.PreviousPosition, type=NiagaraPosition}}
      B=Linked() {linkedVariable {name=Particles.Position, type=NiagaraPosition}}
      Alpha=&FloatFromCurve(
       FloatCurve=DataInterface() {propertyValues {Curve {keys [{arriveTangent=2.1933, time=0, value=0}, <at 0.66 1>]}}}
       CurveIndex=$Particles.HitScanAge
      )
     )
     False Vector=Linked() {linkedVariable {name=Particles.Position, type=NiagaraPosition}}
    )) {isSetParameters=true, module=SetVariables_#4}
   /SetVariables(Particles.Tracer Scale:NiagaraFloat=&ClampFloat(Float=&Divide_Float(
      A=&VectorLength(
       Vector=&SubtractVectorFromPosition(Position=$Particles.Position, Vector=Linked() {linkedVariable {name=Particles.OriginPosition, type=NiagaraPosition}})
       Safe Length Calculation=false
      )
      B=$Particles.ProjectileLength
     ))) {isSetParameters=true, module=SetVariables_#8}
   &ScaleSpriteSize(
    Scale Factor=&MakeVector2D(
     X=1
     Y=&FloatFromCurve(
      FloatCurve=DataInterface() {propertyValues {Curve {keys [<at 0 0>, <at 0.2 1>, {arriveTangent=-4.3415, time=0.8, value=1}, {arriveTangent=-1.2985, leaveTangent=-1.2985, time=1, value=0}]}}}
      CurveIndex=$Particles.HitScanAge
      Scale Curve=$Particles.Tracer Scale
     )
    )
    Scale Sprite Size Mode=::NewEnumerator2
   ) {module=ScaleSpriteSize001}
   /SetVariables(Particles.Muzzle Position Offset:Vector3f=&Multiply_VectorByFloat(
     Vector=$Particles.Direction
     Float=&Multiply_Float(A=&MakeFloatFromVector2d(Vector 2D=$Particles.SpriteSize, Channel=::NewEnumerator1), B=-0.49)
    )) {isSetParameters=true, module=SetVariables_#2}
   /SetVariables(Particles.Position=&AddVectorToPosition(Position=$Particles.Position, Vector=$Particles.Muzzle Position Offset)) {isSetParameters=true, module=SetVariables_#9}
  ]
  rendererClasses [&NiagaraSpriteRendererProperties]
  renderers [{rendererClass=&NiagaraSpriteRendererProperties}]
 }
 Tracer_SuperShot {
  bEnabled=false
  emitterUpdateScript [
   &EmitterState()
   &Spawn_NDCFortProjectiles(Reader:NiagaraDataInterfaceDataChannelRead=$System.NDCReader)
  ]
  particleSpawnScript [
   &InitializeParticleFromNDC_FortProjectiles(
    End Lifetime=true
    End Lifetime Min=0.05
    FallbackAmmoType=2
    FallbackOrigin:NiagaraPosition=<0 50 0>
    Intro Time=true
    LOD Output Scale Min=0.5
    Length Scale Array=DataInterface() {propertyValues {GpuSyncMode=SyncCpuToGpu, MaxElements=4, FloatData [1.4,1.7,2,1]}}
    Length Scale by Array=true
    Length by Velocity=true
    Lifetime=0.0333
    Scale Array=DataInterface() {propertyValues {GpuSyncMode=SyncCpuToGpu, MaxElements=4, FloatData [18,20,26,14]}}
    Scale by Array=true
    Use LOD Distance Scaling=true
   )
   &KillParticles(Kill Particles=$Particles.IsStartPosition) {module=KillParticles001}
   &InitializeParticle(
    Lifetime=$Particles.Lifetime
    Position=$Particles.CurrentPosition
    Position Mode=::NewEnumerator1
    Sprite Size=&MakeVector2D(
     X=&Multiply_Float(A=$Particles.ProjectileScale, B=0.75)
     Y=&Multiply_Float(A=$Particles.ProjectileLength, B=$Particles.NormalizedIntroDuration)
    )
    Sprite Size Mode=::NewEnumerator4
    UsePositionOffset=false
    Write Material Random=true
   )
   /SetVariables(
    Particles.SpriteSize_SuperShot:Vector2f=&Multiply_Vector2D(
     A=$Particles.SpriteSize
     B=&MakeVector2D(
      X=1.75
      Y=&Lerp_Float(
       A=1
       B=5
       Alpha=&FloatFromCurve(
        FloatCurve=DataInterface() {propertyValues {Curve {
           defaultValue=0
           keys [
            {interpMode=RCIM_Cubic, tangentMode=RCTM_User, time=0, value=0}
            {arriveTangent=1.8595, interpMode=RCIM_Cubic, leaveTangent=1.8595, tangentMode=RCTM_User, time=1, value=1}
           ]
          }}}
        CurveIndex=&FN_LOD_Distance_Remap(Start Distance=0, End Distance=10000, Minimum Output=0, Maximum Output=1, Clamp Output=true)
       )
      )
     )
    )
    Particles.SpriteSize_SuperShot_Details:Vector2f=&Multiply_Vector2D(
     A=$Particles.SpriteSize
     B=&MakeVector2D(
      X=6
      Y=&Lerp_Float(
       A=1
       B=5
       Alpha=&FloatFromCurve(
        FloatCurve=DataInterface() {propertyValues {Curve {
           defaultValue=0
           keys [
            {interpMode=RCIM_Cubic, tangentMode=RCTM_User, time=0, value=0}
            {arriveTangent=1.8595, interpMode=RCIM_Cubic, leaveTangent=1.8595, tangentMode=RCTM_User, time=1, value=1}
           ]
          }}}
        CurveIndex=&FN_LOD_Distance_Remap(Start Distance=0, End Distance=10000, Minimum Output=0, Maximum Output=1, Clamp Output=true)
       )
      )
     )
    )
   ) {isSetParameters=true}
   /SetVariables(Particles.IsHitScan:NiagaraBool=&Boolean_AndOperation(
     Bool1=$Particles.IsEndPosition
     Bool2=&SetBoolByFloatComparison(
      A=&VectorLength(
       Vector=&SubtractVectorFromPosition(Position=$Particles.PreviousPosition, Vector=Linked() {linkedVariable {name=Particles.OriginPosition, type=NiagaraPosition}})
       Safe Length Calculation=false
      )
      B=0
      Comparison Type=::NewEnumerator4
     )
    )) {isSetParameters=true, module=SetVariables_#5}
   /SetVariables(Particles.DMP_Tracer:Vector4f=&MakeVector4(X=3.3, Y=1, Z=3, W=0.25)) {isSetParameters=true, module=SetVariables_#6}
   &DynamicMaterialParameters(
    Index 0 Param 4=13
    Index 1 Param 1=&FloatFromCurve(FloatCurve=DataInterface() {propertyValues {Curve {keys [<at 0 0.5>,<at 1 1>]}}}, CurveIndex=$Particles.ProjectileAge)
    Index 1 Param 2=2
    Index 2 Param 1=&Multiply_Float(
     A=&FloatFromCurve(
      FloatCurve=DataInterface() {propertyValues {Curve {keys [{interpMode=RCIM_Cubic, leaveTangent=8.1085, tangentMode=RCTM_Break, time=0.8568, value=1}, {tangentMode=RCTM_User, time=1, value=0}]}}}
      CurveIndex=$Particles.ProjectileAge
     )
     B=&Divide_Float(
      A=&MakeFloatFromVector2d(Vector 2D=$Particles.SpriteSize)
      B=&RandomRangeFloat(Minimum=10, Maximum=15)
     )
    )
    Index 2 Param 2=&FloatFromCurve(
     FloatCurve=DataInterface() {propertyValues {Curve {keys [
         {interpMode=RCIM_Cubic, tangentMode=RCTM_Break, time=0.7, value=0}
         {arriveTangent=-0.0859, interpMode=RCIM_Cubic, tangentMode=RCTM_Break, time=1, value=-1}
        ]}}}
     CurveIndex=$Particles.ProjectileAge
     Scale Curve=50
    )
    Index 2 Param 3=&Waveform(
     Global Input Position=&RandomRangeFloat(Minimum=0, Maximum=15)
     Global Phase=$Particles.ProjectileAge
     "Waveform [1]"=::NewEnumerator12
     "[1] Phase Scale"=2
     "[1] Frequency"=1
     "[1] Amplitude Min/Max"=<0 1>
    )
    Index 2 Param 4=25
    Module.Index0Param0Write=true
    Module.Index0Param1Write=true
    Module.Index0Param2Write=true
    Module.Index0Param3Write=true
    Module.Index1Param0Write=true
    Module.Index1Param1Write=true
    Module.Index1Param2Write=false
    Module.Index1Param3Write=true
    Module.Index2Param0Write=true
    Module.Index2Param1Write=true
    Module.Index2Param2Write=true
    Module.Index2Param3Write=true
    Write Parameter Index 1=true
    Write Parameter Index 2=true
   )
   &SpriteFacingAndAlignment(
    Facing Coordinate Space=::Local
    Sprite Alignment=<0 0 1>
    Sprite Facing=&RotateVectorByQuaternion(
     Vector To Rotate=<0 0.25 1>
     Quaternion=&OrientMeshToVector(
      Look At Direction=&NormalizeVector(Vector To Normalize=$Particles.Velocity, Safe Normalize=true, Safe Fallback Threshold=0, Safe Fallback Vector=<0 0 0.001>)
      Reference Axis Vector=<1 0 0>
     )
    )
    WriteAlignment=false
    WriteFacing=true
   )
  ]
  particleUpdateScript [
   &ParticleState()
   &KillParticles(Kill Particles=&Boolean_NotOperation(Bool=$Particles.IsEndPosition))
   /SetVariables(Particles.HitScanAge:NiagaraFloat=&MakeCustomFloatFromBool(
     Bool=$Particles.IsHitScan
     True Float=$Particles.NormalizedAge
     False Float=&ClampFloat(Float=&Add_Float(A=&Multiply_Float(A=$Particles.NormalizedAge, B=0.5), B=0.66))
    )) {isSetParameters=true, module=SetVariables_#7}
   /SetVariables(Particles.Position=$Particles.Initial.Position) {isSetParameters=true, module=SetVariables_#3}
   /SetVariables(Particles.Position=&MakeCustomVectorFromBool(
     Bool=$Particles.IsEndPosition
     True Vector=&Lerp_Vector(
      A=Linked() {linkedVariable {name=Particles.PreviousPosition, type=NiagaraPosition}}
      B=Linked() {linkedVariable {name=Particles.Position, type=NiagaraPosition}}
      Alpha=&FloatFromCurve(
       FloatCurve=DataInterface() {propertyValues {Curve {keys [{arriveTangent=2.1933, time=0, value=0}, <at 0.66 1>]}}}
       CurveIndex=$Particles.HitScanAge
      )
     )
     False Vector=Linked() {linkedVariable {name=Particles.Position, type=NiagaraPosition}}
    )) {isSetParameters=true, module=SetVariables_#4}
   /SetVariables(Particles.Tracer Scale:NiagaraFloat=&ClampFloat(Float=&Divide_Float(
      A=&VectorLength(
       Vector=&SubtractVectorFromPosition(Position=$Particles.Position, Vector=Linked() {linkedVariable {name=Particles.OriginPosition, type=NiagaraPosition}})
       Safe Length Calculation=false
      )
      B=$Particles.ProjectileLength
     ))) {isSetParameters=true, module=SetVariables_#8}
   &ScaleSpriteSize(
    Scale Factor=&MakeVector2D(
     X=1
     Y=&FloatFromCurve(
      FloatCurve=DataInterface() {propertyValues {Curve {keys [<at 0 0>, <at 0.2 1>, {arriveTangent=-4.3415, time=0.8, value=1}, {arriveTangent=-1.2985, leaveTangent=-1.2985, time=1, value=0}]}}}
      CurveIndex=$Particles.HitScanAge
      Scale Curve=$Particles.Tracer Scale
     )
    )
    Scale Sprite Size Mode=::NewEnumerator2
   ) {module=ScaleSpriteSize001}
   /SetVariables(
    Particles.SpriteSize_SuperShot:Vector2f=&Multiply_Vector2D(
     A=$Particles.Initial.SpriteSize_SuperShot
     B=&MakeVector2D(
      X=1
      Y=&FloatFromCurve(
       FloatCurve=DataInterface() {propertyValues {Curve {keys [<at 0 0>, <at 0.2 1>, {arriveTangent=-4.3415, time=0.8, value=1}, {arriveTangent=-1.2985, leaveTangent=-1.2985, time=1, value=0}]}}}
       CurveIndex=$Particles.HitScanAge
       Scale Curve=$Particles.Tracer Scale
      )
     )
    )
    Particles.SpriteSize_SuperShot_Details:Vector2f=&Multiply_Vector2D(
     A=$Particles.Initial.SpriteSize_SuperShot_Details
     B=&MakeVector2D(
      X=1
      Y=&FloatFromCurve(
       FloatCurve=DataInterface() {propertyValues {Curve {keys [<at 0 0>, <at 0.2 1>, {arriveTangent=-4.3415, time=0.8, value=1}, {arriveTangent=-1.2985, leaveTangent=-1.2985, time=1, value=0}]}}}
       CurveIndex=$Particles.HitScanAge
       Scale Curve=$Particles.Tracer Scale
      )
     )
    )
   ) {isSetParameters=true}
   /SetVariables(Particles.Muzzle Position Offset:Vector3f=&Multiply_VectorByFloat(
     Vector=$Particles.Direction
     Float=&Multiply_Float(A=&MakeFloatFromVector2d(Vector 2D=$Particles.SpriteSize, Channel=::NewEnumerator1), B=-0.49)
    )) {isSetParameters=true, module=SetVariables_#2}
   /SetVariables(Particles.Position=&AddVectorToPosition(Position=$Particles.Position, Vector=$Particles.Muzzle Position Offset)) {isSetParameters=true, module=SetVariables_#9}
  ]
  rendererClasses [&NiagaraSpriteRendererProperties]
  renderers [
   {rendererClass=&NiagaraSpriteRendererProperties}
   {rendererClass=&NiagaraSpriteRendererProperties, rendererIndex=1}
  ]
 }
}
summary {
 systemName=NS_NDC_PiewPistol_Tracer
 emitters [
  {bEnabled=false, emitterName=Bullet, rendererClasses [&NiagaraSpriteRendererProperties]}
  {bEnabled=false, emitterName=SplitLightning_ArcsFromHands, rendererClasses [&NiagaraMeshRendererProperties]}
  {emitterName=Tracer, rendererClasses [&NiagaraSpriteRendererProperties]}
  {bEnabled=false, emitterName=Tracer_SuperShot, rendererClasses [&NiagaraSpriteRendererProperties]}
  {bEnabled=false, emitterName=FN_Debug, rendererClasses [&NiagaraSpriteRendererProperties]}
 ]
 userVariables [User.HueShift {type=NiagaraFloat, defaultValue=0.025, description=""}]
}
