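# Niagara effect for the Piew Pistol's charge-up while aiming down sights, a short burst tinted through User.HueShift.
# Airflow spawns mesh particles that grow with emitter age and fade out, while Corona spawns sprites at increasing size that shrink and brighten over life.
# Lightning ramps up mesh spawning on a sphere, pulls particles inward with a negative point force, and flickers them through color curves and dynamic material parameters.
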NiagaraSystem /PiewPistolGameplay/Assets/FX/NS_PiewPistol_ADS_ChargingUp
dependencies {
 usedDataInterfaces [&NiagaraDataInterfaceCurve, &NiagaraDataInterfaceColorCurve]
 usedDynamicInputs [
  &SimulationPosition
  &ConvertVectorToPosition
  &FNTimeOfDayScalar
 ]
 usedModules [
  &SystemState
  &self:SystemSpawnScript.SystemScriptSource.SystemScriptGraph.NiagaraNodeAssignment_0.SetVariables_#1
 ]
}
emitters {
 Airflow {
  emitterUpdateScript [
   &EmitterState(Life Cycle Mode=::NewEnumerator1, Loop Behavior=::NewEnumerator1, Loop Duration=0.25, UseLoopDelay=false)
   &SpawnRate(
    SpawnRate=&FloatFromCurve(
     FloatCurve=DataInterface() {propertyValues {Curve {
        defaultValue=1
        keys [
         {tangentMode=RCTM_User, time=0, value=1}
         {arriveTangent=-0.0405, interpMode=RCIM_Cubic, leaveTangent=-0.0405, tangentMode=RCTM_User, time=0.7009, value=1}
         {arriveTangent=-8.0089, interpMode=RCIM_Cubic, leaveTangent=-8.0089, tangentMode=RCTM_User, time=1, value=0}
        ]
       }}}
     CurveIndex=$Emitter.NormalizedLoopAge
     Scale Curve=20
    )
    Use Spawn Probability=false
   )
  ]
  particleSpawnScript [
   &InitializeParticle(
    Color=&HueShiftLinearColor(Hue=$User.HueShift, Saturation=1, Value=1, Linear Color=<rgba 5 0.25 0.3292 1>)
    Lifetime=&RandomRangeFloat(Minimum=0.15, Maximum=0.3)
    Mass Mode=::NewEnumerator1
    Mesh Scale=&Multiply_VectorByFloat(Vector=&RandomRangeVector(Minimum=<0.7 0.7 0.7>, Maximum=<1 1 1>), Float=1)
    Mesh Scale Mode=::NewEnumerator4
    Sprite Size=<5 5>
    Sprite Size Mode=::NewEnumerator4
    UsePositionOffset=false
    Write Material Random=true
   ) {module=InitializeParticle001}
   &InitialMeshOrientation(Use Rotation Vector=false)
  ]
  particleUpdateScript [
   &ParticleState()
   &DynamicMaterialParameters(
    Index 0 Param 1=&FloatFromCurve(FloatCurve=DataInterface() {propertyValues {Curve {keys [<at 0 -0.2>,<at 1 1.2>]}}})
    Index 0 Param 2=1
    Index 0 Param 3=&Multiply_Float(
     A=&RandomRangeFloat(Minimum=0.8, Maximum=1)
     B=&FloatFromCurve(FloatCurve=DataInterface() {propertyValues {Curve {defaultValue=1, keys [{tangentMode=RCTM_User, time=0, value=0}, {tangentMode=RCTM_User, time=1, value=1}]}}})
    )
    Module.Param0WriteEnabled=true
    Module.Param1WriteEnabled=true
    Module.Param2WriteEnabled=true
    Module.Param3WriteEnabled=true
   )
   &ScaleColor(
    Scale Alpha=&FloatFromCurve(
     FloatCurve=DataInterface() {propertyValues {Curve {
        defaultValue=1
        keys [
         {arriveTangent=3.3375, interpMode=RCIM_Cubic, leaveTangent=3.3375, tangentMode=RCTM_User, time=0, value=0}
         {arriveTangent=-2.5249, interpMode=RCIM_Cubic, leaveTangent=-2.5249, tangentMode=RCTM_User, time=0.8, value=1.275}
         {arriveTangent=-0.0405, interpMode=RCIM_Cubic, leaveTangent=-0.0405, tangentMode=RCTM_User, time=1, value=0}
        ]
       }}}
     CurveIndex=$Emitter.Age
    )
    ScaleA=true
    ScaleRGB=true
   ) {module=ScaleColor001}
   &ScaleMeshSize(Scale Factor=&Multiply_Vector(
     A=<1 1 1>
     B=&VectorFromFloat(Value=&FloatFromCurve(
       FloatCurve=DataInterface() {propertyValues {Curve {keys [
           {arriveTangent=0.2635, interpMode=RCIM_Cubic, leaveTangent=0.2635, tangentMode=RCTM_User, time=0, value=0.1}
           {arriveTangent=1.1387, interpMode=RCIM_Cubic, leaveTangent=1.1387, tangentMode=RCTM_User, time=0.512, value=0.4126}
           {arriveTangent=0.447, interpMode=RCIM_Cubic, leaveTangent=0.447, tangentMode=RCTM_User, time=1, value=1.2}
          ]}}}
       CurveIndex=$Emitter.Age
       Scale Curve=0.5
      ))
    ))
   &CameraOffset(Camera Offset Amount=5)
  ]
  rendererClasses [&NiagaraMeshRendererProperties]
  renderers [{rendererClass=&NiagaraMeshRendererProperties, rendererIndex=0}]
 }
 Corona {
  emitterUpdateScript [
   &EmitterState(Inactive Response=::NewEnumerator1, Life Cycle Mode=::NewEnumerator1, Loop Behavior=::NewEnumerator1, Loop Duration=0.25, UseLoopDelay=false)
   &SpawnRate(SpawnRate=20, Use Spawn Probability=false)
  ]
  particleSpawnScript [
   &InitializeParticle(
    Color=&HueShiftLinearColor(Hue=$User.HueShift, Saturation=1, Value=1, Linear Color=<rgba 1 0 0.0167 1>)
    Lifetime=&FloatFromCurve(FloatCurve=DataInterface() {propertyValues {Curve {keys [<at 0 1>,<at 1 0.5>]}}}, CurveIndex=$Emitter.Age, Scale Curve=0.25)
    Position Offset Coordinate Space=::Local
    Sprite Rotation Mode=::NewEnumerator2
    Sprite Size Mode=::NewEnumerator3
    Uniform Sprite Size=&Multiply_Float(
     A=50
     B=/Game/Effects/Niagara/DynamicInputs/Math/FN_Fit(
      Clamp:NiagaraBool=true
      Value:NiagaraFloat=$System.NormalizedLoopAge
      SourceMin:NiagaraFloat=0
      SourceMax:NiagaraFloat=1
      DestMin:NiagaraFloat=0.3
      DestMax:NiagaraFloat=1
     )
    )
    UsePositionOffset=true
    Write Material Random=true
   )
   &CameraOffset(Camera Offset Amount=&Multiply_Float(A=50, B=1))
  ]
  particleUpdateScript [
   &ParticleState()
   &ScaleSpriteSize(
    Scale Sprite Size Mode=::NewEnumerator0
    Uniform Scale Factor=&FloatFromCurve(
     FloatCurve=DataInterface() {propertyValues {Curve {keys [
         {arriveTangent=0.4295, interpMode=RCIM_Cubic, leaveTangent=0.4295, tangentMode=RCTM_User, time=0, value=2.5}
         {arriveTangent=-2.2222, interpMode=RCIM_Cubic, leaveTangent=-2.2222, time=0.5336, value=1.0603}
         <at 0.9 0.5 RCIM_Cubic>
        ]}}}
     Scale Curve=0.5
    )
   )
   &ScaleColor(
    Scale Alpha=&FloatFromCurve(
     FloatCurve=DataInterface() {propertyValues {Curve {keys [
         {interpMode=RCIM_Cubic, tangentMode=RCTM_User, time=0, value=0}
         {interpMode=RCIM_Cubic, tangentMode=RCTM_User, time=0.1, value=1}
         {tangentMode=RCTM_User, time=1, value=0}
        ]}}}
     Scale Curve=0.5
    )
    Scale RGB=&MakeVectorFromLinearColorRGB(Color=&ColorFromCurve(ColorCurve=DataInterface() {propertyValues {
        AlphaCurve {keys []}
        BlueCurve {keys [<at 0 1 RCIM_Cubic>,<at 0.9887 5 RCIM_Cubic>]}
        GreenCurve {keys [<at 0 0.968 RCIM_Cubic>,<at 0.9887 3.3507 RCIM_Cubic>]}
        RedCurve {keys [<at 0 0.904 RCIM_Cubic>,<at 0.9887 2.526 RCIM_Cubic>]}
       }}))
    ScaleA=true
    ScaleRGB=true
   )
   &DynamicMaterialParameters(
    Index 0 Param 1=&FloatFromCurve(FloatCurve=DataInterface() {propertyValues {Curve {keys [<at 0 1 RCIM_Cubic>,<at 1 1>]}}})
    Index 0 Param 2=1
    Index 0 Param 3=&FloatFromCurve(FloatCurve=DataInterface() {propertyValues {Curve {keys [<at 0 0.3>,<at 2 1>]}}})
    Index 0 Param 4=1
    Module.Index0Param0Write=true
    Module.Index0Param1Write=true
    Module.Index0Param2Write=true
    Module.Index0Param3Write=true
   )
  ]
  rendererClasses [&NiagaraSpriteRendererProperties]
  renderers [{rendererClass=&NiagaraSpriteRendererProperties}]
 }
 Lightning {
  emitterUpdateScript [
   &EmitterState(Inactive Response=::NewEnumerator1, Life Cycle Mode=::NewEnumerator1, Loop Behavior=::NewEnumerator1, Loop Duration=0.3, UseLoopDelay=false)
   &SpawnRate(
    SpawnRate=&FloatFromCurve(
     FloatCurve=DataInterface() {propertyValues {Curve {
        defaultValue=1
        keys [
         {arriveTangent=2.1175, interpMode=RCIM_Cubic, leaveTangent=2.1175, tangentMode=RCTM_User, time=0, value=0}
         {interpMode=RCIM_Cubic, tangentMode=RCTM_User, time=1, value=1}
        ]
       }}}
     CurveIndex=$Emitter.NormalizedLoopAge
     Scale Curve=50
    )
    Use Spawn Probability=false
   )
  ]
  particleSpawnScript [
   &FN_SetSpawnGroup(Number of Spawn Groups=2)
   &InitializeParticle(
    Color=&HueShiftLinearColor(
     Hue=$User.HueShift
     Saturation=1
     Value=1
     Linear Color=&MultiplyLinearColor(
      Linear Color=&RandomRangeLinearColor(Minimum=<rgba 1.8 0 0 0.3>, Maximum=<rgba 1.8 0 0 1>)
      Scale Factor=&MakeVector4_FromFloat(Float=0.8)
     )
    )
    Lifetime Max=0.2
    Lifetime Min=0.1
    Lifetime Mode=::NewEnumerator1
    Mesh Scale Max=<1 2 2>
    Mesh Scale Min=<0.5 1 1>
    Mesh Scale Mode=::NewEnumerator2
    UsePositionOffset=true
    Write Material Random=true
   )
   &ShapeLocation(MaskSpawnGroup=false, Override Seed=false, Sphere Radius=20, Sphere Surface Distribution=1, UseOwnerScale=true)
   &InitialMeshOrientation(Mesh Axis To Orient=<0 0 1>, Mesh Orientation Mode=::NewEnumerator2, Use Rotation Vector=false)
  ]
  particleUpdateScript [
   &PointForce(Force Strength=-1250, Module.Use Falloff=false, Module.UseLinearForce=false)
   &SolveForcesAndVelocity(Clamp Velocity=false, Limit Acceleration=false)
   &ParticleState()
   &ScaleMeshSize(Scale Factor=&Multiply_VectorByFloat(Vector=<0.5 0.5 0.5>, Float=0.1))
   &DynamicMaterialParameters(
    Index 0 Param 1=&MakeFloatFromVector(Vector=$Particles.Scale, Channel=::NewEnumerator0)
    Index 0 Param 2=&RandomRangeFloat(Minimum=0, Maximum=10)
    Index 0 Param 3=&FloatFromCurve(
     FloatCurve=DataInterface() {propertyValues {Curve {keys [{interpMode=RCIM_Cubic, leaveTangent=8.1085, tangentMode=RCTM_Break, time=0.8568, value=1}, {tangentMode=RCTM_User, time=1, value=0}]}}}
     Scale Curve=0.05
    )
    Index 0 Param 4=1.2
    Index 1 Param 1=&FloatFromCurve(FloatCurve=DataInterface() {propertyValues {Curve {keys [
         {interpMode=RCIM_Cubic, leaveTangent=1.1034, tangentMode=RCTM_Break, time=0, value=0}
         {arriveTangent=2.2399, tangentMode=RCTM_User, time=0.5, value=1}
        ]}}})
    Index 1 Param 2=&FloatFromCurve(FloatCurve=DataInterface() {propertyValues {Curve {keys [<at 0.5 0>, {arriveTangent=3.2053, tangentMode=RCTM_User, time=0.9, value=1}]}}})
    Index 1 Param 3=&FloatFromCurve(FloatCurve=DataInterface() {propertyValues {Curve {keys [<at 0 0 RCIM_Cubic>, {arriveTangent=3.4223, interpMode=RCIM_Cubic, leaveTangent=3.4223, tangentMode=RCTM_User, time=1, value=1}]}}})
    Index 1 Param 4=&FloatFromCurve(
     FloatCurve=DataInterface() {propertyValues {Curve {keys [
         {interpMode=RCIM_Cubic, leaveTangent=0.1169, tangentMode=RCTM_Break, time=0, value=0}
         {arriveTangent=2.5708, interpMode=RCIM_Cubic, leaveTangent=2.5708, tangentMode=RCTM_User, time=1, value=1}
        ]}}}
     Scale Curve=&RandomRangeFloat(Minimum=1, Maximum=40)
    )
    Index 2 Param 1=&FloatFromCurve(FloatCurve=DataInterface() {propertyValues {Curve {keys [
         {interpMode=RCIM_Cubic, leaveTangent=-0.0025, tangentMode=RCTM_Break, time=0, value=0.5}
         {arriveTangent=-0.0316, interpMode=RCIM_Cubic, tangentMode=RCTM_Break, time=1, value=-0.2}
        ]}}})
    Index 2 Param 2=&FloatFromCurve(
     FloatCurve=DataInterface() {propertyValues {Curve {keys [
         {interpMode=RCIM_Cubic, tangentMode=RCTM_User, time=0, value=0}
         {arriveTangent=2.0522, interpMode=RCIM_Cubic, leaveTangent=2.0522, tangentMode=RCTM_User, time=1, value=1}
        ]}}}
     Scale Curve=2
    )
    Module.Index0Param0Write=true
    Module.Index0Param1Write=true
    Module.Index0Param2Write=true
    Module.Index0Param3Write=true
    Module.Index1Param0Write=true
    Module.Index1Param1Write=true
    Module.Index1Param2Write=true
    Module.Index1Param3Write=true
    Module.Index2Param0Write=true
    Module.Index2Param1Write=true
    Module.Index2Param2Write=false
    Module.Index2Param3Write=false
    Module.Index3Param0Write=false
    Module.Index3Param1Write=false
    Module.Index3Param2Write=false
    Module.Index3Param3Write=true
    Write Parameter Index 1=true
    Write Parameter Index 2=true
    Write Parameter Index 3=true
   )
   &ScaleColor(
    Scale Alpha=&FloatFromCurve(FloatCurve=DataInterface() {propertyValues {Curve {keys [
         {tangentMode=RCTM_User, time=0, value=1}
         {arriveTangent=-0.0405, interpMode=RCIM_Cubic, leaveTangent=-0.0405, tangentMode=RCTM_User, time=0.7009, value=1}
         {arriveTangent=-8.0089, interpMode=RCIM_Cubic, leaveTangent=-8.0089, tangentMode=RCTM_User, time=1, value=0}
        ]}}})
    Scale RGB=&Lerp_Vector(
     A=&VectorFromFloat(Value=&FloatFromCurve(
       FloatCurve=DataInterface() {propertyValues {Curve {keys [<at 0 1>,<at 0.1 1>,<at 0.2 2>,<at 0.3 1>,<at 0.6 1>,<at 0.7 2>,<at 0.8 1>,<at 1 1>]}}}
       Scale Curve=10
      ))
     B=&VectorFromFloat(Value=&RandomRangeFloat(Minimum=0.3, Maximum=5, Evaluation Type=::NewEnumerator1))
     Alpha=&FloatFromCurve(FloatCurve=DataInterface() {propertyValues {Curve {keys [
          {interpMode=RCIM_Cubic, leaveTangent=2.5294, tangentMode=RCTM_Break, time=0, value=0}
          {arriveTangent=-0.2166, interpMode=RCIM_Cubic, tangentMode=RCTM_Break, time=0.4, value=1}
         ]}}})
    )
    ScaleA=true
    ScaleRGB=true
   )
  ]
  rendererClasses [&NiagaraMeshRendererProperties]
  renderers [
   {rendererClass=&NiagaraMeshRendererProperties, rendererIndex=0}
   {rendererClass=&NiagaraMeshRendererProperties, rendererIndex=1}
  ]
 }
}
summary {
 systemName=NS_PiewPistol_ADS_ChargingUp
 emitters [
  {emitterName=Lightning, rendererClasses [&NiagaraMeshRendererProperties]}
  {emitterName=Airflow, rendererClasses [&NiagaraMeshRendererProperties]}
  {emitterName=Corona, rendererClasses [&NiagaraSpriteRendererProperties]}
 ]
 userVariables [User.HueShift {type=NiagaraFloat, defaultValue=0.025, description=""}]
}
