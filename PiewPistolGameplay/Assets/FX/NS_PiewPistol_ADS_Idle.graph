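# Niagara system for the Piew Pistol's aim-down-sights idle charge effect.
# A persistent flickering, distance-scaled flare and light combine with inward-pulled sparkles, mesh lightning arcs and a layered three-sprite core pulse gated by User.bCanSuperShot.
# System timelines ramp sparkle and lightning spawn probability and fade the flare and light.
# User.HueShift recolors core, flare and lightning particle colors, while time-of-day scalars dim the core and flare.
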NiagaraSystem /PiewPistolGameplay/Assets/FX/NS_PiewPistol_ADS_Idle
dependencies {
 usedDataInterfaces [&NiagaraDataInterfaceCurve, &NiagaraDataInterfaceVector2DCurve, &NiagaraDataInterfaceColorCurve]
 usedDynamicInputs [&UniformRangedVector]
 usedModules [
  &SystemState
  &Timeline
  &self:SystemSpawnScript.SystemScriptSource.SystemScriptGraph.NiagaraNodeAssignment_0.SetVariables_#1
 ]
}
emitters {
 Absorb_Sparkles {
  emitterUpdateScript [
   &EmitterState(Life Cycle Mode=::NewEnumerator1, Loop Duration=2, UseLoopDelay=false)
   &SpawnRate(
    Spawn Probability=&Multiply_Float(A=0.35, B=$System.Timeline001.NormalizedTime)
    SpawnRate=35
    Use Spawn Probability=true
   )
  ]
  particleSpawnScript [
   &InitializeParticle(
    Color=<rgba 10 0.2 0.8533 0.8803>
    Lifetime Max=0.4
    Lifetime Min=0.25
    Lifetime Mode=::NewEnumerator1
    Mass Mode=::NewEnumerator3
    Sprite Rotation Mode=::NewEnumerator2
    Sprite Size Max=<2 7.5>
    Sprite Size Min=<1 2.5>
    Sprite Size Mode=::NewEnumerator2
    UsePositionOffset=false
    Write Material Random=true
   )
   &ShapeLocation(
    MaskSpawnGroup=false
    Offset=<0 0 -2>
    Offset Coordinate Space=::Local
    Offset Mode=::NewEnumerator0
    Override Seed=false
    Sphere Radius=15
    Sphere Surface Distribution=1
    UseOwnerScale=true
   )
  ]
  particleUpdateScript [
   &ParticleState()
   &ScaleSpriteSize(
    Non-Uniform Curve Scale=<2 1>
    Non-Uniform Curve Sprite Scale=DataInterface() {propertyValues {
      XCurve {
       defaultValue=0
       keys [
        {arriveTangent=2.3721, interpMode=RCIM_Cubic, leaveTangent=2.3721, tangentMode=RCTM_User, time=0, value=0}
        {arriveTangent=0.0736, interpMode=RCIM_Cubic, leaveTangent=0.0736, tangentMode=RCTM_User, time=0.3386, value=0.5665}
        {arriveTangent=0.2955, interpMode=RCIM_Cubic, leaveTangent=0.2955, tangentMode=RCTM_User, time=0.9977, value=0.8882}
       ]
      }
      YCurve {
       defaultValue=0
       keys [
        {arriveTangent=5.921, interpMode=RCIM_Cubic, leaveTangent=5.921, tangentMode=RCTM_User, time=0, value=0}
        {arriveTangent=1.1716, interpMode=RCIM_Cubic, leaveTangent=1.1716, tangentMode=RCTM_User, time=0.6016, value=0.9088}
        {arriveTangent=-0.0405, interpMode=RCIM_Cubic, leaveTangent=-0.0405, tangentMode=RCTM_User, time=1, value=0}
       ]
      }
     }}
    Scale Sprite Size Mode=::NewEnumerator3
   )
   &PointAttractionForce(
    Attraction Radius=500
    Attraction Strength=&RandomRangeFloat(Minimum=200, Maximum=400)
    Falloff Exponent=0.1
    Kill Within Radius=false
    Module.Use Falloff=true
    UseAttractorOffset=false
   )
   &Drag(Drag=&RandomRangeFloat(Minimum=2, Maximum=5), Use Linear Drag=true, Use Rotational Drag=true)
   &SolveForcesAndVelocity(Clamp Velocity=false, Limit Acceleration=false)
   &ScaleSpriteSizeBySpeed(Max Scale Factor=<0.35 1>, Min Scale Factor=<0.75 0.5>, Velocity Threshold=50)
   &ScaleColor(
    Linear Color Curve=DataInterface() {propertyValues {
      AlphaCurve {keys [<at 0.0117 0>,<at 0.7986 1>,<at 1 0>]}
      BlueCurve {keys [<at 0.0038 0>,<at 0.3079 0.2746>,<at 0.8302 1>,<at 0.9774 10>]}
      GreenCurve {keys [<at 0.0038 0>,<at 0.3079 0.2746>,<at 0.8302 1>,<at 0.9774 10>]}
      RedCurve {keys [<at 0.0038 0>,<at 0.3079 0.2746>,<at 0.8302 1>,<at 0.9774 10>]}
     }}
    Scale Mode=::NewEnumerator2
   )
  ]
  rendererClasses [&NiagaraSpriteRendererProperties]
  renderers [{rendererClass=&NiagaraSpriteRendererProperties}]
 }
 Core {
  emitterUpdateScript [
   &EmitterState(Life Cycle Mode=::NewEnumerator1, Loop Behavior=::NewEnumerator1, Loop Delay=0.075, Loop Duration=1, UseLoopDelay=true)
   &SpawnBurst_Instantaneous(
    Spawn Count=1
    Spawn Probability=&MakeCustomFloatFromBool(Bool=$User.bCanSuperShot, True Float=1, False Float=0)
    Use Spawn Probability=true
   )
  ]
  particleSpawnScript [
   &InitializeParticle(Lifetime=0.5, Position Offset=<0 0 -1>, Position Offset Coordinate Space=::Local, UsePositionOffset=true, Write Material Random=true)
   /SetVariables(
    Particles.Sprite_Color_01:LinearColor=&HueShiftLinearColor(Hue=$User.HueShift, Saturation=1, Value=1, Linear Color=<rgba 2.5 0.125 0.1646 1>)
    Particles.Sprite_DP_01:Vector4f=&MakeVector4(X=&RandomRangeFloat(Minimum=0.4, Maximum=0.6), Y=1, Z=1, W=1)
    Particles.Sprite_Size_01:Vector2f=&Vector2DFromFloat(Value=150)
   ) {isSetParameters=true}
   /SetVariables(
    Particles.Sprite_Color_02:LinearColor=&HueShiftLinearColor(Hue=$User.HueShift, Saturation=1, Value=1, Linear Color=<rgba 0.276 0 0 0.3>)
    Particles.Sprite_DP_02:Vector4f=&MakeVector4(X=&RandomRangeFloat(Minimum=0.4, Maximum=0.6), Y=1, Z=1, W=1)
    Particles.Sprite_Size_02:Vector2f=&Vector2DFromFloat(Value=250)
   ) {isSetParameters=true}
   /SetVariables(
    Particles.Sprite_Color_03:LinearColor=&HueShiftLinearColor(Hue=$User.HueShift, Saturation=1, Value=1, Linear Color=<rgba 1 0 0 3.5>)
    Particles.Sprite_Size_03:Vector2f=&Vector2DFromFloat(Value=150)
   ) {isSetParameters=true}
  ]
  particleUpdateScript [
   &ParticleState()
   &CameraOffset(Camera Offset Amount=10)
   /SetVariables(
    Particles.Sprite_Color_01:LinearColor=&MultiplyLinearColor(
     Linear Color=$Particles.Initial.Sprite_Color_01
     Scale Factor=&MakeFloat4FromVectorAndFloat(
      "Vector (RGB)"=<1 1 1>
      "Float (Alpha)"=&Multiply_Float(
       A=&FNTimeOfDayScalar(Night Scale=0.5, Morning Scale=0.3, Evening Scale=0.4)
       B=&FloatFromCurve(
        FloatCurve=DataInterface() {propertyValues {Curve {keys [<at 0 0>,<at 0.2 1>,<at 0.8 1>,<at 1 0>]}}}
        CurveIndex=&Multiply_Float(A=$Particles.NormalizedAge, B=1.75)
       )
      )
     )
    )
    Particles.Sprite_Size_01:Vector2f=&Multiply_Vector2D(
     A=$Particles.Initial.Sprite_Size_01
     B=&Vector2DFromFloat(Value=&FloatFromCurve(
       FloatCurve=DataInterface() {propertyValues {Curve {keys [<at 0 1.5 RCIM_Cubic>,<at 1 0>]}}}
       CurveIndex=&Multiply_Float(A=$Particles.NormalizedAge, B=1.75)
      ))
    )
   ) {isSetParameters=true}
   /SetVariables(
    Particles.Sprite_Color_02:LinearColor=&MultiplyLinearColor(
     Linear Color=$Particles.Initial.Sprite_Color_02
     Scale Factor=&MakeFloat4FromVectorAndFloat(
      "Vector (RGB)"=&Multiply_VectorByFloat(Vector=<1 1 1>, Float=&FNTimeOfDayScalar(Night Scale=0.5, Morning Scale=0.3, Evening Scale=0.4))
      "Float (Alpha)"=&FloatFromCurve(
       FloatCurve=DataInterface() {propertyValues {Curve {keys [<at 0 0>,<at 0.2 1>,<at 0.8 1>,<at 1 0>]}}}
       CurveIndex=&Multiply_Float(A=$Particles.NormalizedAge, B=1.75)
      )
     )
    )
    Particles.Sprite_Size_02:Vector2f=&Multiply_Vector2D(
     A=$Particles.Initial.Sprite_Size_02
     B=&Vector2DFromFloat(Value=&FloatFromCurve(
       FloatCurve=DataInterface() {propertyValues {Curve {keys [<at 0 1.5 RCIM_Cubic>,<at 1 0>]}}}
       CurveIndex=&Multiply_Float(A=$Particles.NormalizedAge, B=1.75)
      ))
    )
   ) {isSetParameters=true}
   /SetVariables(
    Particles.Sprite_Color_03:LinearColor=&MultiplyLinearColor(
     Linear Color=$Particles.Initial.Sprite_Color_03
     Scale Factor=&MakeFloat4FromVectorAndFloat(
      "Vector (RGB)"=<1 1 1>
      "Float (Alpha)"=&Multiply_Float(
       A=&FloatFromCurve(
        FloatCurve=DataInterface() {propertyValues {Curve {keys [{interpMode=RCIM_Cubic, leaveTangent=-0.6434, tangentMode=RCTM_Break, time=0, value=1}, <at 1 0>]}}}
        CurveIndex=&Multiply_Float(A=$Particles.NormalizedAge, B=2)
       )
       B=&FNTimeOfDayScalar(Night Scale=1, Morning Scale=0.3, Evening Scale=0.5)
      )
     )
    )
    Particles.Sprite_Size_03:Vector2f=&Multiply_Vector2D(
     A=$Particles.Initial.Sprite_Size_03
     B=&Vector2DFromFloat(Value=&FloatFromCurve(
       FloatCurve=DataInterface() {propertyValues {Curve {keys [{tangentMode=RCTM_User, time=0, value=0}, <at 1 1>]}}}
       CurveIndex=&Multiply_Float(A=$Particles.NormalizedAge, B=2)
       Scale Curve=0.75
      ))
    )
   ) {isSetParameters=true}
  ]
  rendererClasses [&NiagaraSpriteRendererProperties]
  renderers [
   {rendererClass=&NiagaraSpriteRendererProperties}
   {rendererClass=&NiagaraSpriteRendererProperties, rendererIndex=1}
   {rendererClass=&NiagaraSpriteRendererProperties, rendererIndex=2}
  ]
 }
 Flash_Idle_Core {
  emitterUpdateScript [
   &EmitterState(Life Cycle Mode=::NewEnumerator1, Loop Behavior=::NewEnumerator1, Loop Duration Mode=::NewEnumerator1, UseLoopDelay=false)
   &SpawnBurst_Instantaneous(Use Spawn Probability=false, UseLoopCountLimit=false)
  ]
  particleSpawnScript [
   &InitializeParticle(
    Color=&HueShiftLinearColor(Hue=$User.HueShift, Saturation=1, Value=1, Linear Color=<rgba 2.5 0.175 0.2912 0.8803>)
    Lifetime=0.5
    Mass Mode=::NewEnumerator1
    Mesh Scale Mode=::NewEnumerator3
    Position Offset=<0 0 -1>
    Position Offset Coordinate Space=::Local
    Sprite Size=<20 20>
    Sprite Size Mode=::NewEnumerator4
    UsePositionOffset=true
    Write Material Random=true
   ) {module=InitializeParticle001}
   &CameraOffset(Camera Offset Amount=5)
  ]
  particleUpdateScript [
   &ParticleState(
    Kill Particles When Lifetime Has Elapsed=false
    Let Infinitely Lived Particles Die When Emitter Deactivates=true
    Loop Particles Lifetime=true
   )
   /SetVariables(Particles.DistanceScalar:NiagaraFloat=&Lerp_Float(
     A=1
     B=7.5
     Alpha=&FloatFromCurve(
      FloatCurve=DataInterface() {propertyValues {Curve {
         defaultValue=1
         keys [
          {interpMode=RCIM_Cubic, tangentMode=RCTM_User, time=0, value=0}
          {arriveTangent=0.1871, interpMode=RCIM_Cubic, leaveTangent=0.1871, tangentMode=RCTM_User, time=1, value=1}
         ]
        }}}
      CurveIndex=&FN_LOD_Distance_Remap(Start Distance=0, End Distance=5000, Minimum Output=0, Maximum Output=1, Clamp Output=true)
     )
    )) {isSetParameters=true}
   /Niagara/Noise/Waveform(
    EnableExponent1=false
    GlobalClampMax=false
    GlobalClampMin=false
    UseInput1=false
    UsePhase1=false
    "Waveform [1]":ENiagara_Waveforms=::NewEnumerator12
    "[1] Frequency"=50
   )
   &SolveForcesAndVelocity(Clamp Velocity=false, Limit Acceleration=false)
   &ScaleSpriteSize(
    Uniform Curve Index=$Emitter.Age
    Uniform Curve Scale=$Particles.DistanceScalar
    Uniform Curve Sprite Scale=DataInterface() {propertyValues {Curve {keys [
        {interpMode=RCIM_Cubic, leaveTangentWeight=0.504, tangentMode=RCTM_User, time=0, value=0}
        {arriveTangentWeight=0.1, interpMode=RCIM_Cubic, leaveTangentWeight=0.2333, tangentMode=RCTM_User, time=0.5, value=1}
        {arriveTangent=-0.0037, interpMode=RCIM_Cubic, leaveTangent=-0.0037, tangentMode=RCTM_User, time=0.97, value=1}
        {tangentMode=RCTM_User, time=1, value=1}
       ]}}}
   )
   /SetVariables(Particles.MiddleFlareSize:Vector2f=&Multiply_Vector2D(A=$Particles.SpriteSize, B=&Multiply_Vector2D_ByFloat(Float=$Particles.DistanceScalar, Vector2D=<0.25 0.25>))) {isSetParameters=true}
   /SetVariables(Particles.LensFlareColor:LinearColor=&Multiply_LinearColorByFloat(
     Linear Color=<rgba 75 1.629 13.8575 1>
     Float=&Multiply_Float(A=$System.Timeline001.NormalizedTime, B=$System.Timeline.InverseNormalizedTime)
    )) {isSetParameters=true}
   &FN_Sprite_Attributes(
    Create SpriteSize Attribute=true
    SpriteSize=&Multiply_Vector2D(
     A=<35 35>
     B=&Vector2DFromFloat(Value=&FloatFromCurve(
       FloatCurve=DataInterface() {propertyValues {Curve {
          defaultValue=0
          keys [{interpMode=RCIM_Cubic, tangentMode=RCTM_User, time=0, value=0}, {interpMode=RCIM_Cubic, tangentMode=RCTM_User, time=1, value=1}]
         }}}
       CurveIndex=$Emitter.Age
       Scale Curve=$Particles.DistanceScalar
      ))
    )
    Write to Module Namespace=true
   ) {module=FN_Sprite}
   &FN_Light_Attributes(
    Color=<rgba 1 0 0.1667 1>
    Create Color Attribute=true
    Create Exponent Attribute=true
    Create Position Attribute=true
    Create Radius Attribute=true
    Enable Particle Light Enable=true
    Exponent=5
    Light Enabled=true
    Position:NiagaraPosition=$Engine.Emitter.SimulationPosition
    Radius=35
    Scale Color=&Multiply_Float(
     A=&FloatFromCurve(
      FloatCurve=DataInterface() {propertyValues {Curve {
         defaultValue=0.8
         keys [{interpMode=RCIM_Cubic, tangentMode=RCTM_User, time=0, value=0}, {interpMode=RCIM_Cubic, tangentMode=RCTM_User, time=1, value=1}]
        }}}
      CurveIndex=$Emitter.Age
     )
     B=&FloatFromCurve(
      FloatCurve=DataInterface() {propertyValues {Curve {
         defaultValue=0
         keys [
          {interpMode=RCIM_Cubic, tangentMode=RCTM_User, time=0, value=0.5}
          {arriveTangent=0.2403, interpMode=RCIM_Cubic, leaveTangent=0.2403, time=0.2, value=0.9}
          {arriveTangent=0.3333, interpMode=RCIM_Cubic, leaveTangent=0.3333, time=0.4389, value=0.6055}
          {interpMode=RCIM_Cubic, tangentMode=RCTM_User, time=0.5, value=1}
          {arriveTangent=0.25, interpMode=RCIM_Cubic, leaveTangent=0.25, time=0.6, value=0.8}
          {arriveTangent=-0.75, interpMode=RCIM_Cubic, leaveTangent=-0.75, time=0.7, value=1.05}
          {tangentMode=RCTM_User, time=1, value=0.5}
         ]
        }}}
      CurveIndex=&NormalizeFloat(Value=&Multiply_Float(A=$Particles.NormalizedAge, B=2), Min=0, Max=1)
      Scale Curve=&Multiply_Float(A=25, B=&Multiply_Float(A=$System.Timeline001.NormalizedTime, B=$System.Timeline.InverseNormalizedTime))
     )
    )
   )
   &ScaleColor(Scale RGB=&VectorFromFloat(Value=$System.SMTOD), ScaleA=true, ScaleRGB=true) {module=ScaleColor002}
   &ScaleColor(
    Scale Alpha=&Multiply_Float(A=$System.Timeline001.NormalizedTime, B=$System.Timeline.InverseNormalizedTime)
    Scale RGB=&VectorFromFloat(Value=$Particles.Waveform.Waveform)
    ScaleA=true
    ScaleRGB=true
   )
   &KillParticles(Kill Particles=&Boolean_AndOperation(
     Bool1=&Boolean_NotOperation(Bool=/Game/Effects/Niagara/DynamicInputs/State/FN_IsExecutionStateActive())
     Bool2=&Boolean_NotOperation(Bool=$System.Timeline.IsPlaying)
    ))
  ]
  rendererClasses [&NiagaraLightRendererProperties, &NiagaraSpriteRendererProperties]
  renderers [
   {rendererClass=&NiagaraLightRendererProperties, rendererIndex=0}
   {rendererClass=&NiagaraSpriteRendererProperties, rendererIndex=1}
   {rendererClass=&NiagaraSpriteRendererProperties, rendererIndex=2}
   {rendererClass=&NiagaraSpriteRendererProperties, rendererIndex=3}
  ]
 }
 Lightning {
  emitterUpdateScript [
   &EmitterState(
    Inactive Response=::NewEnumerator1
    Life Cycle Mode=::NewEnumerator1
    Loop Behavior=::NewEnumerator1
    Loop Duration Mode=::NewEnumerator1
    UseLoopDelay=false
   )
   &SpawnRate(
    Spawn Probability=&FloatFromCurve(
     FloatCurve=DataInterface() {propertyValues {Curve {
        defaultValue=1
        keys [{interpMode=RCIM_Cubic, tangentMode=RCTM_User, time=0, value=0}, {interpMode=RCIM_Cubic, tangentMode=RCTM_User, time=1, value=1}]
       }}}
     CurveIndex=$Emitter.Age
     Scale Curve=$System.Timeline001.NormalizedTime
    )
    SpawnRate=5
    Use Spawn Probability=true
   )
  ]
  particleSpawnScript [
   &FN_SetSpawnGroup(Number of Spawn Groups=2)
   &InitializeParticle(
    Color=&HueShiftLinearColor(
     Hue=$User.HueShift
     Saturation=1
     Value=1
     Linear Color=&MultiplyLinearColor(
      Linear Color=&RandomRangeLinearColor(Minimum=<rgba 1.8 0 0 0.3>, Maximum=<rgba 1.8 0 0 1>)
      Scale Factor=&MakeVector4_FromFloat(Float=0.8)
     )
    )
    Lifetime Max=0.35
    Lifetime Min=0.175
    Lifetime Mode=::NewEnumerator1
    Mesh Scale Max=<1 1 1>
    Mesh Scale Min=<0.5 0.5 0.5>
    Mesh Scale Mode=::NewEnumerator2
    Position Offset=&MakeVector(
     X=2.5
     Y=5.75
     Z=&FloatFromCurve(
      FloatCurve=DataInterface() {propertyValues {Curve {
         defaultValue=1
         keys [
          {arriveTangent=1.7743, interpMode=RCIM_Cubic, leaveTangent=1.7743, tangentMode=RCTM_User, time=0, value=0}
          {arriveTangent=0.7519, interpMode=RCIM_Cubic, leaveTangent=0.7519, tangentMode=RCTM_User, time=1, value=1}
         ]
        }}}
      CurveIndex=$Emitter.NormalizedLoopAge
      Scale Curve=&MaskFloatBySpawnGroup(Masked Float=-80, Passthrough Float=50)
     )
    )
    UsePositionOffset=true
    Write Material Random=true
   )
   &ShapeLocation(MaskSpawnGroup=false, Override Seed=false, Sphere Radius=5, UseOwnerScale=true)
   &InitialMeshOrientation(Rotation=&RandomRangeVector(Minimum=<0 0 -1>, Maximum=<0.1 0.1 1>), Use Rotation Vector=true)
  ]
  particleUpdateScript [
   &SolveForcesAndVelocity(Clamp Velocity=false, Limit Acceleration=false)
   &ParticleState()
   &ScaleMeshSize(Scale Factor=&Multiply_VectorByFloat(Vector=<1 1 1>, Float=0.25))
   &DynamicMaterialParameters(
    Index 0 Param 1=&MakeFloatFromVector(Vector=$Particles.Scale, Channel=::NewEnumerator0)
    Index 0 Param 2=&RandomRangeFloat(Minimum=0, Maximum=10)
    Index 0 Param 3=&FloatFromCurve(
     FloatCurve=DataInterface() {propertyValues {Curve {keys [{interpMode=RCIM_Cubic, leaveTangent=8.1085, tangentMode=RCTM_Break, time=0.8568, value=1}, {tangentMode=RCTM_User, time=1, value=0}]}}}
     Scale Curve=0.07
    )
    Index 0 Param 4=0.8
    Index 1 Param 1=&FloatFromCurve(FloatCurve=DataInterface() {propertyValues {Curve {keys [
         {interpMode=RCIM_Cubic, leaveTangent=1.1034, tangentMode=RCTM_Break, time=0, value=0}
         {arriveTangent=2.2399, tangentMode=RCTM_User, time=0.5, value=1}
        ]}}})
    Index 1 Param 2=&FloatFromCurve(FloatCurve=DataInterface() {propertyValues {Curve {keys [<at 0.5 0>, {arriveTangent=3.2053, tangentMode=RCTM_User, time=0.9, value=1}]}}})
    Index 1 Param 3=&FloatFromCurve(FloatCurve=DataInterface() {propertyValues {Curve {keys [<at 0 0 RCIM_Cubic>, {arriveTangent=3.4223, interpMode=RCIM_Cubic, leaveTangent=3.4223, tangentMode=RCTM_User, time=1, value=1}]}}})
    Index 1 Param 4=&FloatFromCurve(
     FloatCurve=DataInterface() {propertyValues {Curve {keys [
         {interpMode=RCIM_Cubic, leaveTangent=0.1169, tangentMode=RCTM_Break, time=0, value=0}
         {arriveTangent=2.5708, interpMode=RCIM_Cubic, leaveTangent=2.5708, tangentMode=RCTM_User, time=1, value=1}
        ]}}}
     Scale Curve=&RandomRangeFloat(Minimum=1, Maximum=40)
    )
    Index 2 Param 1=&FloatFromCurve(FloatCurve=DataInterface() {propertyValues {Curve {keys [
         {interpMode=RCIM_Cubic, leaveTangent=-0.0025, tangentMode=RCTM_Break, time=0, value=0.5}
         {arriveTangent=-0.0316, interpMode=RCIM_Cubic, tangentMode=RCTM_Break, time=1, value=-0.2}
        ]}}})
    Index 2 Param 2=&FloatFromCurve(
     FloatCurve=DataInterface() {propertyValues {Curve {keys [
         {interpMode=RCIM_Cubic, tangentMode=RCTM_User, time=0, value=0}
         {arriveTangent=2.0522, interpMode=RCIM_Cubic, leaveTangent=2.0522, tangentMode=RCTM_User, time=1, value=1}
        ]}}}
     Scale Curve=2
    )
    Module.Index0Param0Write=true
    Module.Index0Param1Write=true
    Module.Index0Param2Write=true
    Module.Index0Param3Write=true
    Module.Index1Param0Write=true
    Module.Index1Param1Write=true
    Module.Index1Param2Write=true
    Module.Index1Param3Write=true
    Module.Index2Param0Write=true
    Module.Index2Param1Write=true
    Module.Index2Param2Write=false
    Module.Index2Param3Write=false
    Module.Index3Param0Write=false
    Module.Index3Param1Write=false
    Module.Index3Param2Write=false
    Module.Index3Param3Write=true
    Write Parameter Index 1=true
    Write Parameter Index 2=true
    Write Parameter Index 3=true
   )
   &UpdateMeshOrientation(Rotation Rate=&RandomRangeFloat(Minimum=-1, Maximum=1), Rotation Vector=<0 0 1>)
   &ScaleColor(
    Scale Alpha=&FloatFromCurve(FloatCurve=DataInterface() {propertyValues {Curve {keys [
         {tangentMode=RCTM_User, time=0, value=1}
         {arriveTangent=-0.0405, interpMode=RCIM_Cubic, leaveTangent=-0.0405, tangentMode=RCTM_User, time=0.7009, value=1}
         {arriveTangent=-8.0089, interpMode=RCIM_Cubic, leaveTangent=-8.0089, tangentMode=RCTM_User, time=1, value=0}
        ]}}})
    Scale RGB=&Lerp_Vector(
     A=&VectorFromFloat(Value=&FloatFromCurve(
       FloatCurve=DataInterface() {propertyValues {Curve {keys [<at 0 1>,<at 0.1 1>,<at 0.2 2>,<at 0.3 1>,<at 0.6 1>,<at 0.7 2>,<at 0.8 1>,<at 1 1>]}}}
       Scale Curve=10
      ))
     B=&VectorFromFloat(Value=&RandomRangeFloat(Minimum=0.3, Maximum=5, Evaluation Type=::NewEnumerator1))
     Alpha=&FloatFromCurve(FloatCurve=DataInterface() {propertyValues {Curve {keys [
          {interpMode=RCIM_Cubic, leaveTangent=2.5294, tangentMode=RCTM_Break, time=0, value=0}
          {arriveTangent=-0.2166, interpMode=RCIM_Cubic, tangentMode=RCTM_Break, time=0.4, value=1}
         ]}}})
    )
    ScaleA=true
    ScaleRGB=true
   )
  ]
  rendererClasses [&NiagaraMeshRendererProperties]
  renderers [
   {rendererClass=&NiagaraMeshRendererProperties, rendererIndex=0}
   {rendererClass=&NiagaraMeshRendererProperties, rendererIndex=1}
  ]
 }
 SplitLightning_ArcsFromHands {
  emitterUpdateScript [
   &EmitterState(Life Cycle Mode=::NewEnumerator1, Loop Behavior=::NewEnumerator1, Loop Duration Mode=::NewEnumerator1, UseLoopDelay=false)
   &SpawnRate(
    Spawn Probability=&FloatFromCurve(
     FloatCurve=DataInterface() {propertyValues {Curve {
        defaultValue=1
        keys [{interpMode=RCIM_Cubic, tangentMode=RCTM_User, time=0, value=0}, {interpMode=RCIM_Cubic, tangentMode=RCTM_User, time=1, value=1}]
       }}}
     CurveIndex=$Emitter.Age
     Scale Curve=$System.Timeline001.NormalizedTime
    )
    SpawnRate=5
    Use Spawn Probability=true
   )
  ]
  particleSpawnScript [
   &FN_SetSpawnGroup(Number of Spawn Groups=2)
   &InitializeParticle(
    Color=&HueShiftLinearColor(
     Hue=$User.HueShift
     Saturation=1
     Value=1
     Linear Color=&MultiplyLinearColor(
      Linear Color=&RandomRangeLinearColor(Minimum=<rgba 1.8 0 0 0.3>, Maximum=<rgba 1.8 0 0 1>)
      Scale Factor=&MakeVector4_FromFloat(Float=0.8)
     )
    )
    Lifetime=&RandomRangeFloat(Minimum=0.175, Maximum=0.25)
    Mesh Scale Mode=::NewEnumerator3
    Mesh Uniform Scale=&RandomRangeFloat(Minimum=0.075, Maximum=0.125)
    Position Offset=&MakeVector(
     X=2.5
     Y=5.75
     Z=&FloatFromCurve(
      FloatCurve=DataInterface() {propertyValues {Curve {
         defaultValue=1
         keys [
          {arriveTangent=2.7764, interpMode=RCIM_Cubic, leaveTangent=2.7764, tangentMode=RCTM_User, time=0, value=0}
          {arriveTangent=0.3042, interpMode=RCIM_Cubic, leaveTangent=0.3042, tangentMode=RCTM_User, time=1, value=1}
         ]
        }}}
      CurveIndex=$Emitter.NormalizedLoopAge
      Scale Curve=&MaskFloatBySpawnGroup(Masked Float=-70, Passthrough Float=50)
     )
    )
    UsePositionOffset=true
    Write Material Random=true
   )
   &ShapeLocation(MaskSpawnGroup=false, Override Seed=false, Sphere Radius=5, UseOwnerScale=true)
   &InitialMeshOrientation(Rotation=&RandomRangeVector(Minimum=<0 0 -1>, Maximum=<0.1 0.1 1>), Use Rotation Vector=true)
  ]
  particleUpdateScript [
   &ParticleState()
   &ScaleMeshSize(Scale Factor=&RandomRangeVector(Minimum=<1 1 2.5>, Maximum=<1.3 1.3 5>)) {module=ScaleMeshSize001}
   &DynamicMaterialParameters(
    Index 0 Param 4=13
    Index 1 Param 1=&FloatFromCurve(FloatCurve=DataInterface() {propertyValues {Curve {keys [<at 0 0.5>,<at 1 1>]}}})
    Index 1 Param 2=2
    Index 1 Param 3=&FloatFromCurve(FloatCurve=DataInterface() {propertyValues {Curve {keys [{interpMode=RCIM_Cubic, tangentMode=RCTM_Break, time=0.5, value=-0.5}, <at 1 0.1>]}}})
    Index 2 Param 1=&Multiply_Float(
     A=&FloatFromCurve(FloatCurve=DataInterface() {propertyValues {Curve {keys [{interpMode=RCIM_Cubic, leaveTangent=8.1085, tangentMode=RCTM_Break, time=0.8568, value=1}, {tangentMode=RCTM_User, time=1, value=0}]}}})
     B=&Divide_Float(
      A=&MakeFloatFromVector(Vector=$Particles.Scale, Channel=::NewEnumerator0)
      B=&RandomRangeFloat(Minimum=10, Maximum=15)
     )
    )
    Index 2 Param 2=&FloatFromCurve(
     FloatCurve=DataInterface() {propertyValues {Curve {keys [
         {interpMode=RCIM_Cubic, tangentMode=RCTM_Break, time=0.7, value=0}
         {arriveTangent=-0.0859, interpMode=RCIM_Cubic, tangentMode=RCTM_Break, time=1, value=-1}
        ]}}}
     Scale Curve=50
    )
    Index 2 Param 3=/Niagara/DynamicInputs/Helpers/Waveform(
     Global Input Position=&RandomRangeFloat(Minimum=0, Maximum=15)
     Global Phase=$Particles.Age
     "Waveform [1]"=::NewEnumerator12
     "[1] Phase Scale"=2
     "[1] Frequency"=1
     "[1] Amplitude Min/Max"=<0 1>
    )
    Index 2 Param 4=25
    Index 3 Param 1=&FloatFromCurve(
     FloatCurve=DataInterface() {propertyValues {Curve {keys [
         {interpMode=RCIM_Cubic, leaveTangent=-0.0025, tangentMode=RCTM_Break, time=0, value=0}
         {arriveTangent=1.8528, interpMode=RCIM_Cubic, tangentMode=RCTM_Break, time=1, value=1}
        ]}}}
     Scale Curve=&RandomRangeFloat(Minimum=0, Maximum=0)
    )
    Index 3 Param 2=&FloatFromCurve(
     FloatCurve=DataInterface() {propertyValues {Curve {keys [
         {interpMode=RCIM_Cubic, leaveTangent=-0.0025, tangentMode=RCTM_Break, time=0, value=0.5}
         {arriveTangent=-0.0316, interpMode=RCIM_Cubic, tangentMode=RCTM_Break, time=1, value=-0.2}
        ]}}}
     Scale Curve=0.1
    )
    Index 3 Param 3=&FloatFromCurve(FloatCurve=DataInterface() {propertyValues {Curve {keys [
         {interpMode=RCIM_Cubic, leaveTangent=1.1034, tangentMode=RCTM_Break, time=0, value=-1}
         {arriveTangent=2.2399, tangentMode=RCTM_User, time=0.7, value=0.125}
        ]}}})
    Index 3 Param 4=&FloatFromCurve(FloatCurve=DataInterface() {propertyValues {Curve {keys [
         {interpMode=RCIM_Cubic, leaveTangent=3.0601, tangentMode=RCTM_Break, time=0.5, value=-0.5}
         {arriveTangent=3.2053, tangentMode=RCTM_User, time=1, value=1}
        ]}}})
    Module.Index0Param0Write=true
    Module.Index0Param1Write=true
    Module.Index0Param2Write=true
    Module.Index0Param3Write=true
    Module.Index1Param0Write=true
    Module.Index1Param1Write=true
    Module.Index1Param2Write=true
    Module.Index1Param3Write=true
    Module.Index2Param0Write=true
    Module.Index2Param1Write=true
    Module.Index2Param2Write=true
    Module.Index2Param3Write=true
    Module.Index3Param0Write=true
    Module.Index3Param1Write=true
    Module.Index3Param2Write=true
    Module.Index3Param3Write=true
    Write Parameter Index 1=true
    Write Parameter Index 2=true
    Write Parameter Index 3=true
   )
   &SolveForcesAndVelocity(Clamp Velocity=false, Limit Acceleration=false)
   &ScaleColor(
    Scale Alpha=&FloatFromCurve(FloatCurve=DataInterface() {propertyValues {Curve {keys [
         {tangentMode=RCTM_User, time=0, value=1}
         {arriveTangent=-0.0405, interpMode=RCIM_Cubic, leaveTangent=-0.0405, tangentMode=RCTM_User, time=0.7009, value=1}
         {arriveTangent=-8.0089, interpMode=RCIM_Cubic, leaveTangent=-8.0089, tangentMode=RCTM_User, time=1, value=0}
        ]}}})
    Scale RGB=&Lerp_Vector(
     A=&VectorFromFloat(Value=&FloatFromCurve(FloatCurve=DataInterface() {propertyValues {Curve {keys [<at 0.2771 1>,<at 0.3786 1>]}}}, Scale Curve=25))
     B=&VectorFromFloat(Value=&RandomRangeFloat(Minimum=0.3, Maximum=2, Evaluation Type=::NewEnumerator1))
     Alpha=&FloatFromCurve(FloatCurve=DataInterface() {propertyValues {Curve {keys [
          {interpMode=RCIM_Cubic, leaveTangent=2.5294, tangentMode=RCTM_Break, time=0, value=0}
          {arriveTangent=-0.2166, interpMode=RCIM_Cubic, tangentMode=RCTM_Break, time=0.4, value=1}
         ]}}})
    )
    ScaleA=true
    ScaleRGB=true
   )
  ]
  rendererClasses [&NiagaraMeshRendererProperties]
  renderers [
   {rendererClass=&NiagaraMeshRendererProperties, rendererIndex=0}
   {rendererClass=&NiagaraMeshRendererProperties, rendererIndex=1}
  ]
 }
}
summary {
 systemName=NS_PiewPistol_ADS_Idle
 emitters [
  {emitterName=Lightning, rendererClasses [&NiagaraMeshRendererProperties]}
  {emitterName=SplitLightning_ArcsFromHands, rendererClasses [&NiagaraMeshRendererProperties]}
  {
   emitterName=Flash_Idle_Core
   rendererClasses [&NiagaraLightRendererProperties, &NiagaraSpriteRendererProperties]
  }
  {emitterName=Core, rendererClasses [&NiagaraSpriteRendererProperties]}
  {emitterName=Absorb_Sparkles, rendererClasses [&NiagaraSpriteRendererProperties]}
 ]
 userVariables [
  User.bCanSuperShot {type=NiagaraBool, defaultValue=true}
  User.HueShift {type=NiagaraFloat, defaultValue=0.025, description=""}
 ]
}
