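# Niagara effect for the PiewPistol overheat reload, combining two mesh-based lightning emitters with rising sprite steam.
# Lightning and SplitLightning_ArcsFromHands tint a red base by User.HueShift, slide their spawn offset along X with system age, and taper spawn rate over the loop.
# Curve-driven dynamic material parameters animate the lightning meshes, with color scaling for bright flashes and fades.
# Steam uses SubUV sprites with drag, upward gravity and a declining spawn rate.
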NiagaraSystem /PiewPistolGameplay/Assets/FX/NS_PiewPistol_Overheat_Reload
dependencies {
 usedDataInterfaces [&NiagaraDataInterfaceCurve]
 usedDynamicInputs [&FNTimeOfDayScalar]
 usedModules [
  &SystemState
  &BlendSpaceHelper
  &self:SystemSpawnScript.SystemScriptSource.SystemScriptGraph.NiagaraNodeAssignment_0.SetVariables_#1
 ]
}
emitters {
 Lightning {
  emitterUpdateScript [
   &EmitterState()
   &SpawnRate(
    SpawnRate=&FloatFromCurve(
     FloatCurve=DataInterface() {propertyValues {Curve {
        defaultValue=1
        keys [
         {tangentMode=RCTM_User, time=0, value=1}
         {arriveTangent=-0.0405, interpMode=RCIM_Cubic, leaveTangent=-0.0405, tangentMode=RCTM_User, time=0.5, value=1}
         {arriveTangent=-0.0508, interpMode=RCIM_Cubic, leaveTangent=-0.0508, tangentMode=RCTM_User, time=1, value=0}
        ]
       }}}
     CurveIndex=$Emitter.NormalizedLoopAge
     Scale Curve=4
    )
    Use Spawn Probability=false
   )
  ]
  particleSpawnScript [
   &FN_SetSpawnGroup(Number of Spawn Groups=2)
   &InitializeParticle(
    Color=&HueShiftLinearColor(
     Hue=$User.HueShift
     Saturation=1
     Value=1
     Linear Color=&MultiplyLinearColor(
      Linear Color=&RandomRangeLinearColor(Minimum=<rgba 1.8 0 0 0.3>, Maximum=<rgba 1.8 0 0 1>)
      Scale Factor=&MakeVector4_FromFloat(Float=0.8)
     )
    )
    Lifetime Max=0.5
    Lifetime Min=0.25
    Lifetime Mode=::NewEnumerator1
    Mesh Scale Max=<1 1 1>
    Mesh Scale Min=<0.5 0.5 0.5>
    Mesh Scale Mode=::NewEnumerator2
    UsePositionOffset=false
    Write Material Random=true
   )
   &ShapeLocation(
    Cylinder Height=0
    Cylinder Height Midpoint=0
    Cylinder Radius=10
    MaskSpawnGroup=false
    Offset=&MakeVector(
     X=&Lerp_Float(
      A=-40
      B=5
      Alpha=&FloatFromCurve(
       FloatCurve=DataInterface() {propertyValues {Curve {
          defaultValue=1
          keys [
           {arriveTangent=0.9497, interpMode=RCIM_Cubic, leaveTangent=0.9497, tangentMode=RCTM_User, time=0, value=0.0111}
           {arriveTangent=0.9656, interpMode=RCIM_Cubic, leaveTangent=0.9656, tangentMode=RCTM_User, time=0.4391, value=0.4191}
           {arriveTangent=0.3007, interpMode=RCIM_Cubic, leaveTangent=0.3007, tangentMode=RCTM_User, time=2, value=1}
          ]
         }}}
       CurveIndex=$Engine.System.Age
      )
     )
     Y=0
     Z=0
    )
    Offset Mode=::NewEnumerator0
    Override Seed=false
    Rotation Mode=::NewEnumerator1
    Shape Primitive=::NewEnumerator1
    Surface Only=false
    UseOwnerScale=true
   )
   &InitialMeshOrientation(
    Mesh Orientation Mode=::NewEnumerator2
    Orientation Vector=<0 0 1>
    Rotation=&RandomRangeVector(Minimum=<-0.1 -0.1 -1>, Maximum=<0.1 0.1 1>)
    Use Rotation Vector=true
   )
   &AddVelocity(Velocity=<-1 0 0>)
  ]
  particleUpdateScript [
   &SolveForcesAndVelocity(Clamp Velocity=false, Limit Acceleration=false)
   &ParticleState()
   &ScaleMeshSize(Scale Factor=&Multiply_VectorByFloat(Vector=<1 1 1>, Float=0.2))
   &DynamicMaterialParameters(
    Index 0 Param 1=&MakeFloatFromVector(Vector=$Particles.Scale, Channel=::NewEnumerator0)
    Index 0 Param 2=&RandomRangeFloat(Minimum=0, Maximum=10)
    Index 0 Param 3=&FloatFromCurve(
     FloatCurve=DataInterface() {propertyValues {Curve {keys [{interpMode=RCIM_Cubic, leaveTangent=8.1085, tangentMode=RCTM_Break, time=0.8568, value=1}, {tangentMode=RCTM_User, time=1, value=0}]}}}
     Scale Curve=0.07
    )
    Index 0 Param 4=0.8
    Index 1 Param 1=&FloatFromCurve(FloatCurve=DataInterface() {propertyValues {Curve {keys [
         {interpMode=RCIM_Cubic, leaveTangent=1.1034, tangentMode=RCTM_Break, time=0, value=0}
         {arriveTangent=2.2399, tangentMode=RCTM_User, time=0.5, value=1}
        ]}}})
    Index 1 Param 2=&FloatFromCurve(FloatCurve=DataInterface() {propertyValues {Curve {keys [<at 0.5 0>, {arriveTangent=3.2053, tangentMode=RCTM_User, time=0.9, value=1}]}}})
    Index 1 Param 3=&FloatFromCurve(FloatCurve=DataInterface() {propertyValues {Curve {keys [<at 0 0 RCIM_Cubic>, {arriveTangent=3.4223, interpMode=RCIM_Cubic, leaveTangent=3.4223, tangentMode=RCTM_User, time=1, value=1}]}}})
    Index 1 Param 4=&FloatFromCurve(
     FloatCurve=DataInterface() {propertyValues {Curve {keys [
         {interpMode=RCIM_Cubic, leaveTangent=0.1169, tangentMode=RCTM_Break, time=0, value=0}
         {arriveTangent=2.5708, interpMode=RCIM_Cubic, leaveTangent=2.5708, tangentMode=RCTM_User, time=1, value=1}
        ]}}}
     Scale Curve=&RandomRangeFloat(Minimum=1, Maximum=40)
    )
    Index 2 Param 1=&FloatFromCurve(FloatCurve=DataInterface() {propertyValues {Curve {keys [
         {interpMode=RCIM_Cubic, leaveTangent=-0.0025, tangentMode=RCTM_Break, time=0, value=0.5}
         {arriveTangent=-0.0316, interpMode=RCIM_Cubic, tangentMode=RCTM_Break, time=1, value=-0.2}
        ]}}})
    Index 2 Param 2=&FloatFromCurve(
     FloatCurve=DataInterface() {propertyValues {Curve {keys [
         {interpMode=RCIM_Cubic, tangentMode=RCTM_User, time=0, value=0}
         {arriveTangent=2.0522, interpMode=RCIM_Cubic, leaveTangent=2.0522, tangentMode=RCTM_User, time=1, value=1}
        ]}}}
     Scale Curve=2
    )
    Module.Index0Param0Write=true
    Module.Index0Param1Write=true
    Module.Index0Param2Write=true
    Module.Index0Param3Write=true
    Module.Index1Param0Write=true
    Module.Index1Param1Write=true
    Module.Index1Param2Write=true
    Module.Index1Param3Write=true
    Module.Index2Param0Write=true
    Module.Index2Param1Write=true
    Module.Index2Param2Write=false
    Module.Index2Param3Write=false
    Module.Index3Param0Write=false
    Module.Index3Param1Write=false
    Module.Index3Param2Write=false
    Module.Index3Param3Write=true
    Write Parameter Index 1=true
    Write Parameter Index 2=true
    Write Parameter Index 3=true
   )
   &ScaleColor(
    Scale Alpha=&Multiply_Float(
     A=&FloatFromCurve(FloatCurve=DataInterface() {propertyValues {Curve {keys [
          {tangentMode=RCTM_User, time=0, value=1}
          {arriveTangent=-0.0405, interpMode=RCIM_Cubic, leaveTangent=-0.0405, tangentMode=RCTM_User, time=0.7009, value=1}
          {arriveTangent=-8.0089, interpMode=RCIM_Cubic, leaveTangent=-8.0089, tangentMode=RCTM_User, time=1, value=0}
         ]}}})
     B=&FloatFromCurve(
      FloatCurve=DataInterface() {propertyValues {Curve {
         defaultValue=0
         keys [
          {arriveTangent=4.1785, interpMode=RCIM_Cubic, leaveTangent=4.1785, tangentMode=RCTM_User, time=0, value=0}
          {arriveTangent=1.1179, interpMode=RCIM_Cubic, leaveTangent=1.1179, tangentMode=RCTM_User, time=0.2042, value=0.6228}
          {arriveTangent=0.0224, interpMode=RCIM_Cubic, leaveTangent=0.0224, tangentMode=RCTM_User, time=1, value=1}
         ]
        }}}
      CurveIndex=$Emitter.Age
     )
    )
    Scale RGB=&Lerp_Vector(
     A=&VectorFromFloat(Value=&FloatFromCurve(
       FloatCurve=DataInterface() {propertyValues {Curve {keys [<at 0 1>,<at 0.1 1>,<at 0.2 2>,<at 0.3 1>,<at 0.6 1>,<at 0.7 2>,<at 0.8 1>,<at 1 1>]}}}
       Scale Curve=10
      ))
     B=&VectorFromFloat(Value=&RandomRangeFloat(Minimum=0.3, Maximum=5, Evaluation Type=::NewEnumerator1))
     Alpha=&FloatFromCurve(FloatCurve=DataInterface() {propertyValues {Curve {keys [
          {interpMode=RCIM_Cubic, leaveTangent=2.5294, tangentMode=RCTM_Break, time=0, value=0}
          {arriveTangent=-0.2166, interpMode=RCIM_Cubic, tangentMode=RCTM_Break, time=0.4, value=1}
         ]}}})
    )
    ScaleA=true
    ScaleRGB=true
   )
  ]
  rendererClasses [&NiagaraMeshRendererProperties]
  renderers [
   {rendererClass=&NiagaraMeshRendererProperties, rendererIndex=0}
   {rendererClass=&NiagaraMeshRendererProperties, rendererIndex=1}
  ]
 }
 SplitLightning_ArcsFromHands {
  emitterUpdateScript [
   &EmitterState()
   &SpawnRate(
    SpawnRate=&FloatFromCurve(
     FloatCurve=DataInterface() {propertyValues {Curve {
        defaultValue=1
        keys [
         {tangentMode=RCTM_User, time=0, value=1}
         {arriveTangent=-0.0405, interpMode=RCIM_Cubic, leaveTangent=-0.0405, tangentMode=RCTM_User, time=0.5, value=1}
         {arriveTangent=-0.0508, interpMode=RCIM_Cubic, leaveTangent=-0.0508, tangentMode=RCTM_User, time=1, value=0}
        ]
       }}}
     CurveIndex=$Emitter.NormalizedLoopAge
     Scale Curve=6
    )
    Use Spawn Probability=false
   )
  ]
  particleSpawnScript [
   &FN_SetSpawnGroup(Number of Spawn Groups=2)
   &InitializeParticle(
    Color=&HueShiftLinearColor(
     Hue=$User.HueShift
     Saturation=1
     Value=1
     Linear Color=&MultiplyLinearColor(
      Linear Color=&RandomRangeLinearColor(Minimum=<rgba 1.8 0 0 0.3>, Maximum=<rgba 1.8 0 0 1>)
      Scale Factor=&MakeVector4_FromFloat(Float=0.8)
     )
    )
    Lifetime=&RandomRangeFloat(Minimum=0.25, Maximum=0.5)
    Mesh Scale Mode=::NewEnumerator3
    Mesh Uniform Scale=&RandomRangeFloat(Minimum=0.075, Maximum=0.125)
    UsePositionOffset=false
    Write Material Random=true
   )
   &ShapeLocation(
    Cylinder Height=0
    Cylinder Height Midpoint=0
    Cylinder Radius=0
    MaskSpawnGroup=false
    Offset=&MakeVector(
     X=&Lerp_Float(
      A=-45
      B=5
      Alpha=&FloatFromCurve(
       FloatCurve=DataInterface() {propertyValues {Curve {
          defaultValue=1
          keys [
           {arriveTangent=0.9497, interpMode=RCIM_Cubic, leaveTangent=0.9497, tangentMode=RCTM_User, time=0, value=0.0111}
           {arriveTangent=0.9656, interpMode=RCIM_Cubic, leaveTangent=0.9656, tangentMode=RCTM_User, time=0.4391, value=0.4191}
           {arriveTangent=0.3007, interpMode=RCIM_Cubic, leaveTangent=0.3007, tangentMode=RCTM_User, time=2, value=1}
          ]
         }}}
       CurveIndex=$Engine.System.Age
      )
     )
     Y=0
     Z=0
    )
    Offset Mode=::NewEnumerator0
    Override Seed=false
    Rotation Mode=::NewEnumerator1
    Shape Primitive=::NewEnumerator1
    Surface Only=false
    UseOwnerScale=true
   )
   &InitialMeshOrientation(Rotation=&RandomRangeVector(Minimum=<0 0 -1>, Maximum=<0.1 0.1 1>), Use Rotation Vector=true)
   &AddVelocity(Velocity=<-1 0 0>)
  ]
  particleUpdateScript [
   &ParticleState()
   &ScaleMeshSize(Scale Factor=&Multiply_VectorByFloat(Vector=&RandomRangeVector(Minimum=<1 1 2.5>, Maximum=<1.3 1.3 5>), Float=0.75)) {module=ScaleMeshSize001}
   &DynamicMaterialParameters(
    Index 0 Param 4=13
    Index 1 Param 1=&FloatFromCurve(FloatCurve=DataInterface() {propertyValues {Curve {keys [<at 0 0.5>,<at 1 1>]}}})
    Index 1 Param 2=2
    Index 1 Param 3=&FloatFromCurve(FloatCurve=DataInterface() {propertyValues {Curve {keys [{interpMode=RCIM_Cubic, tangentMode=RCTM_Break, time=0.5, value=-0.5}, <at 1 0.1>]}}})
    Index 2 Param 1=&Multiply_Float(
     A=&FloatFromCurve(FloatCurve=DataInterface() {propertyValues {Curve {keys [{interpMode=RCIM_Cubic, leaveTangent=8.1085, tangentMode=RCTM_Break, time=0.8568, value=1}, {tangentMode=RCTM_User, time=1, value=0}]}}})
     B=&Divide_Float(
      A=&MakeFloatFromVector(Vector=$Particles.Scale, Channel=::NewEnumerator0)
      B=&RandomRangeFloat(Minimum=10, Maximum=15)
     )
    )
    Index 2 Param 2=&FloatFromCurve(
     FloatCurve=DataInterface() {propertyValues {Curve {keys [
         {interpMode=RCIM_Cubic, tangentMode=RCTM_Break, time=0.7, value=0}
         {arriveTangent=-0.0859, interpMode=RCIM_Cubic, tangentMode=RCTM_Break, time=1, value=-1}
        ]}}}
     Scale Curve=50
    )
    Index 2 Param 3=&Waveform(
     Global Input Position=&RandomRangeFloat(Minimum=0, Maximum=15)
     Global Phase=$Particles.Age
     "Waveform [1]"=::NewEnumerator12
     "[1] Phase Scale"=2
     "[1] Frequency"=1
     "[1] Amplitude Min/Max"=<0 1>
    )
    Index 2 Param 4=25
    Index 3 Param 1=&FloatFromCurve(
     FloatCurve=DataInterface() {propertyValues {Curve {keys [
         {interpMode=RCIM_Cubic, leaveTangent=-0.0025, tangentMode=RCTM_Break, time=0, value=0}
         {arriveTangent=1.8528, interpMode=RCIM_Cubic, tangentMode=RCTM_Break, time=1, value=1}
        ]}}}
     Scale Curve=&RandomRangeFloat(Minimum=0, Maximum=0)
    )
    Index 3 Param 2=&FloatFromCurve(
     FloatCurve=DataInterface() {propertyValues {Curve {keys [
         {interpMode=RCIM_Cubic, leaveTangent=-0.0025, tangentMode=RCTM_Break, time=0, value=0.5}
         {arriveTangent=-0.0316, interpMode=RCIM_Cubic, tangentMode=RCTM_Break, time=1, value=-0.2}
        ]}}}
     Scale Curve=0.1
    )
    Index 3 Param 3=&FloatFromCurve(FloatCurve=DataInterface() {propertyValues {Curve {keys [
         {interpMode=RCIM_Cubic, leaveTangent=1.1034, tangentMode=RCTM_Break, time=0, value=-1}
         {arriveTangent=2.2399, tangentMode=RCTM_User, time=0.7, value=0.125}
        ]}}})
    Index 3 Param 4=&FloatFromCurve(FloatCurve=DataInterface() {propertyValues {Curve {keys [
         {interpMode=RCIM_Cubic, leaveTangent=3.0601, tangentMode=RCTM_Break, time=0.5, value=-0.5}
         {arriveTangent=3.2053, tangentMode=RCTM_User, time=1, value=1}
        ]}}})
    Module.Index0Param0Write=true
    Module.Index0Param1Write=true
    Module.Index0Param2Write=true
    Module.Index0Param3Write=true
    Module.Index1Param0Write=true
    Module.Index1Param1Write=true
    Module.Index1Param2Write=true
    Module.Index1Param3Write=true
    Module.Index2Param0Write=true
    Module.Index2Param1Write=true
    Module.Index2Param2Write=true
    Module.Index2Param3Write=true
    Module.Index3Param0Write=true
    Module.Index3Param1Write=true
    Module.Index3Param2Write=true
    Module.Index3Param3Write=true
    Write Parameter Index 1=true
    Write Parameter Index 2=true
    Write Parameter Index 3=true
   )
   &SolveForcesAndVelocity(Clamp Velocity=false, Limit Acceleration=false)
   &ScaleColor(
    Scale Alpha=&FloatFromCurve(
     FloatCurve=DataInterface() {propertyValues {Curve {keys [
         {tangentMode=RCTM_User, time=0, value=1}
         {arriveTangent=-0.0405, interpMode=RCIM_Cubic, leaveTangent=-0.0405, tangentMode=RCTM_User, time=0.7009, value=1}
         {arriveTangent=-8.0089, interpMode=RCIM_Cubic, leaveTangent=-8.0089, tangentMode=RCTM_User, time=1, value=0}
        ]}}}
     Scale Curve=&FloatFromCurve(
      FloatCurve=DataInterface() {propertyValues {Curve {
         defaultValue=0
         keys [
          {arriveTangent=4.1785, interpMode=RCIM_Cubic, leaveTangent=4.1785, tangentMode=RCTM_User, time=0, value=0}
          {arriveTangent=1.1179, interpMode=RCIM_Cubic, leaveTangent=1.1179, tangentMode=RCTM_User, time=0.2042, value=0.6228}
          {arriveTangent=0.0224, interpMode=RCIM_Cubic, leaveTangent=0.0224, tangentMode=RCTM_User, time=1, value=1}
         ]
        }}}
      CurveIndex=$Emitter.Age
     )
    )
    Scale RGB=&Lerp_Vector(
     A=&VectorFromFloat(Value=&FloatFromCurve(FloatCurve=DataInterface() {propertyValues {Curve {keys [<at 0.2771 1>,<at 0.3786 1>]}}}, Scale Curve=25))
     B=&VectorFromFloat(Value=&RandomRangeFloat(Minimum=0.3, Maximum=2, Evaluation Type=::NewEnumerator1))
     Alpha=&FloatFromCurve(FloatCurve=DataInterface() {propertyValues {Curve {keys [
          {interpMode=RCIM_Cubic, leaveTangent=2.5294, tangentMode=RCTM_Break, time=0, value=0}
          {arriveTangent=-0.2166, interpMode=RCIM_Cubic, tangentMode=RCTM_Break, time=0.4, value=1}
         ]}}})
    )
    ScaleA=true
    ScaleRGB=true
   )
  ]
  rendererClasses [&NiagaraMeshRendererProperties]
  renderers [
   {rendererClass=&NiagaraMeshRendererProperties, rendererIndex=0}
   {rendererClass=&NiagaraMeshRendererProperties, rendererIndex=1}
  ]
 }
 Steam {
  emitterUpdateScript [
   &EmitterState(Life Cycle Mode=::NewEnumerator1, Loop Behavior=::NewEnumerator1, Loop Duration=2, UseLoopDelay=false)
   &SpawnRate(
    SpawnRate=&Multiply_Float(
     A=10
     B=&FloatFromCurve(
      FloatCurve=DataInterface() {propertyValues {Curve {
         defaultValue=0
         keys [
          {arriveTangent=-2.5902, interpMode=RCIM_Cubic, leaveTangent=-2.5902, tangentMode=RCTM_User, time=0, value=1}
          {interpMode=RCIM_Cubic, tangentMode=RCTM_User, time=1, value=0}
         ]
        }}}
      CurveIndex=$Emitter.NormalizedLoopAge
     )
    )
    Use Spawn Probability=false
   )
  ]
  particleSpawnScript [
   &InitializeParticle(
    Color=<rgba 0.75 0.75 0.75 0.4>
    Lifetime Max=1.5
    Lifetime Min=0.6
    Lifetime Mode=::NewEnumerator1
    Sprite Rotation Mode=::NewEnumerator2
    Sprite Size Mode=::NewEnumerator1
    Uniform Sprite Size Max=50
    Uniform Sprite Size Min=25
    UsePositionOffset=true
    Write Material Random=true
   )
   &AddVelocity(Velocity=&RandomRangeVector(Minimum=<-5 -5 10>, Maximum=<5 5 20>, Evaluation Type=::NewEnumerator1))
   &CameraOffset(Camera Offset Amount=10)
  ]
  particleUpdateScript [
   &ParticleState()
   &SubUVAnimation(
    Sub UV Play Rate:NiagaraFloat=&RandomRangeFloat(Minimum=0.05, Maximum=0.1)
    SubUV Animation Mode:ENiagara_SubUVLookupMode=::NewEnumerator3
    SubUV Start Time Offset:NiagaraFloat=&RandomRangeFloat(Minimum=0, Maximum=63)
   )
   &DynamicMaterialParameters(
    Index 0 Param 1=&FloatFromCurve(FloatCurve=DataInterface() {propertyValues {Curve {keys [<at 0 -0.25>,<at 1 -1>]}}})
    Index 0 Param 4=&OneMinusFloat(Float=$Particles.NormalizedAge)
    Module.Param0WriteEnabled=true
    Module.Param1WriteEnabled=false
    Module.Param2WriteEnabled=false
    Module.Param3WriteEnabled=true
   )
   &ScaleColor(
    Scale Alpha=&FloatFromCurve(
     FloatCurve=DataInterface() {propertyValues {Curve {keys [
         {arriveTangent=5.2427, interpMode=RCIM_Cubic, leaveTangent=5.2427, tangentMode=RCTM_User, time=0, value=0}
         {arriveTangent=-1.0898, interpMode=RCIM_Cubic, leaveTangent=-1.0898, tangentMode=RCTM_User, time=0.3, value=1}
         {arriveTangent=-0.1231, interpMode=RCIM_Cubic, leaveTangent=-0.1231, tangentMode=RCTM_User, time=1, value=0.5}
        ]}}}
     Scale Curve=&Multiply_Float(A=&RandomRangeFloat(Minimum=0.5, Maximum=1), B=1)
    )
    ScaleA=true
    ScaleRGB=false
   )
   &ScaleSpriteSize(Scale Factor=&Vector2DFromFloat(Value=&FloatFromCurve(FloatCurve=DataInterface() {propertyValues {Curve {keys [
          {arriveTangent=1.8751, interpMode=RCIM_Cubic, leaveTangent=1.8751, tangentMode=RCTM_User, time=0, value=0.25}
          {arriveTangent=0.72, interpMode=RCIM_Cubic, leaveTangent=0.72, tangentMode=RCTM_User, time=1, value=1.25}
         ]}}})))
   &Drag(Drag=&RandomRangeFloat(Minimum=10, Maximum=20), Use Linear Drag=true, Use Rotational Drag=false)
   &BlendSpace(Rotation Contribution=0.5)
   &GravityForce(Gravity=<0 0 111>)
   &SolveForcesAndVelocity(Clamp Velocity=false, Limit Acceleration=false)
  ]
  rendererClasses [&NiagaraSpriteRendererProperties]
  renderers [{rendererClass=&NiagaraSpriteRendererProperties}]
 }
}
summary {
 systemName=NS_PiewPistol_Overheat_Reload
 emitters [
  {emitterName=Lightning, rendererClasses [&NiagaraMeshRendererProperties]}
  {emitterName=SplitLightning_ArcsFromHands, rendererClasses [&NiagaraMeshRendererProperties]}
  {emitterName=Steam, rendererClasses [&NiagaraSpriteRendererProperties]}
 ]
 userVariables [User.HueShift {type=NiagaraFloat, defaultValue=0.35, description=""}]
}
