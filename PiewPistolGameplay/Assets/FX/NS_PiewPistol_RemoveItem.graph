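# Niagara effect for removing an item from the PiewPistol, showing a short-lived mesh that gets thrown, tumbles and vanishes.
# A single burst emitter with two mesh renderers combines world-space downward velocity, a sideways push and gravity.
# Curves drive the spin rate and shrink the mesh to zero near end of life.
# An age comparison gates bouncy, count-limited collision, which also affects rotation.
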NiagaraSystem /PiewPistolGameplay/Assets/FX/NS_PiewPistol_RemoveItem
dependencies {
 usedDataInterfaces [&NiagaraDataInterfaceCurve]
 usedDynamicInputs []
 usedModules [&SystemState, &self:FindVel_01]
}
emitters {Mesh {
  emitterUpdateScript [
   &EmitterState()
   &SpawnBurst_Instantaneous(Use Spawn Probability=false, UseLoopCountLimit=true)
  ]
  particleSpawnScript [
   &InitializeParticle(Color Mode=::NewEnumerator0, Lifetime=0.3, Mesh Scale Mode=::NewEnumerator3, UsePositionOffset=false, Write Material Random=true)
   &InitialMeshOrientation(Mesh Orientation Mode=::NewEnumerator1, Rotation=<0.25 0 0.25>, Use Rotation Vector=true)
   &AddVelocity(Rotation Coordinate Space=::World, Velocity=<0 0 -350>) {module=AddVelocity001}
   &AddVelocity(Velocity=<0 -240 0>) {module=AddVelocity002}
  ]
  particleUpdateScript [
   &ParticleState()
   &UpdateMeshOrientation(
    Rotation Rate=&FloatFromCurve(
     FloatCurve=DataInterface() {propertyValues {Curve {
        defaultValue=1
        keys [
         {arriveTangent=9.0725, interpMode=RCIM_Cubic, leaveTangent=9.0725, tangentMode=RCTM_User, time=0, value=0}
         {arriveTangent=0.121, interpMode=RCIM_Cubic, leaveTangent=0.121, tangentMode=RCTM_User, time=0.1408, value=0.9847}
         {arriveTangent=-2.6279, interpMode=RCIM_Cubic, leaveTangent=-2.6279, tangentMode=RCTM_User, time=0.5001, value=0.25}
        ]
       }}}
     Scale Curve=2.5
    )
    Rotation Vector=<-1 0 0>
   )
   &ScaleMeshSize(Scale Factor=&VectorFromFloat(Value=&FloatFromCurve(FloatCurve=DataInterface() {propertyValues {Curve {keys [
          <at 0.8365 1 RCIM_Cubic>
          {arriveTangent=-2.0444, interpMode=RCIM_Cubic, leaveTangent=-2.0444, tangentMode=RCTM_User, time=0.8834, value=0.9443}
          {arriveTangent=-12.9752, interpMode=RCIM_Cubic, leaveTangent=-12.9752, tangentMode=RCTM_User, time=1, value=0}
         ]}}})))
   &GravityForce()
   &Collision(
    Age Colliding Particles=false
    CPU Collision Trace Channel=ECC_WorldStatic
    Collision Enabled=&SetBoolByFloatComparison(A=$Particles.Age, B=0.2)
    Control Roll On Collision=true
    Friction=1
    Friction During a Bounce=1
    KillOnCollision=false
    Max Number Of Collisions=3
    Module.Control Rest State Via Surface Normal=false
    Module.EnableMaxCollisionCount=true
    Module.Randomize Collision Normal Vector=true
    Particle Radius Scale=14
    Restitution=0.3
    ScaleRateOfRotation=true
   )
   &SolveForcesAndVelocity(Clamp Velocity=false, Limit Acceleration=false, Manually Enable Rotational Solver=true)
  ]
  rendererClasses [&NiagaraMeshRendererProperties]
  renderers [
   {rendererClass=&NiagaraMeshRendererProperties, rendererIndex=0}
   {rendererClass=&NiagaraMeshRendererProperties, rendererIndex=1}
  ]
 }}
summary {
 systemName=NS_PiewPistol_RemoveItem
 emitters [{emitterName=Mesh, rendererClasses [&NiagaraMeshRendererProperties]}]
 userVariables []
}
