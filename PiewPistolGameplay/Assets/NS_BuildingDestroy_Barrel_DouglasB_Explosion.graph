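# Niagara one-shot explosion effect, named for the destruction of the DouglasB barrel building, lasting about a second.
# Yellow-tinted mesh spheres, distortion, ground noise and an energy wave with a light renderer form the expanding blast.
# Sprite glow, flare, streaks, speedlines and radial smoke join cyan-blue arcs and mesh electricity, which adds a brief spawn rate to its burst.
# Lifetime curves drive scale, color fades and dynamic material parameters; no user variables are exposed.
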NiagaraSystem /PiewPistolGameplay/Assets/NS_BuildingDestroy_Barrel_DouglasB_Explosion
dependencies {
 usedDataInterfaces [
  &NiagaraDataInterfaceCurve
  &NiagaraDataInterfaceVector2DCurve
  &NiagaraDataInterfaceVectorCurve
  &NiagaraDataInterfaceColorCurve
 ]
 usedDynamicInputs [&MakeVector4_FromLinearColor]
 usedModules [&SystemState]
}
emitters {
 AirGlow {
  emitterUpdateScript [
   &EmitterState()
   &SpawnBurst_Instantaneous(Spawn Count=1, Use Spawn Probability=false)
  ]
  particleSpawnScript [&InitializeParticle(
    Color=<rgba 4 4 0.2322 1>
    Lifetime Max=0.5
    Lifetime Min=0.35
    Lifetime Mode=::NewEnumerator1
    Sprite Rotation Mode=::NewEnumerator2
    Sprite Size Mode=::NewEnumerator1
    Uniform Sprite Size Max=650
    Uniform Sprite Size Min=600
    UsePositionOffset=false
    Write Material Random=true
   )]
  particleUpdateScript [
   &ParticleState()
   &ScaleSpriteSize(Scale Factor=&Vector2DFromFloat(Value=&FloatFromCurve(
      FloatCurve=DataInterface() {propertyValues {Curve {keys [
          {arriveTangent=2.6083, interpMode=RCIM_Cubic, leaveTangent=0.4318, tangentMode=RCTM_User, time=0, value=0.5}
          {arriveTangent=1.3014, interpMode=RCIM_Cubic, leaveTangent=1.3014, tangentMode=RCTM_User, time=0.3, value=0.8}
          {arriveTangent=0.1331, leaveTangent=0.1331, time=1, value=1.2}
         ]}}}
      Scale Curve=1.25
     )))
   &ScaleColor(
    Scale Alpha=0.2
    Scale RGB=&VectorFromFloat(Value=&FloatFromCurve(FloatCurve=DataInterface() {propertyValues {Curve {keys [{interpMode=RCIM_Cubic, tangentMode=RCTM_User, time=0, value=1}, {interpMode=RCIM_Cubic, tangentMode=RCTM_User, time=1, value=0}]}}}))
    ScaleA=true
    ScaleRGB=true
   )
   &CameraOffset(Camera Offset Amount=20)
   &DynamicMaterialParameters(
    Index 0 Param 1=1
    Index 0 Param 2=1
    Index 0 Param 3=1
    Index 0 Param 4=1
    Module.Param0WriteEnabled=true
    Module.Param1WriteEnabled=true
    Module.Param2WriteEnabled=true
    Module.Param3WriteEnabled=true
   )
  ]
  rendererClasses [&NiagaraSpriteRendererProperties]
  renderers [{rendererClass=&NiagaraSpriteRendererProperties}]
 }
 Arcs {
  emitterUpdateScript [
   &EmitterState()
   &SpawnBurst_Instantaneous(Spawn Count=&RandomRangeInt(Minimum=2, Maximum=4), Use Spawn Probability=false)
  ]
  particleSpawnScript [
   /Niagara/Modules/Spawn/Initialization/InitializeParticle(
    Lifetime=&RandomRangeFloat(Minimum=0.25, Maximum=0.3)
    Module.Write Color=true
    Module.Write Lifetime=true
    Module.Write Mass=true
    Module.Write Material Random=true
    Module.Write Position=false
    Module.Write Scale=false
    Module.Write SpriteRotation=true
    Module.Write SpriteSize=true
    Position=&SimulationPosition()
    Sprite Rotation=&RandomRangeFloat(Minimum=-180, Maximum=180)
    Sprite Size=&Vector2DFromFloat(Value=&RandomRangeFloat(Minimum=350, Maximum=450))
   )
   &SubUVAnimation(End Frame=3, SubUV Animation Mode:ENiagara_SubUVLookupMode=::NewEnumerator2)
   &ShapeLocation(MaskSpawnGroup=false, Override Seed=false, UseOwnerScale=true)
  ]
  particleUpdateScript [
   &ParticleState()
   &CameraOffset(Camera Offset Amount=5)
   &ScaleSpriteSize(Scale Factor=&Vector2DFromFloat(Value=&FloatFromCurve(FloatCurve=DataInterface() {propertyValues {Curve {keys [
          {arriveTangent=1.7865, interpMode=RCIM_Cubic, leaveTangent=1.7865, tangentMode=RCTM_User, time=0.0023, value=0.1}
          {arriveTangent=0.6344, interpMode=RCIM_Cubic, leaveTangent=0.6344, tangentMode=RCTM_User, time=1, value=1}
         ]}}})))
   &Color(Color=&ColorFromCurve(ColorCurve=DataInterface() {propertyValues {
       AlphaCurve {keys [<at 0 1 RCIM_Cubic>,<at 1 0 RCIM_Cubic>]}
       BlueCurve {keys [<at 0 10 RCIM_Cubic>,<at 1 1 RCIM_Cubic>]}
       GreenCurve {keys [<at 0 7.0312 RCIM_Cubic>,<at 1 0.402 RCIM_Cubic>]}
       RedCurve {keys [<at 0 2.4906 RCIM_Cubic>,<at 1 0 RCIM_Cubic>]}
      }}))
   &SpriteRotationRate(Rotation Rate=100)
  ]
  rendererClasses [&NiagaraSpriteRendererProperties]
  renderers [{rendererClass=&NiagaraSpriteRendererProperties}]
 }
 Distortion {
  emitterUpdateScript [
   &EmitterState()
   &SpawnBurst_Instantaneous(Spawn Count=1, Use Spawn Probability=false)
  ]
  particleSpawnScript [/Niagara/Modules/Spawn/Initialization/V2/InitializeParticle(
    Color=<rgba 1 1 0.2491 1>
    Lifetime Max=0.4
    Lifetime Min=0.35
    Lifetime Mode=::NewEnumerator1
    Mesh Scale Mode=::NewEnumerator3
    Mesh Uniform Scale=26
    UsePositionOffset=false
    Write Material Random=true
   )]
  particleUpdateScript [
   &ParticleState()
   &DynamicMaterialParameters(
    Index 0 Param 1=&FloatFromCurve(FloatCurve=DataInterface() {propertyValues {Curve {keys [
         {interpMode=RCIM_Cubic, tangentMode=RCTM_User, time=0, value=1}
         {interpMode=RCIM_Cubic, tangentMode=RCTM_User, time=0.3, value=0.98}
         {arriveTangent=-0.0005, interpMode=RCIM_Cubic, leaveTangent=-0.0005, tangentMode=RCTM_User, time=0.6, value=0.98}
         {interpMode=RCIM_Cubic, tangentMode=RCTM_User, time=1, value=1}
        ]}}})
    Index 0 Param 2=&FloatFromCurve(FloatCurve=DataInterface() {propertyValues {Curve {keys [<at 0 15 RCIM_Cubic>, {arriveTangent=-40.0295, interpMode=RCIM_Cubic, leaveTangent=-40.0295, tangentMode=RCTM_User, time=1, value=0}]}}})
    Module.Param0WriteEnabled=true
    Module.Param1WriteEnabled=true
    Module.Param2WriteEnabled=true
    Module.Param3WriteEnabled=true
   )
   &ScaleMeshSize(Scale Factor=&VectorFromFloat(Value=&FloatFromCurve(FloatCurve=DataInterface() {propertyValues {Curve {keys [
          {arriveTangent=4.1785, interpMode=RCIM_Cubic, leaveTangent=4.1785, tangentMode=RCTM_User, time=0, value=0}
          {arriveTangent=1.9532, interpMode=RCIM_Cubic, leaveTangent=1.9532, tangentMode=RCTM_User, time=0.2223, value=0.6}
          {arriveTangent=-0.0405, interpMode=RCIM_Cubic, leaveTangent=-0.0405, tangentMode=RCTM_User, time=1, value=1.25}
         ]}}})))
   &ScaleColor(
    Scale Alpha=&FloatFromCurve(
     FloatCurve=DataInterface() {propertyValues {Curve {keys [
         {interpMode=RCIM_Cubic, tangentMode=RCTM_User, time=0.0045, value=0.9397}
         {arriveTangent=-1.0036, interpMode=RCIM_Cubic, leaveTangent=-1.0036, tangentMode=RCTM_User, time=0.4647, value=0.5233}
         {arriveTangent=-2.4196, interpMode=RCIM_Cubic, leaveTangent=-2.4196, tangentMode=RCTM_User, time=1.0011, value=0.002}
        ]}}}
     Scale Curve=0.2
    )
    Scale RGB=&VectorFromCurve(VectorCurve=DataInterface() {propertyValues {XCurve {keys [<at 0 1>,<at 0.5 0.8>]}, YCurve {keys [<at 0 1>,<at 0.5 0.8>]}, ZCurve {keys [<at 0 1>,<at 0.5 0.8>]}}})
    ScaleA=true
    ScaleRGB=true
   )
  ]
  rendererClasses [&NiagaraMeshRendererProperties]
  renderers [{rendererClass=&NiagaraMeshRendererProperties, rendererIndex=0}]
 }
 Electricity {
  emitterUpdateScript [
   &EmitterState(Life Cycle Mode=::NewEnumerator1, Loop Behavior=::NewEnumerator1, Loop Duration=0.2, UseLoopDelay=true)
   &SpawnBurst_Instantaneous(Spawn Count=6, Use Spawn Probability=false)
   &SpawnRate(SpawnRate=30, Use Spawn Probability=false)
  ]
  particleSpawnScript [
   /Niagara/Modules/Spawn/Initialization/InitializeParticle(
    Lifetime=&RandomRangeFloat(Minimum=0.225, Maximum=0.275)
    Mesh Scale=&Multiply_VectorByFloat(Vector=&RandomRangeVector(Minimum=<34 72 72>, Maximum=<45 90 90>), Float=2)
    Module.Write Color=true
    Module.Write Lifetime=true
    Module.Write Mass=true
    Module.Write Material Random=true
    Module.Write Position=true
    Module.Write Scale=true
    Module.Write SpriteRotation=false
    Module.Write SpriteSize=false
    Position=&SimulationPosition()
    Sprite Size=<25 25>
   )
   &ShapeLocation(
    Disc Coverage=1
    MaskSpawnGroup=false
    Override Seed=false
    Rotation Mode=::NewEnumerator1
    Shape Primitive=::NewEnumerator4
    UseOwnerScale=true
   )
   &InitialMeshOrientation(
    Mesh Orientation Mode=::NewEnumerator2
    Orientation Vector=$Engine.Owner.SystemXAxis
    Rotation=&RandomRangeVector(Minimum=<-1 -1 -1>, Maximum=<1 1 1>)
    Use Rotation Vector=true
   )
   &DynamicMaterialParameters(
    Index 0 Param 1=&RandomRangeFloat(Minimum=6.5, Maximum=8.5)
    Index 0 Param 3=2.2
    Index 0 Param 4=&RandomRangeFloat(Minimum=-60, Maximum=60)
    Module.Param0WriteEnabled=true
    Module.Param1WriteEnabled=false
    Module.Param2WriteEnabled=true
    Module.Param3WriteEnabled=false
   )
  ]
  particleUpdateScript [
   &ParticleState()
   &ScaleMeshSize(Scale Factor=&VectorFromFloat(Value=&FloatFromCurve(
      FloatCurve=DataInterface() {propertyValues {Curve {keys [
          {arriveTangent=4.469, interpMode=RCIM_Cubic, leaveTangent=4.469, tangentMode=RCTM_User, time=0, value=0}
          {arriveTangent=1.1375, interpMode=RCIM_Cubic, leaveTangent=1.1375, tangentMode=RCTM_User, time=1, value=2}
         ]}}}
      Scale Curve=0.25
     )))
   &DynamicMaterialParameters(
    Index 0 Param 2=&FloatFromCurve(FloatCurve=DataInterface() {propertyValues {Curve {keys [<at 0 2>,<at 1 -0.6>]}}})
    Index 0 Param 3=2.2
    Index 0 Param 4=&RandomRangeFloat(Minimum=-60, Maximum=60)
    Module.Param0WriteEnabled=false
    Module.Param1WriteEnabled=true
    Module.Param2WriteEnabled=false
    Module.Param3WriteEnabled=true
   ) {module=DynamicMaterialParameters001}
   &Color(Color=&ColorFromCurve(ColorCurve=DataInterface() {propertyValues {
       AlphaCurve {keys [<at 0 1 RCIM_Cubic>,<at 1 0 RCIM_Cubic>]}
       BlueCurve {keys [<at 0 10 RCIM_Cubic>,<at 1 1 RCIM_Cubic>]}
       GreenCurve {keys [<at 0 7.0312 RCIM_Cubic>,<at 1 0.402 RCIM_Cubic>]}
       RedCurve {keys [<at 0 2.4906 RCIM_Cubic>,<at 1 0 RCIM_Cubic>]}
      }}))
  ]
  rendererClasses [&NiagaraMeshRendererProperties]
  renderers [{rendererClass=&NiagaraMeshRendererProperties, rendererIndex=0}]
 }
 EnergyWave {
  emitterUpdateScript [
   &EmitterState()
   &SpawnBurst_Instantaneous(Use Spawn Probability=false, UseLoopCountLimit=false)
  ]
  particleSpawnScript [
   /Niagara/Modules/Spawn/Initialization/V2/InitializeParticle(
    Color Mode=::NewEnumerator0
    Lifetime=0.5
    Mesh Scale Mode=::NewEnumerator3
    Mesh Uniform Scale=10.7
    UsePositionOffset=false
    Write Material Random=true
   )
   &InitialMeshOrientation(Orientation Axis=<0 0 1>, Orientation Vector=<0 0 1>, Rotation=<0 0.25 0>, Use Orientation Vector=true, Use Rotation Vector=true)
   &InitialMeshOrientation(
    Orientation Axis=<0 0 0>
    Orientation Vector=<0 0 0>
    Rotation=&RandomRangeVector(Minimum=<0 0 0>, Maximum=<1 0 0>)
    Use Orientation Vector=false
    Use Rotation Vector=true
   ) {module=InitialMeshOrientation001}
   &CameraOffset(Camera Offset Amount=30)
   &ShapeLocation(MaskSpawnGroup=false, Offset=<0 0 25>, Offset Mode=::NewEnumerator0, Override Seed=false, Sphere Radius=1, UseOwnerScale=true)
   /SetVariables(Particles.LightPosOffset:NiagaraPosition=<0 0 50>) {isSetParameters=true}
  ]
  particleUpdateScript [
   &ParticleState()
   &ScaleColor(
    Linear Color Curve=DataInterface() {propertyValues {
      AlphaCurve {keys [<at 0 1>,<at 0.2388 1>,<at 1 0>]}
      BlueCurve {keys [<at 0 0>,<at 0.168 0>,<at 0.365 0>,<at 1 1>]}
      GreenCurve {keys [<at 0 1>,<at 0.168 1>,<at 0.365 1>,<at 1 0.8167>]}
      RedCurve {keys [<at 0 1>,<at 0.168 1>,<at 0.365 1>,<at 1 0>]}
     }}
    Scale Mode=::NewEnumerator2
   )
   &DynamicMaterialParameters(
    Index 0 Param 1=&FloatFromCurve(FloatCurve=DataInterface() {propertyValues {Curve {keys [{arriveTangent=-2.1015, interpMode=RCIM_Cubic, leaveTangent=-2.1015, tangentMode=RCTM_User, time=0, value=0.5}, <at 1 -0.15>]}}})
    Index 0 Param 2=&FloatFromCurve(FloatCurve=DataInterface() {propertyValues {Curve {keys [{arriveTangent=-0.3464, interpMode=RCIM_Cubic, leaveTangent=-0.3464, tangentMode=RCTM_User, time=0, value=1}, <at 1 0.9>]}}})
    Index 0 Param 3=&FloatFromCurve(FloatCurve=DataInterface() {propertyValues {Curve {keys [
         {interpMode=RCIM_Cubic, tangentMode=RCTM_Break, time=0, value=0}
         {arriveTangent=0.4478, interpMode=RCIM_Cubic, leaveTangent=0.4478, tangentMode=RCTM_User, time=0.15, value=1}
         {interpMode=RCIM_Cubic, tangentMode=RCTM_Break, time=1, value=1.2}
        ]}}})
    Module.Param0WriteEnabled=true
    Module.Param1WriteEnabled=true
    Module.Param2WriteEnabled=true
    Module.Param3WriteEnabled=false
   )
   &ScaleColor(
    Scale Alpha=&FloatFromCurve(FloatCurve=DataInterface() {propertyValues {Curve {keys [{interpMode=RCIM_Cubic, tangentMode=RCTM_User, time=0.5, value=1}, {interpMode=RCIM_Cubic, tangentMode=RCTM_User, time=1, value=0}]}}})
    Scale RGB=&VectorFromFloat(Value=&FloatFromCurve(FloatCurve=DataInterface() {propertyValues {Curve {keys [
          {interpMode=RCIM_Cubic, tangentMode=RCTM_Break, time=0, value=10}
          {arriveTangent=-6.6072, interpMode=RCIM_Cubic, leaveTangent=-6.6072, tangentMode=RCTM_User, time=0.15, value=3}
          {interpMode=RCIM_Cubic, tangentMode=RCTM_Break, time=1, value=1}
         ]}}}))
    ScaleA=true
    ScaleRGB=true
   ) {module=ScaleColor001}
   &ScaleMeshSize(Scale Factor=&VectorFromFloat(Value=&FloatFromCurve(FloatCurve=DataInterface() {propertyValues {Curve {keys [
          {interpMode=RCIM_Cubic, tangentMode=RCTM_Break, time=0, value=0.2}
          {arriveTangent=2.9365, interpMode=RCIM_Cubic, leaveTangent=2.9365, tangentMode=RCTM_User, time=0.1, value=1}
          {interpMode=RCIM_Cubic, tangentMode=RCTM_Break, time=1, value=2}
         ]}}})))
   &FN_Light_Attributes(
    Color=<rgba 0.0652 1.7006 3 1>
    Create Color Attribute=true
    Create Exponent Attribute=true
    Create Position Attribute=true
    Create Radius Attribute=true
    Exponent=2
    Position=&Add_Vector(A=Linked() {linkedVariable {name=Particles.Initial.Position, type=NiagaraPosition}}, B=<0 0 25>)
    Radius=100
    Scale Color=&FloatFromCurve(
     FloatCurve=DataInterface() {propertyValues {Curve {keys [
         <at 0 1 RCIM_Cubic>
         {arriveTangent=-1, interpMode=RCIM_Cubic, leaveTangent=-1, leaveTangentWeight=0.9776, time=0.7, value=1}
         <at 1 0 RCIM_Cubic>
        ]}}}
     Scale Curve=10
    )
   )
   /SetVariables(Particles.LightPos:NiagaraPosition=&Add_Vector(
     A=Linked() {linkedVariable {name=Particles.Position, type=NiagaraPosition}}
     B=Linked() {linkedVariable {name=Particles.LightPosOffset, type=NiagaraPosition}}
    )) {isSetParameters=true}
  ]
  rendererClasses [&NiagaraMeshRendererProperties, &NiagaraLightRendererProperties]
  renderers [
   {rendererClass=&NiagaraMeshRendererProperties, rendererIndex=0}
   {rendererClass=&NiagaraLightRendererProperties, rendererIndex=1}
  ]
 }
 GroundNoise {
  emitterUpdateScript [
   &EmitterState()
   &SpawnBurst_Instantaneous(Spawn Count=1, Use Spawn Probability=false)
  ]
  particleSpawnScript [
   &InitializeParticle(
    Color=<rgba 4 4 0.5319 10>
    Lifetime=0.45
    Mesh Scale Mode=::NewEnumerator3
    Mesh Uniform Scale=12
    UsePositionOffset=false
    Write Material Random=true
   )
   &InitialMeshOrientation(Mesh Orientation Mode=::NewEnumerator1, Use Rotation Vector=true)
  ]
  particleUpdateScript [
   &ParticleState()
   &ScaleColor(
    Scale Alpha=&FloatFromCurve(FloatCurve=DataInterface() {propertyValues {Curve {keys [<at 0 1>, {interpMode=RCIM_Cubic, tangentMode=RCTM_User, time=0.1, value=1}, {interpMode=RCIM_Cubic, tangentMode=RCTM_User, time=1, value=0}]}}})
    ScaleA=true
    ScaleRGB=true
   )
   &DynamicMaterialParameters(
    Index 0 Param 1=1
    Index 0 Param 3=1
    Index 0 Param 4=1
    Module.Param0WriteEnabled=true
    Module.Param1WriteEnabled=true
    Module.Param2WriteEnabled=true
    Module.Param3WriteEnabled=true
   )
   &ScaleMeshSize(Scale Factor=&VectorFromFloat(Value=&FloatFromCurve(FloatCurve=DataInterface() {propertyValues {Curve {keys [
          {arriveTangent=4.1785, interpMode=RCIM_Cubic, leaveTangent=4.1785, tangentMode=RCTM_User, time=0, value=0.05}
          {arriveTangent=0.9995, interpMode=RCIM_Cubic, leaveTangent=0.9995, tangentMode=RCTM_User, time=0.3, value=0.6}
          {arriveTangent=-0.0405, interpMode=RCIM_Cubic, leaveTangent=-0.0405, tangentMode=RCTM_User, time=1, value=1}
         ]}}})))
   &CameraOffset(Camera Offset Amount=250)
  ]
  rendererClasses [&NiagaraMeshRendererProperties]
  renderers [{rendererClass=&NiagaraMeshRendererProperties, rendererIndex=0}]
 }
 Impact_Flare {
  emitterUpdateScript [
   &EmitterState()
   &SpawnBurst_Instantaneous(Use Spawn Probability=false, UseLoopCountLimit=false)
  ]
  particleSpawnScript [
   &InitializeParticle(
    Color=<rgba 2 2 2 1>
    Lifetime=0.15
    Position Offset=<50 0 25>
    Position Offset Coordinate Space=::Local
    Sprite Rotation Mode=::NewEnumerator2
    Sprite Size Mode=::NewEnumerator1
    Uniform Sprite Size Max=1500
    Uniform Sprite Size Min=750
    UsePositionOffset=true
    Write Material Random=true
   )
   &DynamicMaterialParameters(
    Index 0 Param 1=1
    Index 0 Param 2=1
    Index 0 Param 3=1
    Index 0 Param 4=1
    Module.Param0WriteEnabled=true
    Module.Param1WriteEnabled=true
    Module.Param2WriteEnabled=true
    Module.Param3WriteEnabled=true
   )
   &CameraOffset(Camera Offset Amount=50)
  ]
  particleUpdateScript [
   &ParticleState()
   &ScaleColor(
    Scale Alpha=&FloatFromCurve(FloatCurve=DataInterface() {propertyValues {Curve {keys [{arriveTangent=-3.3102, interpMode=RCIM_Cubic, time=0, value=0}, <at 0.5 1 RCIM_Cubic>, <at 1 0 RCIM_Cubic>]}}})
    Scale RGB=&MakeVectorFromLinearColorRGB(Color=&ColorFromCurve(ColorCurve=DataInterface() {propertyValues {
        AlphaCurve {keys []}
        BlueCurve {keys [<at 0 10 RCIM_Cubic>,<at 0.125 1>,<at 1 0.1878 RCIM_Cubic>]}
        GreenCurve {keys [<at 0 0.3013 RCIM_Cubic>,<at 0.125 0.0301>,<at 1 0 RCIM_Cubic>]}
        RedCurve {keys [<at 0 4.989 RCIM_Cubic>,<at 0.125 0.4989>,<at 1 0.0343 RCIM_Cubic>]}
       }}))
    ScaleA=true
    ScaleRGB=true
   )
  ]
  rendererClasses [&NiagaraSpriteRendererProperties]
  renderers [{rendererClass=&NiagaraSpriteRendererProperties}]
 }
 Impact_Streaks {
  emitterUpdateScript [
   &EmitterState()
   &SpawnBurst_Instantaneous(
    Spawn Count=&RandomRangeInt(Minimum=18, Maximum=36)
    Use Spawn Probability=false
    UseLoopCountLimit=false
   )
  ]
  particleSpawnScript [
   &InitializeParticle(
    Color=<rgba 10 10 10 1>
    Lifetime Max=0.5
    Lifetime Min=0.35
    Lifetime Mode=::NewEnumerator1
    Sprite Size Max=<100 225>
    Sprite Size Min=<25 80>
    Sprite Size Mode=::NewEnumerator2
    UsePositionOffset=false
    Write Material Random=true
   )
   &SubUVAnimation(End Frame=3, SubUV Animation Mode:ENiagara_SubUVLookupMode=::NewEnumerator2)
   &ShapeLocation(
    Cylinder Distribution=::NewEnumerator2
    Cylinder Height=0
    Cylinder Radius=10
    MaskSpawnGroup=false
    Shape Primitive=::NewEnumerator1
    U Position=&ReturnNormalizedExecIndex()
    UseOwnerScale=true
   )
   &AddVelocity(
    Velocity=&Add_Vector(A=$Particles.ShapeLocation.ShapeNormal, B=&RandomRangeVector(Minimum=<0 0 0>, Maximum=<1 0 0>))
    Velocity Speed Scale=50
   )
  ]
  particleUpdateScript [
   &SolveForcesAndVelocity(Clamp Velocity=false, Limit Acceleration=false)
   &ParticleState()
   &ScaleColor(
    Scale Alpha=&FloatFromCurve(FloatCurve=DataInterface() {propertyValues {Curve {keys [{arriveTangent=-3.3102, interpMode=RCIM_Cubic, leaveTangent=-3.3102, tangentMode=RCTM_User, time=0, value=1}, <at 1 0 RCIM_Cubic>]}}})
    Scale RGB=&MakeVectorFromLinearColorRGB(Color=&ColorFromCurve(ColorCurve=DataInterface() {propertyValues {
        AlphaCurve {keys []}
        BlueCurve {keys [<at 0 10 RCIM_Cubic>,<at 0.125 1>,<at 1 0.1878 RCIM_Cubic>]}
        GreenCurve {keys [<at 0 4.0191 RCIM_Cubic>,<at 0.125 0.2888>,<at 1 0.0313 RCIM_Cubic>]}
        RedCurve {keys [<at 0 0.3013 RCIM_Cubic>,<at 0.125 0.0301>,<at 1 0 RCIM_Cubic>]}
       }}))
    ScaleA=true
    ScaleRGB=true
   )
   &ScaleSpriteSize(
    Non-Uniform Curve Sprite Scale=DataInterface() {propertyValues {
      XCurve {keys [
        {arriveTangent=0.9067, interpMode=RCIM_Cubic, leaveTangent=0.9067, tangentMode=RCTM_User, time=0, value=0.8}
        {arriveTangentWeight=0.2034, interpMode=RCIM_Cubic, leaveTangentWeight=0.2034, tangentMode=RCTM_User, time=0.5, value=1}
        {arriveTangent=-3.1509, interpMode=RCIM_Cubic, leaveTangent=-3.1509, tangentMode=RCTM_User, time=1, value=0.1}
       ]}
      YCurve {keys [
        {arriveTangent=3.7985, interpMode=RCIM_Cubic, leaveTangent=3.7985, tangentMode=RCTM_User, time=0, value=0}
        {arriveTangent=2, interpMode=RCIM_Cubic, leaveTangent=2, time=0.5, value=1.6}
        <at 1 2 RCIM_Cubic>
       ]}
     }}
    Scale Sprite Size Mode=::NewEnumerator3
   )
  ]
  rendererClasses [&NiagaraSpriteRendererProperties]
  renderers [{rendererClass=&NiagaraSpriteRendererProperties}]
 }
 InnerNoiseSphere {
  emitterUpdateScript [
   &EmitterState()
   &SpawnBurst_Instantaneous(Spawn Count=1, Use Spawn Probability=false)
  ]
  particleSpawnScript [&InitializeParticle(
    Color=<rgba 4 4 0.5319 3>
    Lifetime=0.3
    Mesh Scale Mode=::NewEnumerator3
    Mesh Uniform Scale=14
    UsePositionOffset=false
    Write Material Random=true
   )]
  particleUpdateScript [
   &ParticleState()
   &ScaleColor(
    Scale Alpha=&FloatFromCurve(FloatCurve=DataInterface() {propertyValues {Curve {keys [<at 0 1>, {interpMode=RCIM_Cubic, tangentMode=RCTM_User, time=0.1, value=1}, {interpMode=RCIM_Cubic, tangentMode=RCTM_User, time=1, value=0}]}}})
    ScaleA=true
    ScaleRGB=true
   )
   &DynamicMaterialParameters(
    Index 0 Param 1=1
    Index 0 Param 3=1
    Index 0 Param 4=1
    Module.Param0WriteEnabled=true
    Module.Param1WriteEnabled=true
    Module.Param2WriteEnabled=true
    Module.Param3WriteEnabled=true
   )
   &ScaleMeshSize(Scale Factor=&VectorFromFloat(Value=&FloatFromCurve(FloatCurve=DataInterface() {propertyValues {Curve {keys [
          {arriveTangent=4.1785, interpMode=RCIM_Cubic, leaveTangent=4.1785, tangentMode=RCTM_User, time=0, value=0.05}
          {arriveTangent=0.9995, interpMode=RCIM_Cubic, leaveTangent=0.9995, tangentMode=RCTM_User, time=0.3, value=0.6}
          {arriveTangent=-0.0405, interpMode=RCIM_Cubic, leaveTangent=-0.0405, tangentMode=RCTM_User, time=1, value=1}
         ]}}})))
   &CameraOffset(Camera Offset Amount=250)
  ]
  rendererClasses [&NiagaraMeshRendererProperties]
  renderers [{rendererClass=&NiagaraMeshRendererProperties, rendererIndex=0}]
 }
 InnerRadialSphere {
  emitterUpdateScript [
   &EmitterState()
   &SpawnBurst_Instantaneous(Spawn Count=1, Use Spawn Probability=false)
  ]
  particleSpawnScript [&InitializeParticle(
    Color=<rgba 1 1 0.2491 0.7>
    Lifetime=0.2
    Mesh Scale Mode=::NewEnumerator3
    Mesh Uniform Scale=26
    UsePositionOffset=false
    Write Material Random=true
   )]
  particleUpdateScript [
   &ParticleState()
   &ScaleColor(
    Scale Alpha=&FloatFromCurve(
     FloatCurve=DataInterface() {propertyValues {Curve {keys [
         {interpMode=RCIM_Cubic, tangentMode=RCTM_User, time=0, value=1}
         {interpMode=RCIM_Cubic, tangentMode=RCTM_User, time=0.1, value=1}
         {interpMode=RCIM_Cubic, tangentMode=RCTM_User, time=1, value=0}
        ]}}}
     Scale Curve=&FN_Select_Float_from_Time_Of_Day(Morning=0.6, Day=0.6, Evening=0.6, Night=0.6)
    )
    ScaleA=true
    ScaleRGB=true
   )
   &ScaleMeshSize(Scale Factor=&VectorFromFloat(Value=&FloatFromCurve(
      FloatCurve=DataInterface() {propertyValues {Curve {keys [
          {arriveTangent=4.1785, interpMode=RCIM_Cubic, leaveTangent=4.1785, tangentMode=RCTM_User, time=0, value=0.2}
          {arriveTangent=0.9995, interpMode=RCIM_Cubic, leaveTangent=0.9995, tangentMode=RCTM_User, time=0.15, value=0.8}
          {arriveTangent=-0.0405, interpMode=RCIM_Cubic, leaveTangent=-0.0405, tangentMode=RCTM_User, time=1, value=1}
         ]}}}
      Scale Curve=1.5
     )))
   &CameraOffset(Camera Offset Amount=-20)
  ]
  rendererClasses [&NiagaraMeshRendererProperties]
  renderers [{rendererClass=&NiagaraMeshRendererProperties, rendererIndex=0}]
 }
 OuterRadialSphere001 {
  emitterUpdateScript [
   &EmitterState()
   &SpawnBurst_Instantaneous(Spawn Count=2, Use Spawn Probability=false)
  ]
  particleSpawnScript [
   &InitializeParticle(
    Color=<rgba 4 4 0.8044 0.7>
    Lifetime=0.5
    Mesh Scale Mode=::NewEnumerator3
    Mesh Uniform Scale=28
    UsePositionOffset=false
    Write Material Random=true
   )
   &InitialMeshOrientation(
    Rotation=&RandomRangeVector(Minimum=<0 0 0>, Maximum=<1 1 1>)
    Use Orientation Vector=true
    Use Rotation Vector=true
   )
  ]
  particleUpdateScript [
   &ParticleState()
   &ScaleColor(
    Scale Alpha=&FloatFromCurve(
     FloatCurve=DataInterface() {propertyValues {Curve {keys [
         {arriveTangent=45, interpMode=RCIM_Cubic, time=0, value=1}
         {arriveTangent=-1, interpMode=RCIM_Cubic, leaveTangent=-1, time=0.3, value=2}
         <at 1 0 RCIM_Cubic>
        ]}}}
     Scale Curve=0.25
    )
    Scale RGB=<3 3 3>
    ScaleA=true
    ScaleRGB=true
   )
   &ScaleMeshSize(Scale Factor=&VectorFromFloat(Value=&FloatFromCurve(FloatCurve=DataInterface() {propertyValues {Curve {keys [
          {arriveTangent=4.1785, interpMode=RCIM_Cubic, leaveTangent=4.1785, tangentMode=RCTM_User, time=0, value=0.05}
          {arriveTangent=0.9995, interpMode=RCIM_Cubic, leaveTangent=0.9995, tangentMode=RCTM_User, time=0.15, value=0.6825}
          {arriveTangent=-0.0405, interpMode=RCIM_Cubic, leaveTangent=-0.0405, tangentMode=RCTM_User, time=1, value=1}
         ]}}})))
   &DynamicMaterialParameters(
    Index 0 Param 1=1
    Index 0 Param 2=&FloatFromCurve(FloatCurve=DataInterface() {propertyValues {Curve {keys [{interpMode=RCIM_Cubic, tangentMode=RCTM_User, time=0, value=0.25}, {interpMode=RCIM_Cubic, tangentMode=RCTM_User, time=1, value=-1}]}}})
    Index 0 Param 3=0.5
    Index 0 Param 4=50
    Module.Param0WriteEnabled=true
    Module.Param1WriteEnabled=true
    Module.Param2WriteEnabled=true
    Module.Param3WriteEnabled=true
   )
  ]
  rendererClasses [&NiagaraMeshRendererProperties]
  renderers [{rendererClass=&NiagaraMeshRendererProperties, rendererIndex=0}]
 }
 SmokeWorld {
  emitterUpdateScript [
   &EmitterState()
   &SpawnBurst_Instantaneous(Spawn Count=26, Use Spawn Probability=false)
  ]
  particleSpawnScript [
   &InitializeParticle(
    Lifetime Max=0.85
    Lifetime Min=0.6
    Lifetime Mode=::NewEnumerator1
    Position Offset=<0 0 5>
    Sprite Rotation Mode=::NewEnumerator2
    Sprite Size Mode=::NewEnumerator1
    Uniform Sprite Size Max=300
    Uniform Sprite Size Min=180
    UsePositionOffset=true
    Write Material Random=true
   )
   &ShapeLocation(
    Cylinder Distribution=::NewEnumerator2
    Cylinder Height=0
    Cylinder Radius=2
    MaskSpawnGroup=false
    Rotation Mode=::NewEnumerator1
    Shape Primitive=::NewEnumerator1
    U Position=&ReturnNormalizedExecIndex()
    UseOwnerScale=true
   )
   &AddVelocity(
    Velocity=$Particles.ShapeLocation.ShapeNormal
    Velocity Speed Scale=&RandomRangeFloat(Minimum=1400, Maximum=1900)
   )
   /SetVariables(Particles.RandomFloat=&RandomRangeFloat(Minimum=0, Maximum=1)) {isSetParameters=true}
  ]
  particleUpdateScript [
   &ParticleState()
   &SpriteRotationRate(Rotation Rate=&Multiply_Float(
     A=1
     B=&FloatFromCurve(
      FloatCurve=DataInterface() {propertyValues {Curve {keys [
          {arriveTangent=-4.0107, interpMode=RCIM_Cubic, leaveTangent=-4.0107, tangentMode=RCTM_User, time=-0.0046, value=1.0113}
          {arriveTangent=-0.2496, interpMode=RCIM_Cubic, leaveTangent=-0.2496, tangentMode=RCTM_User, time=0.5, value=0.16}
         ]}}}
      CurveIndex=$Emitter.Age
      Scale Curve=150
     )
    ))
   &Drag(
    Drag=&FloatFromCurve(
     FloatCurve=DataInterface() {propertyValues {Curve {keys [<at 0 2 RCIM_Cubic>,<at 0.5 1 RCIM_Cubic>]}}}
     Scale Curve=&RandomRangeFloat(Minimum=1, Maximum=1.5)
    )
    Use Linear Drag=true
    Use Rotational Drag=false
   )
   &SolveForcesAndVelocity(Clamp Velocity=false, Limit Acceleration=false)
   &ScaleSpriteSize(Scale Factor=&Vector2DFromFloat(Value=&FloatFromCurve(FloatCurve=DataInterface() {propertyValues {Curve {keys [{arriveTangent=1.7458, interpMode=RCIM_Cubic, leaveTangent=1.7458, tangentMode=RCTM_User, time=0, value=0.5}, <at 1 1.25 RCIM_Cubic>]}}})))
   &DynamicMaterialParameters(
    Index 0 Param 1=&Add_Float(
     A=$Particles.RandomFloat
     B=&FloatFromCurve(FloatCurve=DataInterface() {propertyValues {Curve {keys [
          {arriveTangent=0.9088, interpMode=RCIM_Cubic, leaveTangent=0.9088, tangentMode=RCTM_User, time=0, value=1}
          {arriveTangent=0.5473, interpMode=RCIM_Cubic, leaveTangent=0.5473, tangentMode=RCTM_User, time=1, value=2}
         ]}}})
    )
    Index 1 Param 1=&FloatFromCurve(
     FloatCurve=DataInterface() {propertyValues {Curve {keys [<at 0 0>,<at 0.25 1>]}}}
     Scale Curve=&RandomRangeFloat(Minimum=2, Maximum=1)
    )
    Index 2 Param 1=&FloatFromCurve(FloatCurve=DataInterface() {propertyValues {Curve {keys [{interpMode=RCIM_Cubic, leaveTangent=3.2309, tangentMode=RCTM_Break, time=0.1, value=0}, <at 1 1>]}}})
    Index 2 Param 2=&FloatFromCurve(
     FloatCurve=DataInterface() {propertyValues {Curve {keys [<at 0 0.2 RCIM_Cubic>,<at 1 0.8 RCIM_Cubic>]}}}
     Scale Curve=&RandomRangeFloat(Minimum=0.2, Maximum=1.2)
    )
    Index 2 Param 3=&FloatFromCurve(FloatCurve=DataInterface() {propertyValues {Curve {keys [
         {interpMode=RCIM_Cubic, tangentMode=RCTM_Break, time=0, value=1}
         {interpMode=RCIM_Cubic, leaveTangent=11.1233, tangentMode=RCTM_Break, time=1, value=2}
        ]}}})
    Index 2 Param 4=&FloatFromCurve(
     FloatCurve=DataInterface() {propertyValues {Curve {keys [<at 0 1 RCIM_Cubic>, {interpMode=RCIM_Cubic, leaveTangent=-1.2956, tangentMode=RCTM_Break, time=1, value=0.75}]}}}
     Scale Curve=0.8
    )
    Module.Param0WriteEnabled=true
    Module.Param0WriteEnabled1=false
    Module.Param0WriteEnabled2=true
    Module.Param1WriteEnabled=false
    Module.Param1WriteEnabled1=false
    Module.Param1WriteEnabled2=true
    Module.Param2WriteEnabled=false
    Module.Param2WriteEnabled1=false
    Module.Param2WriteEnabled2=true
    Module.Param3WriteEnabled=true
    Module.Param3WriteEnabled1=false
    Module.Param3WriteEnabled2=true
    Write Parameter Index 1=true
    Write Parameter Index 2=true
   )
   &Color(Color=&ColorFromCurve(ColorCurve=DataInterface() {propertyValues {
       AlphaCurve {keys [<at 0 0.5 RCIM_Cubic>,<at 1 0 RCIM_Cubic>]}
       BlueCurve {keys [<at 0 0.2491 RCIM_Cubic>,<at 0.5389 0.4149 RCIM_Cubic>]}
       GreenCurve {keys [<at 0 1 RCIM_Cubic>,<at 0.5389 0.4149 RCIM_Cubic>]}
       RedCurve {keys [<at 0 1 RCIM_Cubic>,<at 0.5389 0.4149 RCIM_Cubic>]}
      }}))
   &CameraOffset(Camera Offset Amount=&RandomRangeFloat(Minimum=30, Maximum=50))
  ]
  rendererClasses [&NiagaraSpriteRendererProperties]
  renderers [
   {rendererClass=&NiagaraSpriteRendererProperties}
   {rendererClass=&NiagaraSpriteRendererProperties, rendererIndex=1}
  ]
 }
 SnapMesh {
  emitterUpdateScript [
   &EmitterState()
   &SpawnBurst_Instantaneous(Spawn Count=1, Spawn Probability=0.05, Use Spawn Probability=false)
  ]
  particleSpawnScript [
   &InitializeParticle(
    Color=<rgba 2 2 0.0181 1>
    Lifetime Max=0.6
    Lifetime Min=0.4
    Lifetime Mode=::NewEnumerator1
    Mesh Scale Mode=::NewEnumerator1
    Mesh Uniform Scale Max=16
    Mesh Uniform Scale Min=14.5
    UsePositionOffset=false
    Write Material Random=true
   )
   &InitialMeshOrientation(
    Rotation=&RandomRangeVector(Minimum=<-180 -180 -180>, Maximum=<180 180 180>)
    Use Orientation Vector=true
    Use Rotation Vector=true
   )
  ]
  particleUpdateScript [
   &ParticleState()
   &ScaleColor(
    Scale Alpha=&FloatFromCurve(FloatCurve=DataInterface() {propertyValues {Curve {keys [
         {tangentMode=RCTM_User, time=0, value=1}
         {arriveTangent=-0.0405, interpMode=RCIM_Cubic, leaveTangent=-0.0405, tangentMode=RCTM_User, time=0.7009, value=1}
         {arriveTangent=-8.0089, interpMode=RCIM_Cubic, leaveTangent=-8.0089, tangentMode=RCTM_User, time=1, value=0}
        ]}}})
    Scale RGB=&VectorFromFloat(Value=&FloatFromCurve(
      FloatCurve=DataInterface() {propertyValues {Curve {keys [{interpMode=RCIM_Cubic, tangentMode=RCTM_User, time=0, value=1}, {interpMode=RCIM_Cubic, tangentMode=RCTM_User, time=1, value=0}]}}}
      Scale Curve=20
     ))
    ScaleA=true
    ScaleRGB=true
   )
   &DynamicMaterialParameters(
    Index 0 Param 1=&FloatFromCurve(FloatCurve=DataInterface() {propertyValues {Curve {keys [{arriveTangent=0.0592, interpMode=RCIM_Cubic, time=0, value=0}, <at 1 1 RCIM_Cubic>]}}})
    Index 0 Param 2=50
    Index 0 Param 3=1
    Index 0 Param 4=1
    Module.Param0WriteEnabled=true
    Module.Param1WriteEnabled=true
    Module.Param2WriteEnabled=true
    Module.Param3WriteEnabled=true
   )
   &ScaleMeshSize(Scale Factor=&VectorFromFloat(Value=&FloatFromCurve(
      FloatCurve=DataInterface() {propertyValues {Curve {keys [{arriveTangent=2.0369, interpMode=RCIM_Cubic, leaveTangent=2.0369, tangentMode=RCTM_User, time=0, value=0.25}, <at 1 1>]}}}
      Scale Curve=1.8
     )))
  ]
  rendererClasses [&NiagaraMeshRendererProperties]
  renderers [{rendererClass=&NiagaraMeshRendererProperties, rendererIndex=0}]
 }
 Speedlines {
  emitterUpdateScript [
   &EmitterState()
   &SpawnBurst_Instantaneous(Spawn Count=&RandomRangeInt(Minimum=20, Maximum=30), Use Spawn Probability=false)
  ]
  particleSpawnScript [
   /Niagara/Modules/Spawn/Initialization/InitializeParticle(
    Color=<rgba 1 1 1 0.25>
    Lifetime=&RandomRangeFloat(Minimum=0.3, Maximum=0.4, Evaluation Type=::NewEnumerator1)
    Module.Write Color=true
    Module.Write Lifetime=true
    Module.Write Mass=false
    Module.Write Material Random=true
    Module.Write Position=false
    Module.Write Scale=false
    Module.Write SpriteRotation=false
    Module.Write SpriteSize=true
    Sprite Size=&RandomRangeVector2D(Minimum=<4 18>, Maximum=<8 24>, Evaluation Type=::NewEnumerator1)
   )
   /Niagara/Modules/Spawn/Location/SphereLocation(MaskSpawnGroup=false, Module.Override Randomness=false, Sphere Radius=300, Surface Only=true)
   &AddVelocityFromPoint(Module.Use Falloff Distance=false, Velocity Strength=&RandomRangeFloat(Minimum=600, Maximum=1800))
  ]
  particleUpdateScript [
   &ParticleState()
   &ScaleColor(
    Scale Alpha=&FloatFromCurve(FloatCurve=DataInterface() {propertyValues {Curve {keys [<at 0 1>,<at 1 0>]}}}, Scale Curve=0.5)
    Scale RGB=<3 3 3>
    ScaleA=true
    ScaleRGB=true
   )
   &Drag(Use Linear Drag=true, Use Rotational Drag=false)
   &ScaleSpriteSize(Scale Factor=&Vector2DFromCurve(Vector2Curve=DataInterface() {propertyValues {XCurve {keys [<at 0 1>,<at 1 0>]}, YCurve {keys [<at 0 8>,<at 1 4>]}}}, ScaleCurve=<1 1>))
   &SolveForcesAndVelocity(Clamp Velocity=false, Limit Acceleration=false)
  ]
  rendererClasses [&NiagaraSpriteRendererProperties]
  renderers [{rendererClass=&NiagaraSpriteRendererProperties}]
 }
 SphereChroma {
  emitterUpdateScript [
   &EmitterState()
   &SpawnBurst_Instantaneous(Spawn Count=2, Use Spawn Probability=false)
  ]
  particleSpawnScript [
   /Niagara/Modules/Spawn/Initialization/V2/InitializeParticle(
    Color=<rgba 2 2 0.4022 2>
    Lifetime=0.5
    Mesh Scale Mode=::NewEnumerator3
    Mesh Uniform Scale=25.3
    UsePositionOffset=false
    Write Material Random=true
   )
   &InitialMeshOrientation(
    Rotation=&RandomRangeVector(Minimum=<0 0 0>, Maximum=<1 1 1>)
    Use Orientation Vector=true
    Use Rotation Vector=true
   )
  ]
  particleUpdateScript [
   &ParticleState()
   &ScaleColor(
    Scale Alpha=&FloatFromCurve(
     FloatCurve=DataInterface() {propertyValues {Curve {keys [
         {arriveTangent=45, interpMode=RCIM_Cubic, time=0, value=1}
         {arriveTangent=-1, interpMode=RCIM_Cubic, leaveTangent=-1, time=0.3, value=2}
         <at 1 0 RCIM_Cubic>
        ]}}}
     Scale Curve=0.25
    )
    Scale RGB=<3 3 3>
    ScaleA=true
    ScaleRGB=true
   )
   &ScaleMeshSize(Scale Factor=&VectorFromFloat(Value=&FloatFromCurve(FloatCurve=DataInterface() {propertyValues {Curve {keys [
          {arriveTangent=4.1785, interpMode=RCIM_Cubic, leaveTangent=4.1785, tangentMode=RCTM_User, time=0, value=0.05}
          {arriveTangent=0.9995, interpMode=RCIM_Cubic, leaveTangent=0.9995, tangentMode=RCTM_User, time=0.15, value=0.6825}
          {arriveTangent=-0.0405, interpMode=RCIM_Cubic, leaveTangent=-0.0405, tangentMode=RCTM_User, time=1, value=1}
         ]}}})))
   &DynamicMaterialParameters(
    Index 0 Param 1=&FloatFromCurve(FloatCurve=DataInterface() {propertyValues {Curve {keys [<at 0 0.07>,<at 1 0.25>]}}})
    Index 0 Param 2=&RandomRangeFloat(Minimum=0, Maximum=1)
    Index 0 Param 3=&RandomRangeFloat(Minimum=0, Maximum=1)
    Index 0 Param 4=&FloatFromCurve(FloatCurve=DataInterface() {propertyValues {Curve {keys [<at 0 1>,<at 1 3>]}}})
    Module.Param0WriteEnabled=true
    Module.Param1WriteEnabled=true
    Module.Param2WriteEnabled=true
    Module.Param3WriteEnabled=true
   )
  ]
  rendererClasses [&NiagaraMeshRendererProperties]
  renderers [{rendererClass=&NiagaraMeshRendererProperties, rendererIndex=0}]
 }
}
summary {
 systemName=NS_BuildingDestroy_Barrel_DouglasB_Explosion
 emitters [
  {emitterName=InnerRadialSphere, rendererClasses [&NiagaraMeshRendererProperties]}
  {emitterName=SphereChroma, rendererClasses [&NiagaraMeshRendererProperties]}
  {emitterName=SnapMesh, rendererClasses [&NiagaraMeshRendererProperties]}
  {emitterName=Speedlines, rendererClasses [&NiagaraSpriteRendererProperties]}
  {emitterName=AirGlow, rendererClasses [&NiagaraSpriteRendererProperties]}
  {emitterName=InnerNoiseSphere, rendererClasses [&NiagaraMeshRendererProperties]}
  {emitterName=Distortion, rendererClasses [&NiagaraMeshRendererProperties]}
  {emitterName=SmokeWorld, rendererClasses [&NiagaraSpriteRendererProperties]}
  {emitterName=OuterRadialSphere001, rendererClasses [&NiagaraMeshRendererProperties]}
  {emitterName=GroundNoise, rendererClasses [&NiagaraMeshRendererProperties]}
  {emitterName=Electricity, rendererClasses [&NiagaraMeshRendererProperties]}
  {emitterName=Arcs, rendererClasses [&NiagaraSpriteRendererProperties]}
  {emitterName=Impact_Streaks, rendererClasses [&NiagaraSpriteRendererProperties]}
  {emitterName=Impact_Flare, rendererClasses [&NiagaraSpriteRendererProperties]}
  {emitterName=EnergyWave, rendererClasses [&NiagaraMeshRendererProperties, &NiagaraLightRendererProperties]}
 ]
 userVariables []
}
