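# Sound Cue for a neon pistol fire sound in the PiewPistolGameplay content, one numbered "04" by name.
# A single Wave Player node is the root and plays the Pistol_Neon_Fire_04 sound wave directly.
# The graph has no randomization, modulation or attenuation nodes, so the cue acts as a plain wrapper around that one wave.
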SoundCue /PiewPistolGameplay/Assets/Pistol_Neon_Fire_04_Cue
root=WavePlayer_0
nodes [{id=WavePlayer_0, properties {soundWaveAssetPtr=&/Pistol_Neon_Fire_04}}]
properties {maxDistance=2097152}
