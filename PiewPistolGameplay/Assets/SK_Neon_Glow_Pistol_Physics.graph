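# Physics asset for the Neon Glow Pistol skeletal mesh, giving the weapon simple collision and physics bodies.
# Each of the six bones (Root, Handle, Magazine, Slide, Hammer, Trigger) has one box, with the large Handle box as the main volume.
# Constraints form a star: Magazine, Slide, Hammer and Trigger join to Handle, which joins to Root.
# All swing and twist axes are Free, so the 45-degree angular limits do not restrict rotation.
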PhysicsAsset /PiewPistolGameplay/Assets/SK_Neon_Glow_Pistol_Physics
bodies {
 Hammer [{shapeType=Box, extentX=1.2419, extentY=2.6563, extentZ=2.7336, center=<0.3312 -1.556 1.032>}]
 Handle [{shapeType=Box, extentX=33.9048, extentY=9.4161, extentZ=25.703, center=<12.0453 -0.4218 5.4481>}]
 Magazine [{shapeType=Box, extentX=8.2753, extentY=2.8509, extentZ=12.0776, center=<0.1418 -0.4292 0.5193>}]
 Root [{shapeType=Box, extentX=1.01, extentY=1.01, extentZ=1.01}]
 Slide [{shapeType=Box, extentX=4.1342, extentY=3.2585, extentZ=5.2477, center=<1.4569 -0.4729 0.9257>}]
 Trigger [{shapeType=Box, extentX=3.4143, extentY=1.4638, extentZ=3.1012, center=<-1.1492 -0.4728 -1.7349>}]
}
constraints [
 {
  bone1Name=Handle
  bone2Name=Root
  swing1LimitDegrees=45
  swing1Motion=Free
  swing2LimitDegrees=45
  swing2Motion=Free
  twistLimitDegrees=45
  twistMotion=Free
 }
 {
  bone1Name=Magazine
  bone2Name=Handle
  swing1LimitDegrees=45
  swing1Motion=Free
  swing2LimitDegrees=45
  swing2Motion=Free
  twistLimitDegrees=45
  twistMotion=Free
 }
 {
  bone1Name=Slide
  bone2Name=Handle
  swing1LimitDegrees=45
  swing1Motion=Free
  swing2LimitDegrees=45
  swing2Motion=Free
  twistLimitDegrees=45
  twistMotion=Free
 }
 {
  bone1Name=Hammer
  bone2Name=Handle
  swing1LimitDegrees=45
  swing1Motion=Free
  swing2LimitDegrees=45
  swing2Motion=Free
  twistLimitDegrees=45
  twistMotion=Free
 }
 {
  bone1Name=Trigger
  bone2Name=Handle
  swing1LimitDegrees=45
  swing1Motion=Free
  swing2LimitDegrees=45
  swing2Motion=Free
  twistLimitDegrees=45
  twistMotion=Free
 }
]
